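# Parent head material for the LazarusLens Wing character, layering animated face and glowing tattoo effects between the shared character PreFX and PostFX functions.
# The face animation function takes base color, emissive and pre-skinned position, then outputs tattoo color and a mask that drives the glow.
# An IsEyes static switch uses procedural eyes instead, built from mirrored, offset UVs, time-scrolled noise distortion, sphere masks and sine waves blending parameterized colors.
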Material /BRCosmetics/Characters/Player/Male/Medium/Bodies/M_MED_LazarusLens_Wing/Materials/M_LazarusLens_Wing_Head_Parent
graph {
 properties {
  bUseMaterialAttributes=true
  bUsedWithMeshDeformer=true
  bUsedWithSkeletalMesh=true
  pixelDepthOffsetMode=PDOM_Legacy
  shadingModel=MSM_FromMaterialExpression
  usedShadingModels=Default Lit | Subsurface | Clear Coat | Subsurface Profile | Cloth
 }
 nodes [
  Constant() {r=1}
  n1 = NamedRerouteUsage()
  n2 = MaterialFunctionCall() {materialFunction=&MF_CharacterParent_2023_PreFX}
  n3 = GetMaterialAttributes(None=@n2.Result) {attributeGetTypes [#1, #2]}
  n4 = MaterialFunctionCall!(
   BaseColor=NamedRerouteDeclaration(None=@n3.Base Color) {name=InBaseColor, nodeColor=<rgba 1 0.5859 0 1>}
   Emissive=NamedRerouteDeclaration(None=@n3.Emissive Color) {name=InEmissive, nodeColor=<rgba 0.9922 1 0 1>}
   PreSkinLocalPos=NamedRerouteDeclaration(None=@n2.Pre skinned Local Position) {name=PreSkinLoc, nodeColor=<rgba 0 0.5312 1 1>}
  ) {materialFunction=/Game/Characters/Player/Male/Medium/Bodies/M_MED_LazarusLens_Wing/Materials/MF_LazarusLens_Wings_FaceAnim}
  n5 = MaterialFunctionCall!(TattooMask=@n4.TatooMask) {materialFunction=/Game/Characters/Player/Male/Medium/Bodies/M_MED_LazarusLens/Materials/MF_LazarusLens_TattooGlow}
  n6 = StaticSwitchParameter(True=Multiply(A=@n1, B=ScalarParameter() {defaultValue=0.5, parameterName=Eyes_EmissiveIntensity}) {constB=0.5}, False=NamedRerouteUsage()) {parameterName=IsEyes}
  MaterialFunctionCall(
   Material=SetMaterialAttributes!(
    MaterialAttributes=@n2.Result
    Base Color=StaticSwitchParameter(True=@n1, False=@n4.TatooBaseColor) {parameterName=IsEyes}
    Emissive Color=Add(A=@n5.Result, B=@n6)
   ) {attributeSetTypes [#1, #2]}
   M=@n2.M
   Pre Skinned Local Position=@n2.M
   Distance Blend=@n2.M
   Fresnel=@n2.M
  ) {materialFunction=&MF_CharacterParent_2023_PostFX}
  n7 = VectorParameter() {group=Eye, parameterName=EyeColor_Background_A, defaultValue=<rgba 0.7101 0.1837 0.2149 1>}
  n8 = VectorParameter() {group=Eye, parameterName=EyeColor_Background_B, defaultValue=<rgba 0.6337 0.4236 0.436 1>}
  n9 = MaterialFunctionCall(Float2=Add(A=TextureCoordinate(), B=Constant2Vector() {r=0.04})) {materialFunction=&BreakOutFloat2Components}
  n10 = ConstantBiasScale(None=@n9.G) {bias=-0.5, scale=2}
  n11 = Abs(None=@n10)
  n12 = Constant2Vector() {g=0.267, r=0.02}
  n13 = Subtract(A=Multiply(A=Ceil(None=@n10)) {constB=2})
  n14 = Add(A=@n12.R, B=Multiply(A=ScalarParameter() {defaultValue=-0.0075, group=Eye, parameterName=EyePosition_X}, B=@n13))
  n15 = Add(A=@n12.G, B=Multiply(A=@n13, B=ScalarParameter() {defaultValue=-0.01, group=Eye, parameterName=EyePosition_Y}))
  n16 = Add(A=Add(A=AppendVector(A=@n9.R, B=@n11), B=AppendVector(A=@n14, B=@n15)), B=Constant2Vector() {r=-0.5})
  n17 = Multiply(A=Constant2Vector() {g=2, r=0.5}, B=Constant() {r=0.25})
  n18 = MaterialFunctionCall(UVs=@n16, Texture Scale=@n17) {materialFunction=&ScaleUVsByCenter}
  n19 = ScalarParameter() {defaultValue=1.5, group=Eye, parameterName=EyeDistortion_Scale}
  n20 = Add(
   A=MaterialFunctionCall(UVs=@n16, Texture Scale=Multiply(A=@n17, B=@n19) {constB=1.5}) {materialFunction=&ScaleUVsByCenter}.UVs
   B=Multiply(A=Time(), B=ScalarParameter() {defaultValue=-0.02, group=Eye, parameterName=EyeDistortion_Speed}) {constB=-0.02}
  )
  n21 = Abs(None=ConstantBiasScale(None=TextureSample(UVs=@n20) {samplerType=SAMPLERTYPE_LinearColor, texture=/Game/Effects/FogVolume/noise_flamy_Tile_Linear}.R) {bias=-0.5, scale=2})
  n22 = Constant2Vector() {g=1, r=0.5}
  n23 = ScalarParameter() {group=Eye, parameterName=EyeDistortion_Intensity_Center}
  n24 = ConstantBiasScale(None=Frac(None=Time())) {bias=-0.5, scale=2}
  n25 = Add(A=@n23, B=Multiply(A=Divide(A=@n23), B=Abs(None=@n24)))
  n26 = Add(A=@n18.UVs, B=Multiply(A=Multiply(A=@n21, B=@n22), B=Multiply(A=@n22, B=@n25)) {constB=0.05})
  n27 = Constant() {r=0.5}
  n28 = SphereMask(A=@n26, B=@n27, Radius=ScalarParameter() {defaultValue=0.2, group=Eye, parameterName=EyeColor_BlendRadius}) {attenuationRadius=0.2, hardnessPercent=0}
  n29 = VectorParameter() {group=Eye, parameterName=EyeColor_Waves, defaultValue=<rgba 0.6372 0 0.0076 1>}
  n30 = Add(A=@n18.UVs, B=Multiply(A=@n21, B=ScalarParameter() {group=Eye, parameterName=EyeDistortion_Intensity}) {constB=0.07})
  n31 = Constant2Vector() {g=0.5, r=0.5}
  n32 = Multiply(A=Time()) {constB=0.1}
  n33 = Sine(None=Multiply(A=SphereMask(A=@n30, B=@n31, Radius=@n32) {attenuationRadius=0.5, hardnessPercent=0}, B=@n32)) {desc=.4, period=0.2}
  n34 = SphereMask(A=@n26, B=@n27, Radius=ScalarParameter() {defaultValue=0.1, group=Eye, parameterName=EyeColor_CenterRadius}) {attenuationRadius=0.1, hardnessPercent=0}
  n35 = ScalarParameter() {defaultValue=3, group=Eye, parameterName=IrisMaskIntensity}
  n36 = LinearInterpolate(A=ConstantBiasScale(None=@n33), B=Saturate(None=Sine(None=@n34)), Alpha=Saturate(None=Multiply(A=@n34, B=@n35) {constB=5}))
  NamedRerouteDeclaration(None=LinearInterpolate(A=LinearInterpolate(A=@n7.RGB, B=@n8.RGB, Alpha=@n28), B=@n29.RGB, Alpha=Multiply(A=@n36, B=@n29.A)) {constA=0.18}) {name=Eyes, nodeColor=<rgba 1 0.2109 0 1>}
  Saturate(None=@n33)
 ]
}
