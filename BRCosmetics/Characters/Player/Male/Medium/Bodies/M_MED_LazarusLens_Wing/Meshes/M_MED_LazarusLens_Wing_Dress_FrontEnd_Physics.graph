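# Physics asset for the front-end variant of the LazarusLens Wing male body's dress, covering its flaps, sash and pants.
# Capsules on the dynamic cloth bones sit alongside pelvis, thigh and foot capsules, plus a wide floor-like root box.
# Flap and sash chains branch from the pelvis with tight twist limits, while pants hang from the thighs.
# Fully free constraints handle most anchors and cross-link back tips and front inner/outer strands.
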PhysicsAsset /BRCosmetics/Characters/Player/Male/Medium/Bodies/M_MED_LazarusLens_Wing/Meshes/M_MED_LazarusLens_Wing_Dress_FrontEnd_Physics
bodies {
 dyn_flap_bk_c [{shapeType=Capsule, length=25.5943, radius=6.7792, shapeName=dyn_flap_bk_c_capsule, center=<8.2909 0 6>, rotation=<rot 0 0 90>}]
 dyn_flap_bk_l_01 [{shapeType=Capsule, length=14.3244, radius=7.7775, shapeName=dyn_flap_bk_l_01_capsule, center=<8.2909 0 6>, rotation=<rot 0 0 90>}]
 dyn_flap_bk_l_02 [{shapeType=Capsule, length=27.5881, radius=5.4328, shapeName=dyn_flap_bk_l_02_capsule, center=<8.2909 0 6>, rotation=<rot 0 0 90>}]
 dyn_flap_bk_l_03 [{shapeType=Capsule, length=29.7735, radius=5.2453, shapeName=dyn_flap_bk_l_03_capsule, center=<8.2909 0 6>, rotation=<rot 0 0 90>}]
 dyn_flap_bk_r_01 [{shapeType=Capsule, length=14.2542, radius=7.8478, shapeName=dyn_flap_bk_r_01_capsule, center=<8.2909 0 6>, rotation=<rot 0 0 90>}]
 dyn_flap_bk_r_02 [{shapeType=Capsule, length=27.4206, radius=5.6003, shapeName=dyn_flap_bk_r_02_capsule, center=<8.2909 0 6>, rotation=<rot 0 0 90>}]
 dyn_flap_bk_r_03 [{shapeType=Capsule, length=29.7735, radius=5.2453, shapeName=dyn_flap_bk_r_03_capsule, center=<8.2909 0 6>, rotation=<rot 0 0 90>}]
 dyn_flap_fr_c [{shapeType=Capsule, length=0.2, radius=0.2, shapeName=dyn_flap_fr_c_capsule}]
 dyn_flap_fr_l_01 [{shapeType=Capsule, length=7.661, radius=11.6159, shapeName=dyn_flap_fr_l_01_capsule, center=<8.4264 0 14>, rotation=<rot 0 0 106.369>}]
 dyn_flap_fr_l_02_innter [{shapeType=Capsule, length=5.5924, radius=13.6845, shapeName=dyn_flap_fr_l_02_innter_capsule, center=<8.4264 0 16>, rotation=<rot 0 0 90>}]
 dyn_flap_fr_l_02_outer [{shapeType=Capsule, length=6.1511, radius=13.1259, shapeName=dyn_flap_fr_l_02_outer_capsule, center=<8.4264 0 16>, rotation=<rot 0 0 90>}]
 dyn_flap_fr_l_03_innter [{shapeType=Capsule, length=8.9809, radius=10.296, shapeName=dyn_flap_fr_l_03_innter_capsule, center=<8.4264 0 16>, rotation=<rot 0 0 90>}]
 dyn_flap_fr_l_03_outer [{shapeType=Capsule, length=6.1711, radius=13.1059, shapeName=dyn_flap_fr_l_03_outer_capsule, center=<8.4264 0 16>, rotation=<rot 0 0 90>}]
 dyn_flap_fr_l_base [{shapeType=Capsule, length=0.2, radius=0.2, shapeName=dyn_flap_fr_l_base_capsule}]
 dyn_flap_fr_mid_01 [{shapeType=Capsule, length=21.0007, radius=2.1085, shapeName=dyn_flap_fr_mid_02_capsule, center=<0 0 0.3238>, rotation=<rot 0 0 90>}]
 dyn_flap_fr_mid_02 [{shapeType=Capsule, length=21.0007, radius=2.1085, shapeName=dyn_flap_fr_mid_02_capsule, center=<0 0 0.3238>, rotation=<rot 0 0 90>}]
 dyn_flap_fr_mid_03 [{shapeType=Capsule, length=18.1094, radius=4.9998, shapeName=dyn_flap_fr_mid_02_capsule, center=<0 0 0.3238>, rotation=<rot 0 0 90>}]
 dyn_flap_fr_mid_04 [{shapeType=Capsule, length=21.0007, radius=2.1085, shapeName=dyn_flap_fr_mid_02_capsule, center=<0 0 0.3238>, rotation=<rot 0 0 90>}]
 dyn_flap_fr_r_01 [{shapeType=Capsule, length=7.7831, radius=11.4939, shapeName=dyn_flap_fr_r_01_capsule, center=<8.4264 0 14>, rotation=<rot 0 0 65.4196>}]
 dyn_flap_fr_r_02_innter [{shapeType=Capsule, length=5.5924, radius=13.6845, shapeName=dyn_flap_fr_r_02_innter_capsule, center=<8.4264 0 16>, rotation=<rot 0 0 90>}]
 dyn_flap_fr_r_02_outer [{shapeType=Capsule, length=6.1511, radius=13.1259, shapeName=dyn_flap_fr_r_02_outer_capsule, center=<8.4264 0 16>, rotation=<rot 0 0 90>}]
 dyn_flap_fr_r_03_innter [{shapeType=Capsule, length=8.9809, radius=10.296, shapeName=dyn_flap_fr_r_03_innter_capsule, center=<8.4264 0 16>, rotation=<rot 0 0 90>}]
 dyn_flap_fr_r_03_outer [{shapeType=Capsule, length=6.1711, radius=13.1059, shapeName=dyn_flap_fr_r_03_outer_capsule, center=<8.4264 0 16>, rotation=<rot 0 0 90>}]
 dyn_flap_fr_r_base [{shapeType=Capsule, length=0.1, radius=0.1, shapeName=dyn_flap_fr_r_base_capsule}]
 dyn_pants_l_01 [{shapeType=Capsule, length=18.5913, radius=13.0728, center=<20.0422 0 0>, rotation=<rot 90 0 0>}]
 dyn_pants_l_02 [{shapeType=Capsule, length=2.1847, radius=15.8045, center=<11.5026 0 0>, rotation=<rot 90 0 0>}]
 dyn_pants_r_01 [{shapeType=Capsule, length=18.8558, radius=12.2862, center=<20.2237 0 0>, rotation=<rot 90 0 0>}]
 dyn_pants_r_02 [{shapeType=Capsule, length=1.3849, radius=15.8045, center=<12.1873 0 0>, rotation=<rot 90 0 0>}]
 dyn_sash_bot_01 [{shapeType=Capsule, length=1.7883, radius=8.1028, shapeName=dyn_sash_bot_01_capsule, center=<6.4448 0 14>, rotation=<rot 90 0 0>}]
 dyn_sash_bot_02 [{shapeType=Capsule, length=1.6375, radius=8.6562, shapeName=dyn_sash_bot_02_capsule, center=<6.4448 0 14>, rotation=<rot 90 0 0>}]
 dyn_sash_bot_mstr [{shapeType=Capsule, length=1, radius=1, shapeName=dyn_sash_bot_mstr_capsule, center=<1.6568 0 0>, rotation=<rot 0 0 90>}]
 dyn_sash_top_02 [{shapeType=Capsule, length=2.8599, radius=1, shapeName=dyn_sash_top_02_capsule, rotation=<rot 0 0 90>}]
 foot_l [{
   shapeType=Capsule
   length=5.8887
   radius=10.7755
   shapeName=foot_l_capsule
   center=<-4.6682 3.0533 0.3949>
   rotation=<rot 4.0577 -1.7839 -85.3414>
  }]
 foot_r [{
   shapeType=Capsule
   length=5.8887
   radius=10.4473
   shapeName=foot_r_capsule
   center=<4.6933 -2.372 -0.4503>
   rotation=<rot 4.0577 -1.7839 -85.3414>
  }]
 pelvis [{shapeType=Capsule, length=3, radius=8, rotation=<rot 0 0 90>}]
 root [{shapeType=Box, extentX=262.3494, extentY=262.3494, extentZ=23.7236, shapeName=root_box, center=<0 0 -11.5>, rotation=<rot 0 -90 0>}]
 thigh_l [{shapeType=Capsule, length=23.8432, radius=11.2365, shapeName=thigh_l_capsule, center=<-21.709 0 0>, rotation=<rot 90 0 0>}]
 thigh_r [{shapeType=Capsule, length=24.4471, radius=11.2366, shapeName=thigh_r_capsule, center=<23.0246 0 0>, rotation=<rot 90 0 0>}]
}
constraints [
 {bone1Name=dyn_pants_l_02, bone2Name=dyn_pants_l_01, swing1LimitDegrees=45, swing2LimitDegrees=45, twistLimitDegrees=45}
 {bone1Name=dyn_pants_r_02, bone2Name=dyn_pants_r_01, swing1LimitDegrees=45, swing2LimitDegrees=45, twistLimitDegrees=45}
 {bone1Name=dyn_flap_bk_c, bone2Name=pelvis, swing1LimitDegrees=45, swing2LimitDegrees=45, twistLimitDegrees=5}
 {bone1Name=dyn_flap_bk_l_01, bone2Name=dyn_flap_bk_c, swing1LimitDegrees=45, swing2LimitDegrees=45, twistLimitDegrees=5}
 {bone1Name=dyn_flap_bk_r_02, bone2Name=dyn_flap_bk_r_01, swing1LimitDegrees=45, swing2LimitDegrees=45, twistLimitDegrees=5}
 {bone1Name=dyn_flap_bk_l_03, bone2Name=dyn_flap_bk_l_02, swing1LimitDegrees=45, swing2LimitDegrees=45, twistLimitDegrees=5}
 {bone1Name=dyn_flap_bk_l_02, bone2Name=dyn_flap_bk_l_01, swing1LimitDegrees=45, swing2LimitDegrees=45, twistLimitDegrees=5}
 {bone1Name=dyn_flap_bk_r_03, bone2Name=dyn_flap_bk_r_02, swing1LimitDegrees=45, swing2LimitDegrees=45, twistLimitDegrees=5}
 {bone1Name=dyn_flap_bk_r_01, bone2Name=dyn_flap_bk_c, swing1LimitDegrees=45, swing2LimitDegrees=45, twistLimitDegrees=5}
 {
  bone1Name=dyn_pants_l_01
  bone2Name=thigh_l
  swing1LimitDegrees=45
  swing1Motion=Free
  swing2LimitDegrees=45
  swing2Motion=Free
  twistLimitDegrees=45
  twistMotion=Free
 }
 {
  bone1Name=dyn_pants_r_01
  bone2Name=thigh_r
  swing1LimitDegrees=45
  swing1Motion=Free
  swing2LimitDegrees=45
  swing2Motion=Free
  twistLimitDegrees=45
  twistMotion=Free
 }
 {
  bone1Name=thigh_r
  bone2Name=pelvis
  swing1LimitDegrees=45
  swing1Motion=Free
  swing2LimitDegrees=45
  swing2Motion=Free
  twistLimitDegrees=45
  twistMotion=Free
 }
 {
  bone1Name=thigh_l
  bone2Name=pelvis
  swing1LimitDegrees=45
  swing1Motion=Free
  swing2LimitDegrees=45
  swing2Motion=Free
  twistLimitDegrees=45
  twistMotion=Free
 }
 {
  bone1Name=pelvis
  bone2Name=root
  swing1LimitDegrees=45
  swing1Motion=Free
  swing2LimitDegrees=45
  swing2Motion=Free
  twistLimitDegrees=45
  twistMotion=Free
 }
 {
  bone1Name=foot_r
  bone2Name=thigh_r
  swing1LimitDegrees=45
  swing1Motion=Free
  swing2LimitDegrees=45
  swing2Motion=Free
  twistLimitDegrees=45
  twistMotion=Free
 }
 {
  bone1Name=foot_l
  bone2Name=thigh_l
  swing1LimitDegrees=45
  swing1Motion=Free
  swing2LimitDegrees=45
  swing2Motion=Free
  twistLimitDegrees=45
  twistMotion=Free
 }
 {
  bone1Name=dyn_flap_bk_r_03
  bone2Name=dyn_flap_bk_l_03
  swing1LimitDegrees=45
  swing1Motion=Free
  swing2LimitDegrees=45
  swing2Motion=Free
  twistLimitDegrees=45
  twistMotion=Free
 }
 {
  bone1Name=dyn_sash_bot_mstr
  bone2Name=pelvis
  swing1LimitDegrees=45
  swing1Motion=Free
  swing2LimitDegrees=45
  swing2Motion=Free
  twistLimitDegrees=45
  twistMotion=Free
 }
 {bone1Name=dyn_sash_bot_01, bone2Name=dyn_sash_bot_mstr, swing1LimitDegrees=45, swing2LimitDegrees=45, twistLimitDegrees=15}
 {bone1Name=dyn_sash_bot_02, bone2Name=dyn_sash_bot_01, swing1LimitDegrees=45, swing2LimitDegrees=45, twistLimitDegrees=15}
 {
  bone1Name=dyn_flap_fr_mid_04
  bone2Name=dyn_flap_fr_mid_02
  swing1LimitDegrees=45
  swing1Motion=Free
  swing2LimitDegrees=45
  swing2Motion=Free
  twistLimitDegrees=45
  twistMotion=Free
 }
 {
  bone1Name=dyn_flap_fr_mid_02
  bone2Name=pelvis
  swing1LimitDegrees=45
  swing1Motion=Free
  swing2LimitDegrees=45
  swing2Motion=Free
  twistLimitDegrees=45
  twistMotion=Free
 }
 {
  bone1Name=dyn_flap_fr_mid_01
  bone2Name=pelvis
  swing1LimitDegrees=45
  swing1Motion=Free
  swing2LimitDegrees=45
  swing2Motion=Free
  twistLimitDegrees=45
  twistMotion=Free
 }
 {
  bone1Name=dyn_flap_fr_mid_03
  bone2Name=pelvis
  swing1LimitDegrees=45
  swing1Motion=Free
  swing2LimitDegrees=45
  swing2Motion=Free
  twistLimitDegrees=45
  twistMotion=Free
 }
 {
  bone1Name=dyn_flap_fr_c
  bone2Name=pelvis
  swing1LimitDegrees=45
  swing1Motion=Free
  swing2LimitDegrees=45
  swing2Motion=Free
  twistLimitDegrees=45
  twistMotion=Free
 }
 {
  bone1Name=dyn_flap_fr_l_base
  bone2Name=dyn_flap_fr_c
  swing1LimitDegrees=45
  swing1Motion=Free
  swing2LimitDegrees=45
  swing2Motion=Free
  twistLimitDegrees=45
  twistMotion=Free
 }
 {
  bone1Name=dyn_flap_fr_r_base
  bone2Name=dyn_flap_fr_c
  swing1LimitDegrees=45
  swing1Motion=Free
  swing2LimitDegrees=45
  swing2Motion=Free
  twistLimitDegrees=45
  twistMotion=Free
 }
 {bone1Name=dyn_flap_fr_l_01, bone2Name=dyn_flap_fr_l_base, swing1LimitDegrees=45, swing2LimitDegrees=45, twistLimitDegrees=5}
 {bone1Name=dyn_flap_fr_l_02_innter, bone2Name=dyn_flap_fr_l_01, swing1LimitDegrees=45, swing2LimitDegrees=45, twistLimitDegrees=5}
 {bone1Name=dyn_flap_fr_l_03_innter, bone2Name=dyn_flap_fr_l_02_innter, swing1LimitDegrees=45, swing2LimitDegrees=45, twistLimitDegrees=5}
 {bone1Name=dyn_flap_fr_l_02_outer, bone2Name=dyn_flap_fr_l_01, swing1LimitDegrees=45, swing2LimitDegrees=45, twistLimitDegrees=5}
 {bone1Name=dyn_flap_fr_l_03_outer, bone2Name=dyn_flap_fr_l_02_outer, swing1LimitDegrees=45, swing2LimitDegrees=45, twistLimitDegrees=5}
 {bone1Name=dyn_flap_fr_r_01, bone2Name=dyn_flap_fr_r_base, swing1LimitDegrees=45, swing2LimitDegrees=45, twistLimitDegrees=5}
 {bone1Name=dyn_flap_fr_r_02_innter, bone2Name=dyn_flap_fr_r_01, swing1LimitDegrees=45, swing2LimitDegrees=45, twistLimitDegrees=5}
 {bone1Name=dyn_flap_fr_r_03_innter, bone2Name=dyn_flap_fr_r_02_innter, swing1LimitDegrees=45, swing2LimitDegrees=45, twistLimitDegrees=5}
 {bone1Name=dyn_flap_fr_r_02_outer, bone2Name=dyn_flap_fr_r_01, swing1LimitDegrees=45, swing2LimitDegrees=45, twistLimitDegrees=5}
 {bone1Name=dyn_flap_fr_r_03_outer, bone2Name=dyn_flap_fr_r_02_outer, swing1LimitDegrees=45, swing2LimitDegrees=45, twistLimitDegrees=5}
 {
  bone1Name=dyn_flap_fr_l_02_outer
  bone2Name=dyn_flap_fr_l_02_innter
  swing1LimitDegrees=45
  swing1Motion=Free
  swing2LimitDegrees=45
  swing2Motion=Free
  twistLimitDegrees=45
  twistMotion=Free
 }
 {
  bone1Name=dyn_flap_fr_r_02_outer
  bone2Name=dyn_flap_fr_r_02_innter
  swing1LimitDegrees=45
  swing1Motion=Free
  swing2LimitDegrees=45
  swing2Motion=Free
  twistLimitDegrees=45
  twistMotion=Free
 }
]
