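# Physics asset for the LazarusLens Wing medium male player body, defining collision and joints for its dangling flaps, pants and sash.
# Capsules on pelvis, thighs and feet cover the lower body, with a wide, flat root box at ground level.
# Back flaps branch from a pelvis-linked center into limited left and right chains whose tips share a free constraint.
# Lower sash and pants joints are limited; front flap constraints leave rotation free.
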PhysicsAsset /BRCosmetics/Characters/Player/Male/Medium/Bodies/M_MED_LazarusLens_Wing/Meshes/M_MED_LazarusLens_Wing_Physics
bodies {
 dyn_flap_bk_c [
  {shapeType=Capsule, length=27.2951, radius=5.0784, center=<6.5752 0 3>, rotation=<rot 0 0 90>}
  {shapeType=Capsule, length=27.0663, radius=5.3072, shapeName=dyn_flap_bk_c_capsule, center=<16.5752 0 3>, rotation=<rot 0 0 90>}
 ]
 dyn_flap_bk_l_01 [
  {shapeType=Capsule, length=16.9577, radius=5.1443, center=<6.7177 0 3>, rotation=<rot 0 0 90>}
  {shapeType=Capsule, length=16.9577, radius=5.1443, shapeName=dyn_flap_bk_l_01_capsule, center=<14.1486 0 3>, rotation=<rot 0 0 90>}
 ]
 dyn_flap_bk_l_02 [
  {shapeType=Capsule, length=27.0535, radius=4.7679, center=<5.6767 0 3>, rotation=<rot 0 0 90>}
  {shapeType=Capsule, length=28.253, radius=4.7679, shapeName=dyn_flap_bk_l_02_capsule, center=<13.8409 0 3>, rotation=<rot 0 0 90>}
 ]
 dyn_flap_bk_l_03 [
  {shapeType=Capsule, length=28.6329, radius=5.1864, center=<10.7479 0 3>, rotation=<rot 0 0 90>}
  {shapeType=Capsule, length=29.8323, radius=5.1864, shapeName=dyn_flap_bk_l_03_capsule, center=<4.5856 0 3>, rotation=<rot 0 0 90>}
 ]
 dyn_flap_bk_r_01 [
  {shapeType=Capsule, length=16.8874, radius=5.2146, center=<7.2962 0 3>, rotation=<rot 0 0 90>}
  {shapeType=Capsule, length=16.8874, radius=5.2146, shapeName=dyn_flap_bk_r_01_capsule, center=<14.2585 0 3>, rotation=<rot 0 0 90>}
 ]
 dyn_flap_bk_r_02 [
  {shapeType=Capsule, length=26.886, radius=4.9354, center=<5.696 0 3>, rotation=<rot 0 0 90>}
  {shapeType=Capsule, length=28.0855, radius=4.9354, shapeName=dyn_flap_bk_r_02_capsule, center=<13.8409 0 3>, rotation=<rot 0 0 90>}
 ]
 dyn_flap_bk_r_03 [
  {shapeType=Capsule, length=28.6329, radius=5.1864, center=<11.347 0 3>, rotation=<rot 0 0 90>}
  {shapeType=Capsule, length=29.8323, radius=5.1864, shapeName=dyn_flap_bk_r_03_capsule, center=<4.7588 0 3>, rotation=<rot 0 0 90>}
 ]
 dyn_flap_fr_mid_01 [
  {shapeType=Capsule, length=21.0007, radius=2.1085, shapeName=dyn_flap_fr_mid_02_capsule, center=<0 0 0.3238>, rotation=<rot 0 0 90>}
  {shapeType=Capsule, length=21.0007, radius=2.1085, shapeName=dyn_flap_fr_mid_02_capsule, center=<2.7544 0 0.3238>, rotation=<rot 0 0 90>}
  {shapeType=Capsule, length=21.0007, radius=2.1085, shapeName=dyn_flap_fr_mid_02_capsule, center=<6.0098 0 0.3238>, rotation=<rot 0 0 90>}
  {shapeType=Capsule, length=21.0007, radius=2.1085, shapeName=dyn_flap_fr_mid_02_capsule, center=<10.3121 0 0.3238>, rotation=<rot 0 0 90>}
  {shapeType=Capsule, length=17.3737, radius=5.7355, shapeName=dyn_flap_fr_mid_02_capsule, center=<5.8582 0 -4.6743>, rotation=<rot 0 0 90>}
 ]
 dyn_flap_fr_mid_02 [
  {shapeType=Capsule, length=21.0007, radius=2.1085, shapeName=dyn_flap_fr_mid_02_capsule, center=<0 0 0.3238>, rotation=<rot 0 0 90>}
  {shapeType=Capsule, length=21.0007, radius=2.1085, shapeName=dyn_flap_fr_mid_02_capsule, center=<2.7544 0 0.3238>, rotation=<rot 0 0 90>}
  {shapeType=Capsule, length=21.0007, radius=2.1085, shapeName=dyn_flap_fr_mid_02_capsule, center=<6.0098 0 0.3238>, rotation=<rot 0 0 90>}
  {shapeType=Capsule, length=21.0007, radius=2.1085, shapeName=dyn_flap_fr_mid_02_capsule, center=<10.3121 0 0.3238>, rotation=<rot 0 0 90>}
  {shapeType=Capsule, length=17.3737, radius=5.7355, shapeName=dyn_flap_fr_mid_02_capsule, center=<5.8582 0 -4.6743>, rotation=<rot 0 0 90>}
 ]
 dyn_flap_fr_mid_03 [
  {shapeType=Capsule, length=21.0007, radius=2.1085, shapeName=dyn_flap_fr_mid_02_capsule, center=<0 0 0.3238>, rotation=<rot 0 0 90>}
  {shapeType=Capsule, length=21.0007, radius=2.1085, shapeName=dyn_flap_fr_mid_02_capsule, center=<2.7544 0 0.3238>, rotation=<rot 0 0 90>}
  {shapeType=Capsule, length=21.0007, radius=2.1085, shapeName=dyn_flap_fr_mid_02_capsule, center=<6.0098 0 0.3238>, rotation=<rot 0 0 90>}
  {shapeType=Capsule, length=21.0007, radius=2.1085, shapeName=dyn_flap_fr_mid_02_capsule, center=<10.3121 0 0.3238>, rotation=<rot 0 0 90>}
  {shapeType=Capsule, length=17.3737, radius=5.7355, shapeName=dyn_flap_fr_mid_02_capsule, center=<5.8582 0 -4.6743>, rotation=<rot 0 0 90>}
 ]
 dyn_flap_fr_mid_04 [
  {shapeType=Capsule, length=21.0007, radius=2.1085, shapeName=dyn_flap_fr_mid_02_capsule, center=<0 0 0.3238>, rotation=<rot 0 0 90>}
  {shapeType=Capsule, length=21.0007, radius=2.1085, shapeName=dyn_flap_fr_mid_02_capsule, center=<2.7544 0 0.3238>, rotation=<rot 0 0 90>}
  {shapeType=Capsule, length=21.0007, radius=2.1085, shapeName=dyn_flap_fr_mid_02_capsule, center=<6.0098 0 0.3238>, rotation=<rot 0 0 90>}
  {shapeType=Capsule, length=21.0007, radius=2.1085, shapeName=dyn_flap_fr_mid_02_capsule, center=<10.3121 0 0.3238>, rotation=<rot 0 0 90>}
  {shapeType=Capsule, length=17.3737, radius=5.7355, shapeName=dyn_flap_fr_mid_02_capsule, center=<5.8582 0 -4.6743>, rotation=<rot 0 0 90>}
 ]
 dyn_pants_l_01 [{shapeType=Capsule, length=13.1294, radius=13.0728, center=<15.6222 0 0>, rotation=<rot 90 0 0>}]
 dyn_pants_l_02 [{shapeType=Capsule, length=2.1847, radius=15.8045, center=<11.5026 0 0>, rotation=<rot 90 0 0>}]
 dyn_pants_r_01 [{shapeType=Capsule, length=13.916, radius=12.2862, center=<15.6222 0 0>, rotation=<rot 90 0 0>}]
 dyn_pants_r_02 [{shapeType=Capsule, length=1.3849, radius=15.8045, center=<12.1873 0 0>, rotation=<rot 90 0 0>}]
 dyn_sash_bot_01 [
  {shapeType=Capsule, length=6.9129, radius=2.2502, shapeName=dyn_sash_bot_01_capsule, center=<2.2992 0 4>, rotation=<rot 0 0 90>}
  {shapeType=Capsule, length=6.9129, radius=2.2502, shapeName=dyn_sash_bot_01_capsule, center=<9.2896 0 4>, rotation=<rot 0 0 90>}
 ]
 dyn_sash_bot_02 [
  {shapeType=Capsule, length=7.3956, radius=2.1701, shapeName=dyn_sash_bot_02_capsule, center=<2.2992 0 5>, rotation=<rot 0.0064 0 90>}
  {shapeType=Capsule, length=7.3956, radius=2.1701, shapeName=dyn_sash_bot_02_capsule, center=<8.8541 0 5>, rotation=<rot 0.0064 0 90>}
 ]
 dyn_sash_bot_mstr [
  {shapeType=Capsule, length=1, radius=1, shapeName=dyn_sash_bot_mstr_capsule, center=<1.6568 0 0>, rotation=<rot 0 0 90>}
  {shapeType=Capsule, length=1, radius=1, shapeName=dyn_sash_bot_mstr_capsule, center=<11.6568 0 0>, rotation=<rot 0 0 90>}
 ]
 dyn_sash_top_02 [{shapeType=Capsule, length=2.8599, radius=1, shapeName=dyn_sash_top_02_capsule, rotation=<rot 0 0 90>}]
 foot_l [{
   shapeType=Capsule
   length=5.8887
   radius=10.7755
   shapeName=foot_l_capsule
   center=<-4.6682 3.0533 0.3949>
   rotation=<rot 4.0577 -1.7839 -85.3414>
  }]
 foot_r [{
   shapeType=Capsule
   length=5.8887
   radius=10.4473
   shapeName=foot_r_capsule
   center=<4.6933 -2.372 -0.4503>
   rotation=<rot 4.0577 -1.7839 -85.3414>
  }]
 pelvis [{shapeType=Capsule, length=3, radius=8, rotation=<rot 0 0 90>}]
 root [{shapeType=Box, extentX=262.3494, extentY=262.3494, extentZ=23.7236, shapeName=root_box, center=<0 0 -11.5>, rotation=<rot 0 -90 0>}]
 thigh_l [{shapeType=Capsule, length=27.5049, radius=10.8288, shapeName=thigh_l_capsule, center=<-24.0722 0 0>, rotation=<rot 90 0 0>}]
 thigh_r [{shapeType=Capsule, length=28.1087, radius=10.8289, shapeName=thigh_r_capsule, center=<24.0294 0 0>, rotation=<rot 90 0 0>}]
}
constraints [
 {bone1Name=dyn_pants_l_02, bone2Name=dyn_pants_l_01, swing1LimitDegrees=45, swing2LimitDegrees=45, twistLimitDegrees=45}
 {bone1Name=dyn_pants_r_02, bone2Name=dyn_pants_r_01, swing1LimitDegrees=45, swing2LimitDegrees=45, twistLimitDegrees=45}
 {bone1Name=dyn_flap_bk_c, bone2Name=pelvis, swing1LimitDegrees=3, swing2LimitDegrees=45, twistLimitDegrees=3}
 {bone1Name=dyn_flap_bk_l_01, bone2Name=dyn_flap_bk_c, swing1LimitDegrees=20, swing2LimitDegrees=45, twistLimitDegrees=15}
 {bone1Name=dyn_flap_bk_r_02, bone2Name=dyn_flap_bk_r_01, swing1LimitDegrees=20, swing2LimitDegrees=30, twistLimitDegrees=15}
 {bone1Name=dyn_flap_bk_l_03, bone2Name=dyn_flap_bk_l_02, swing1LimitDegrees=20, swing2LimitDegrees=30, twistLimitDegrees=15}
 {bone1Name=dyn_flap_bk_l_02, bone2Name=dyn_flap_bk_l_01, swing1LimitDegrees=20, swing2LimitDegrees=30, twistLimitDegrees=15}
 {bone1Name=dyn_flap_bk_r_03, bone2Name=dyn_flap_bk_r_02, swing1LimitDegrees=20, swing2LimitDegrees=30, twistLimitDegrees=15}
 {bone1Name=dyn_flap_bk_r_01, bone2Name=dyn_flap_bk_c, swing1LimitDegrees=20, swing2LimitDegrees=30, twistLimitDegrees=15}
 {
  bone1Name=dyn_pants_l_01
  bone2Name=thigh_l
  swing1LimitDegrees=45
  swing1Motion=Free
  swing2LimitDegrees=45
  swing2Motion=Free
  twistLimitDegrees=45
  twistMotion=Free
 }
 {
  bone1Name=dyn_pants_r_01
  bone2Name=thigh_r
  swing1LimitDegrees=45
  swing1Motion=Free
  swing2LimitDegrees=45
  swing2Motion=Free
  twistLimitDegrees=45
  twistMotion=Free
 }
 {
  bone1Name=thigh_r
  bone2Name=pelvis
  swing1LimitDegrees=45
  swing1Motion=Free
  swing2LimitDegrees=45
  swing2Motion=Free
  twistLimitDegrees=45
  twistMotion=Free
 }
 {
  bone1Name=thigh_l
  bone2Name=pelvis
  swing1LimitDegrees=45
  swing1Motion=Free
  swing2LimitDegrees=45
  swing2Motion=Free
  twistLimitDegrees=45
  twistMotion=Free
 }
 {
  bone1Name=pelvis
  bone2Name=root
  swing1LimitDegrees=45
  swing1Motion=Free
  swing2LimitDegrees=45
  swing2Motion=Free
  twistLimitDegrees=45
  twistMotion=Free
 }
 {
  bone1Name=foot_r
  bone2Name=thigh_r
  swing1LimitDegrees=45
  swing1Motion=Free
  swing2LimitDegrees=45
  swing2Motion=Free
  twistLimitDegrees=45
  twistMotion=Free
 }
 {
  bone1Name=foot_l
  bone2Name=thigh_l
  swing1LimitDegrees=45
  swing1Motion=Free
  swing2LimitDegrees=45
  swing2Motion=Free
  twistLimitDegrees=45
  twistMotion=Free
 }
 {
  bone1Name=dyn_flap_bk_r_03
  bone2Name=dyn_flap_bk_l_03
  swing1LimitDegrees=45
  swing1Motion=Free
  swing2LimitDegrees=45
  swing2Motion=Free
  twistLimitDegrees=45
  twistMotion=Free
 }
 {
  bone1Name=dyn_sash_bot_mstr
  bone2Name=pelvis
  swing1LimitDegrees=45
  swing1Motion=Free
  swing2LimitDegrees=45
  swing2Motion=Free
  twistLimitDegrees=45
  twistMotion=Free
 }
 {bone1Name=dyn_sash_bot_01, bone2Name=dyn_sash_bot_mstr, swing1LimitDegrees=20, swing2LimitDegrees=90, twistLimitDegrees=20}
 {bone1Name=dyn_sash_bot_02, bone2Name=dyn_sash_bot_01, swing1LimitDegrees=20, swing2LimitDegrees=90, twistLimitDegrees=20}
 {
  bone1Name=dyn_flap_fr_mid_04
  bone2Name=dyn_flap_fr_mid_02
  swing1LimitDegrees=45
  swing1Motion=Free
  swing2LimitDegrees=45
  swing2Motion=Free
  twistLimitDegrees=45
  twistMotion=Free
 }
 {
  bone1Name=dyn_flap_fr_mid_02
  bone2Name=pelvis
  swing1LimitDegrees=45
  swing1Motion=Free
  swing2LimitDegrees=45
  swing2Motion=Free
  twistLimitDegrees=45
  twistMotion=Free
 }
 {
  bone1Name=dyn_flap_fr_mid_01
  bone2Name=pelvis
  swing1LimitDegrees=45
  swing1Motion=Free
  swing2LimitDegrees=45
  swing2Motion=Free
  twistLimitDegrees=45
  twistMotion=Free
 }
 {
  bone1Name=dyn_flap_fr_mid_03
  bone2Name=pelvis
  swing1LimitDegrees=45
  swing1Motion=Free
  swing2LimitDegrees=45
  swing2Motion=Free
  twistLimitDegrees=45
  twistMotion=Free
 }
]
