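# Post-copy AnimBP adding secondary motion to the flaps, pants and sash of the medium male LazarusLens Wing cosmetic body.
# It copies the parent mesh pose, runs a deform Control Rig, then layers bone constraints and Anim Dynamics.
# RigidBody simulation from cached PreRBAN switches between front-end and gameplay setups, followed by corrective offsets, including a sliding toggle.
# RBAN_off toggles trail controllers, weighted by backward movement and emotes, before final sash constraints.
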AnimBlueprint /BRCosmetics/Characters/Player/Male/Medium/Bodies/M_MED_LazarusLens_Wing/Meshes/M_MED_LazarusLens_Wing_AnimBP
target_skeleton=&/M_MED_LazarusLens_Wing_Skeleton
graphs [
 AnimGraph {nodes [
   n1 = Get crouch move alpha() -> (crouch move alpha:double)
   "Transform (Modify) Bone - Bone: dyn_sash_bot_02"(Alpha=@n1, Translation="0,0,0", Rotation="-5,0,0", Scale="1,1,1") {
    rotationMode=BMM_Additive
    rotationSpace=BCS_BoneSpace
    translationMode=BMM_Additive
    translationSpace=BCS_BoneSpace
    alphaBoolBlend {blendOption=Linear}
    boneToModify {boneName=dyn_sash_bot_02}
    rotation {pitch=-5}
   }
   "Transform (Modify) Bone - Bone: dyn_sash_bot_01"(Alpha=@n1, Translation="0,0,2", Rotation="10,0,0", Scale="1,1,1") {
    rotationMode=BMM_Additive
    rotationSpace=BCS_BoneSpace
    translationMode=BMM_Additive
    translationSpace=BCS_BoneSpace
    alphaBoolBlend {blendOption=Linear}
    boneToModify {boneName=dyn_sash_bot_01}
    rotation {pitch=10}
    translation {z=2}
   }
   n2 = Get BwdMovement() -> (BwdMovement:float)
   n3 = Trail controller - Bone: dyn_flap_bk_l_02!(ComponentPose=Local To Component(LocalPose=Use cached pose 'Modified Pose'()), Alpha=@n2, FakeVelocity:Vector="0,0,0") -> (Pose) {
    bActorSpaceFakeVel=true
    bLimitRotation=true
    bShowTrailLocation=true
    debugLifeTime=0.1
    maxDeltaTime=0.0333
    relaxationSpeedScale=3
    alphaScaleBiasClamp {bInterpResult=true}
    baseJoint {boneName=root}
    planarLimits [
     {drivingBone {boneName=root}, planeTransform {location=<0 0 10>, rotation=<rot 0 0 0>, scale=<1 1 1>}}
     {drivingBone {boneName=spine_03}, planeTransform {location=<0 -20 0>, rotation=<rot 0 15 -89.9999>, scale=<1 1 1>}}
    ]
    rotationLimits [{limitMax=<10 0 45>, limitMin=<-10 -45 -45>}, {limitMax=<10 0 45>, limitMin=<-10 -45 -45>}]
    rotationOffsets [<0 0 0>, <0 0 0>]
    trailBone {boneName=dyn_flap_bk_l_02}
    trailRelaxationSpeed {editorCurveData {keys [{arriveTangent=-72.0739, interpMode=RCIM_Cubic, leaveTangent=-72.0739, tangentMode=RCTM_User, time=0, value=35}, <at 1 5 RCIM_Cubic>]}}
   }
   n4 = Trail controller - Bone: dyn_flap_bk_r_02!(ComponentPose=@n3, Alpha=@n2, FakeVelocity:Vector="0,0,0") -> (Pose) {
    bActorSpaceFakeVel=true
    bLimitRotation=true
    bShowTrailLocation=true
    debugLifeTime=0.1
    maxDeltaTime=0.0333
    relaxationSpeedScale=3
    alphaScaleBiasClamp {bInterpResult=true}
    baseJoint {boneName=root}
    planarLimits [
     {drivingBone {boneName=root}, planeTransform {location=<0 0 10>, rotation=<rot 0 0 0>, scale=<1 1 1>}}
     {drivingBone {boneName=spine_03}, planeTransform {location=<0 -20 0>, rotation=<rot 0 15 -89.9999>, scale=<1 1 1>}}
    ]
    rotationLimits [{limitMax=<10 0 45>, limitMin=<-10 -45 -45>}, {limitMax=<10 0 45>, limitMin=<-10 -45 -45>}]
    rotationOffsets [<0 0 0>, <0 0 0>]
    trailBone {boneName=dyn_flap_bk_r_02}
    trailRelaxationSpeed {editorCurveData {keys [{arriveTangent=-72.0739, interpMode=RCIM_Cubic, leaveTangent=-72.0739, tangentMode=RCTM_User, time=0, value=35}, <at 1 5 RCIM_Cubic>]}}
   }
   n5 = Get Is_Custom_Emote_Playing() -> (Is_Custom_Emote_Playing:double)
   n6 = Trail controller - Bone: dyn_flap_fr_l_03_innter!(ComponentPose=@n4, Alpha=@n5, FakeVelocity:Vector="0,0,0") -> (Pose) {
    bActorSpaceFakeVel=true
    bLimitRotation=true
    bShowTrailLocation=true
    chainLength=3
    debugLifeTime=0.1
    maxDeltaTime=0.0333
    relaxationSpeedScale=0.8
    alphaScaleBiasClamp {bInterpResult=true}
    baseJoint {boneName=pelvis}
    planarLimits [
     {drivingBone {boneName=root}, planeTransform {location=<0 0 10>, rotation=<rot 0 0 0>, scale=<1 1 1>}}
     {drivingBone {boneName=spine_03}, planeTransform {location=<0 -20 0>, rotation=<rot 0 15 -89.9999>, scale=<1 1 1>}}
    ]
    rotationLimits [{limitMax=<10 0 45>, limitMin=<-10 -45 -45>}, {limitMax=<10 0 45>, limitMin=<-10 -45 -45>}, {limitMax=<-10 0 20>, limitMin=<-10 -45 -20>}]
    rotationOffsets [<0 0 0>, <0 0 0>, <0 0 0>]
    trailBone {boneName=dyn_flap_fr_l_03_innter}
    trailRelaxationSpeed {editorCurveData {keys [{arriveTangent=-72.0739, interpMode=RCIM_Cubic, leaveTangent=-72.0739, tangentMode=RCTM_User, time=0, value=35}, <at 1 5 RCIM_Cubic>]}}
   }
   n7 = Trail controller - Bone: dyn_flap_fr_l_03_outer!(ComponentPose=@n6, Alpha=@n5, FakeVelocity:Vector="0,0,0") -> (Pose) {
    bActorSpaceFakeVel=true
    bLimitRotation=true
    bShowTrailLocation=true
    chainLength=3
    debugLifeTime=0.1
    maxDeltaTime=0.0333
    relaxationSpeedScale=0.8
    alphaScaleBiasClamp {bInterpResult=true}
    baseJoint {boneName=pelvis}
    planarLimits [
     {drivingBone {boneName=root}, planeTransform {location=<0 0 10>, rotation=<rot 0 0 0>, scale=<1 1 1>}}
     {drivingBone {boneName=spine_03}, planeTransform {location=<0 -20 0>, rotation=<rot 0 15 -89.9999>, scale=<1 1 1>}}
    ]
    rotationLimits [{limitMax=<10 0 45>, limitMin=<-10 -45 -45>}, {limitMax=<10 0 45>, limitMin=<-10 -45 -45>}, {limitMax=<-10 0 20>, limitMin=<-10 -45 -20>}]
    rotationOffsets [<0 0 0>, <0 0 0>, <0 0 0>]
    trailBone {boneName=dyn_flap_fr_l_03_outer}
    trailRelaxationSpeed {editorCurveData {keys [{arriveTangent=-72.0739, interpMode=RCIM_Cubic, leaveTangent=-72.0739, tangentMode=RCTM_User, time=0, value=35}, <at 1 5 RCIM_Cubic>]}}
   }
   n8 = Trail controller - Bone: dyn_flap_fr_r_03_innter!(ComponentPose=@n7, Alpha=@n5, FakeVelocity:Vector="0,0,0") -> (Pose) {
    bActorSpaceFakeVel=true
    bLimitRotation=true
    bShowTrailLocation=true
    chainLength=3
    debugLifeTime=0.1
    maxDeltaTime=0.0333
    relaxationSpeedScale=0.8
    alphaScaleBiasClamp {bInterpResult=true}
    baseJoint {boneName=pelvis}
    planarLimits [
     {drivingBone {boneName=root}, planeTransform {location=<0 0 10>, rotation=<rot 0 0 0>, scale=<1 1 1>}}
     {drivingBone {boneName=spine_03}, planeTransform {location=<0 -20 0>, rotation=<rot 0 15 -89.9999>, scale=<1 1 1>}}
    ]
    rotationLimits [{limitMax=<10 0 45>, limitMin=<-10 -45 -45>}, {limitMax=<10 0 45>, limitMin=<-10 -45 -45>}, {limitMax=<-10 0 20>, limitMin=<-10 -45 -20>}]
    rotationOffsets [<0 0 0>, <0 0 0>, <0 0 0>]
    trailBone {boneName=dyn_flap_fr_r_03_innter}
    trailRelaxationSpeed {editorCurveData {keys [{arriveTangent=-72.0739, interpMode=RCIM_Cubic, leaveTangent=-72.0739, tangentMode=RCTM_User, time=0, value=35}, <at 1 5 RCIM_Cubic>]}}
   }
   n9 = Trail controller - Bone: dyn_flap_fr_r_03_outer!(ComponentPose=@n8, Alpha=@n5, FakeVelocity:Vector="0,0,0") -> (Pose) {
    bActorSpaceFakeVel=true
    bLimitRotation=true
    bShowTrailLocation=true
    chainLength=3
    debugLifeTime=0.1
    maxDeltaTime=0.0333
    relaxationSpeedScale=0.8
    alphaScaleBiasClamp {bInterpResult=true}
    baseJoint {boneName=pelvis}
    planarLimits [
     {drivingBone {boneName=root}, planeTransform {location=<0 0 10>, rotation=<rot 0 0 0>, scale=<1 1 1>}}
     {drivingBone {boneName=spine_03}, planeTransform {location=<0 -20 0>, rotation=<rot 0 15 -89.9999>, scale=<1 1 1>}}
    ]
    rotationLimits [{limitMax=<10 0 45>, limitMin=<-10 -45 -45>}, {limitMax=<10 0 45>, limitMin=<-10 -45 -45>}, {limitMax=<-10 0 20>, limitMin=<-10 -45 -20>}]
    rotationOffsets [<0 0 0>, <0 0 0>, <0 0 0>]
    trailBone {boneName=dyn_flap_fr_r_03_outer}
    trailRelaxationSpeed {editorCurveData {keys [{arriveTangent=-72.0739, interpMode=RCIM_Cubic, leaveTangent=-72.0739, tangentMode=RCTM_User, time=0, value=35}, <at 1 5 RCIM_Cubic>]}}
   }
   n10 = "Constraint - dyn_sash_bot_01 (PP)"!(ComponentPose=@n9, Alpha=1, ConstraintWeights_0=0.75, ConstraintWeights_1:float=0.25) -> (Pose) {
    alphaBoolBlend {blendOption=Linear}
    boneToModify {boneName=dyn_sash_bot_01}
    constraintSetup [{targetBone {boneName=dyn_flap_fr_mid_01}}, {targetBone {boneName=dyn_flap_fr_mid_02}}]
    constraintWeights [1,1]
   }
   n11 = "Constraint - dyn_sash_bot_01 (R)"(ComponentPose=@n10, Alpha=1, ConstraintWeights_0=1) {
    alphaBoolBlend {blendOption=Linear}
    boneToModify {boneName=dyn_sash_bot_01}
    constraintSetup [{transformType=Rotation, targetBone {boneName=dyn_flap_fr_mid_02}}]
    constraintWeights [1]
   }
   n12 = "Constraint - dyn_sash_bot_02 (PP)"!(ComponentPose=@n11, Alpha=1, ConstraintWeights_0=0.25, ConstraintWeights_1:float=0.75) -> (Pose) {
    alphaBoolBlend {blendOption=Linear}
    boneToModify {boneName=dyn_sash_bot_02}
    constraintSetup [{targetBone {boneName=dyn_flap_fr_mid_01}}, {targetBone {boneName=dyn_flap_fr_mid_02}}]
    constraintWeights [1,1]
   }
   n13 = "Constraint - dyn_sash_bot_02 (R)"(ComponentPose=@n12, Alpha=1, ConstraintWeights_0=1) {
    alphaBoolBlend {blendOption=Linear}
    boneToModify {boneName=dyn_sash_bot_02}
    constraintSetup [{transformType=Rotation, targetBone {boneName=dyn_flap_fr_mid_03}}]
    constraintWeights [1]
   }
   n14 = "Transform (Modify) Bone - Bone: dyn_sash_bot_01"(ComponentPose=@n13, Alpha=1, Translation="0,0,0", Rotation="5,0,0", Scale="1,1,1") {
    rotationMode=BMM_Additive
    rotationSpace=BCS_BoneSpace
    translationMode=BMM_Additive
    translationSpace=BCS_BoneSpace
    alphaBoolBlend {blendOption=Linear}
    boneToModify {boneName=dyn_sash_bot_01}
    rotation {pitch=5}
   }
   n15 = "Constraint - dyn_sash_top_02 (R)"(ComponentPose=@n14, Alpha=1, ConstraintWeights_0=1) {
    alphaBoolBlend {blendOption=Linear}
    boneToModify {boneName=dyn_sash_top_02}
    constraintSetup [{transformType=Rotation, targetBone {boneName=dyn_sash_bot_01}}]
    constraintWeights [1]
   }
   n16 = "Transform (Modify) Bone - Bone: dyn_sash_top_02"(ComponentPose=@n15, Alpha=1, Translation="0,0,0", Rotation="0,0,0", Scale="1,1,1") {
    rotationMode=BMM_Additive
    rotationSpace=BCS_BoneSpace
    translationMode=BMM_Additive
    translationSpace=BCS_BoneSpace
    alphaBoolBlend {blendOption=Linear}
    boneToModify {boneName=dyn_sash_top_02}
   }
   n17 = Component To Local(ComponentPose=@n16)
   n18 = Use cached pose 'Modified Pose'()
   n19 = Get RBAN_off() -> (RBAN_off:bool)
   n20 = "Constraint - dyn_sash_bot_01 (PP)"(
    ComponentPose=Local To Component(LocalPose=Blend Poses by bool(BlendPose_0=@n17, BlendPose_1=@n18, bActiveValue=@n19) {blendType=Linear})
    Alpha=1
    ConstraintWeights_0=0.75
    ConstraintWeights_1=0.25
   ) {
    alphaBoolBlend {blendOption=Linear}
    boneToModify {boneName=dyn_sash_bot_01}
    constraintSetup [{targetBone {boneName=dyn_flap_fr_mid_01}}, {targetBone {boneName=dyn_flap_fr_mid_02}}]
    constraintWeights [1,1]
   }
   n21 = "Constraint - dyn_sash_bot_01 (R)"(ComponentPose=@n20, Alpha=1, ConstraintWeights_0=1) {
    alphaBoolBlend {blendOption=Linear}
    boneToModify {boneName=dyn_sash_bot_01}
    constraintSetup [{transformType=Rotation, targetBone {boneName=dyn_flap_fr_mid_02}}]
    constraintWeights [1]
   }
   n22 = "Constraint - dyn_sash_bot_02 (PP)"(ComponentPose=@n21, Alpha=1, ConstraintWeights_0=0.25, ConstraintWeights_1=0.75) {
    alphaBoolBlend {blendOption=Linear}
    boneToModify {boneName=dyn_sash_bot_02}
    constraintSetup [{targetBone {boneName=dyn_flap_fr_mid_01}}, {targetBone {boneName=dyn_flap_fr_mid_02}}]
    constraintWeights [1,1]
   }
   n23 = "Constraint - dyn_sash_bot_02 (R)"(ComponentPose=@n22, Alpha=1, ConstraintWeights_0=1) {
    alphaBoolBlend {blendOption=Linear}
    boneToModify {boneName=dyn_sash_bot_02}
    constraintSetup [{transformType=Rotation, targetBone {boneName=dyn_flap_fr_mid_03}}]
    constraintWeights [1]
   }
   n24 = "Transform (Modify) Bone - Bone: dyn_sash_bot_01"(ComponentPose=@n23, Alpha=1, Translation="0,0,0", Rotation="5,0,0", Scale="1,1,1") {
    rotationMode=BMM_Additive
    rotationSpace=BCS_BoneSpace
    translationMode=BMM_Additive
    translationSpace=BCS_BoneSpace
    alphaBoolBlend {blendOption=Linear}
    boneToModify {boneName=dyn_sash_bot_01}
    rotation {pitch=5}
   }
   n25 = "Constraint - dyn_sash_top_02 (R)"(ComponentPose=@n24, Alpha=1, ConstraintWeights_0=1) {
    alphaBoolBlend {blendOption=Linear}
    boneToModify {boneName=dyn_sash_top_02}
    constraintSetup [{transformType=Rotation, targetBone {boneName=dyn_sash_bot_01}}]
    constraintWeights [1]
   }
   n26 = "Transform (Modify) Bone - Bone: dyn_sash_top_02"(ComponentPose=@n25, Alpha=1, Translation="0,0,0", Rotation="0,0,0", Scale="1,1,1") {
    rotationMode=BMM_Additive
    rotationSpace=BCS_BoneSpace
    translationMode=BMM_Additive
    translationSpace=BCS_BoneSpace
    alphaBoolBlend {blendOption=Linear}
    boneToModify {boneName=dyn_sash_top_02}
   }
   Output Pose(Result=Component To Local(ComponentPose=@n26)) {layerGroup=None}
   n27 = RigidBody!(
    ComponentPose=Local To Component(LocalPose=Use cached pose 'PreRBAN'())
    Alpha
    OverrideWorldGravity:Vector=Get Gravity Override()
    ExternalForce
    SimSpaceSettings:Sim Space Settings Structure=Get Sim Space Settings() -> (Sim Space Settings:Sim Space Settings Structure)
   ) -> (Pose) {
    bClampLinearTranslationLimitToRefPose=true
    bOverrideWorldGravity=true
    overridePhysicsAsset=&/M_MED_LazarusLens_Wing_Physics
    simulationSpace=BaseBoneSpace
    alphaBoolBlend {blendInTime=0.1, blendOption=Linear, blendOutTime=0.1}
    baseBoneRef {boneName=dyn_pelvis_space}
    componentAppliedLinearAccClamp=<2000 2000 2000>
    componentLinearAccScale=<0.3 0.3 0.5>
    componentLinearVelScale=<0.3 0.3 0.5>
    simSpaceSettings {
     maxAngularAcceleration=1000
     maxAngularVelocity=3
     maxLinearAcceleration=700
     maxLinearVelocity=500
     velocityScaleZ=0
     externalLinearDragV=<0.2 0.2>
    }
   }
   n28 = Component To Local(ComponentPose=@n27)
   n29 = RigidBody!(ComponentPose=Local To Component(LocalPose=Use cached pose 'PreRBAN'()), Alpha, OverrideWorldGravity="0,0,-980", ExternalForce) -> (Pose) {
    bClampLinearTranslationLimitToRefPose=true
    bOverrideWorldGravity=true
    overridePhysicsAsset=&/M_MED_LazarusLens_Wing_Dress_FrontEnd_Physics
    simulationSpace=BaseBoneSpace
    alphaBoolBlend {blendInTime=0.1, blendOption=Linear, blendOutTime=0.1}
    baseBoneRef {boneName=root}
    componentAppliedLinearAccClamp=<3000 3000 3000>
    componentLinearAccScale=<0.4 0.4 0.5>
    simSpaceSettings {
     maxAngularAcceleration=1000
     maxAngularVelocity=3
     maxLinearAcceleration=700
     maxLinearVelocity=500
     velocityScaleZ=0
     externalLinearDragV=<0.2 0.2>
    }
   }
   n30 = Component To Local(ComponentPose=@n29)
   n31 = Get bIsInFrontEnd()
   n32 = "Constraint - dyn_flap_bk_c (P)"(
    ComponentPose=Local To Component(LocalPose=Blend Poses by bool(BlendPose_0=@n30, BlendPose_1=@n28, bActiveValue=@n31) {blendType=Linear})
    Alpha=1
    ConstraintWeights_0=1
   ) {
    alphaBoolBlend {blendOption=Linear}
    boneToModify {boneName=dyn_flap_bk_c}
    constraintSetup [{targetBone {boneName=pelvis}}]
    constraintWeights [1]
   }
   n33 = "Constraint - dyn_flap_bk_l_01 (R)"(ComponentPose=@n32, Alpha=1, ConstraintWeights_0=1) {
    alphaBoolBlend {blendOption=Linear}
    boneToModify {boneName=dyn_flap_bk_l_01}
    constraintSetup [{transformType=Rotation, targetBone {boneName=thigh_l}}]
    constraintWeights [1]
   }
   n34 = "Constraint - dyn_flap_bk_r_01 (R)"(ComponentPose=@n33, Alpha=1, ConstraintWeights_0=1) {
    alphaBoolBlend {blendOption=Linear}
    boneToModify {boneName=dyn_flap_bk_r_01}
    constraintSetup [{transformType=Rotation, targetBone {boneName=thigh_r}}]
    constraintWeights [1]
   }
   n35 = "Transform (Modify) Bone - Bone: dyn_flap_bk_l_01"(ComponentPose=@n34, Alpha=1, Translation="0,0,2", Rotation="0,0,0", Scale="1,1,1") {
    rotationMode=BMM_Additive
    rotationSpace=BCS_BoneSpace
    translationMode=BMM_Additive
    translationSpace=BCS_BoneSpace
    alphaBoolBlend {blendOption=Linear}
    boneToModify {boneName=dyn_flap_bk_l_01}
    translation {z=2}
   }
   n36 = "Transform (Modify) Bone - Bone: dyn_flap_bk_r_01"(ComponentPose=@n35, Alpha=1, Translation="0,0,2", Rotation="0,0,0", Scale="1,1,1") {
    rotationMode=BMM_Additive
    rotationSpace=BCS_BoneSpace
    translationMode=BMM_Additive
    translationSpace=BCS_BoneSpace
    alphaBoolBlend {blendOption=Linear}
    boneToModify {boneName=dyn_flap_bk_r_01}
    translation {z=2}
   }
   n37 = float * float(A=Get wing_backpack_var() -> (wing_backpack_var:double), B=0.5) {is=title K2Node_PromotableOperator}
   n38 = "Constraint - dyn_flap_bk_l_02 (P)"(ComponentPose=@n36, Alpha=@n37, ConstraintWeights_0=1) {
    alphaBoolBlend {blendOption=Linear}
    boneToModify {boneName=dyn_flap_bk_l_02}
    constraintSetup [{targetBone {boneName=dyn_pants_l_01}}]
    constraintWeights [1]
   }
   n39 = "Constraint - dyn_flap_bk_r_02 (P)"(ComponentPose=@n38, Alpha=@n37, ConstraintWeights_0=1) {
    alphaBoolBlend {blendOption=Linear}
    boneToModify {boneName=dyn_flap_bk_r_02}
    constraintSetup [{targetBone {boneName=dyn_pants_r_01}}]
    constraintWeights [1]
   }
   n40 = "Constraint - dyn_flap_bk_l_02 (R)"(ComponentPose=@n39, Alpha=1, ConstraintWeights_0=1) {
    alphaBoolBlend {blendOption=Linear}
    boneToModify {boneName=dyn_flap_bk_l_02}
    constraintSetup [{transformType=Rotation, targetBone {boneName=dyn_pants_l_01}}]
    constraintWeights [1]
   }
   n41 = "Constraint - dyn_flap_bk_r_02 (R)"(ComponentPose=@n40, Alpha=1, ConstraintWeights_0=1) {
    alphaBoolBlend {blendOption=Linear}
    boneToModify {boneName=dyn_flap_bk_r_02}
    constraintSetup [{transformType=Rotation, targetBone {boneName=dyn_pants_r_01}}]
    constraintWeights [1]
   }
   n42 = "Transform (Modify) Bone - Bone: dyn_flap_bk_l_02"(ComponentPose=@n41, Alpha=1, Translation="0,0,3", Rotation="5,0,0", Scale="1,1,1") {
    rotationMode=BMM_Additive
    rotationSpace=BCS_BoneSpace
    translationMode=BMM_Additive
    translationSpace=BCS_BoneSpace
    alphaBoolBlend {blendOption=Linear}
    boneToModify {boneName=dyn_flap_bk_l_02}
    rotation {pitch=5}
    translation {z=3}
   }
   n43 = "Transform (Modify) Bone - Bone: dyn_flap_bk_r_02"(ComponentPose=@n42, Alpha=1, Translation="0,0,3", Rotation="5,0,0", Scale="1,1,1") {
    rotationMode=BMM_Additive
    rotationSpace=BCS_BoneSpace
    translationMode=BMM_Additive
    translationSpace=BCS_BoneSpace
    alphaBoolBlend {blendOption=Linear}
    boneToModify {boneName=dyn_flap_bk_r_02}
    rotation {pitch=5}
    translation {z=3}
   }
   n44 = "Constraint - dyn_flap_bk_l_03 (PP)"!(ComponentPose=@n43, Alpha=1, ConstraintWeights_0=0.4, ConstraintWeights_1:float=0.6) -> (Pose) {
    alphaBoolBlend {blendOption=Linear}
    boneToModify {boneName=dyn_flap_bk_l_03}
    constraintSetup [{targetBone {boneName=dyn_pants_l_02}}, {targetBone {boneName=dyn_pants_l_01}}]
    constraintWeights [1,1]
   }
   n45 = "Constraint - dyn_flap_bk_r_03 (PP)"!(ComponentPose=@n44, Alpha=1, ConstraintWeights_0=0.6, ConstraintWeights_1:float=0.4) -> (Pose) {
    alphaBoolBlend {blendOption=Linear}
    boneToModify {boneName=dyn_flap_bk_r_03}
    constraintSetup [{targetBone {boneName=dyn_pants_r_01}}, {targetBone {boneName=dyn_pants_r_02}}]
    constraintWeights [1,1]
   }
   n46 = "Constraint - dyn_flap_bk_l_03 (R)"(ComponentPose=@n45, Alpha=1, ConstraintWeights_0=1) {
    alphaBoolBlend {blendOption=Linear}
    boneToModify {boneName=dyn_flap_bk_l_03}
    constraintSetup [{transformType=Rotation, targetBone {boneName=dyn_pants_l_02}}]
    constraintWeights [1]
   }
   n47 = "Constraint - dyn_flap_bk_r_03 (R)"(ComponentPose=@n46, Alpha=1, ConstraintWeights_0=1) {
    alphaBoolBlend {blendOption=Linear}
    boneToModify {boneName=dyn_flap_bk_r_03}
    constraintSetup [{transformType=Rotation, targetBone {boneName=dyn_pants_r_02}}]
    constraintWeights [1]
   }
   n48 = "Transform (Modify) Bone - Bone: dyn_flap_bk_l_03"(ComponentPose=@n47, Alpha=1, Translation="0,0,3", Rotation="5,0,0", Scale="1,1,1") {
    rotationMode=BMM_Additive
    rotationSpace=BCS_BoneSpace
    translationMode=BMM_Additive
    translationSpace=BCS_BoneSpace
    alphaBoolBlend {blendOption=Linear}
    boneToModify {boneName=dyn_flap_bk_l_03}
    rotation {pitch=5}
    translation {z=3}
   }
   n49 = "Transform (Modify) Bone - Bone: dyn_flap_bk_r_03"(ComponentPose=@n48, Alpha=1, Translation="0,0,3", Rotation="5,0,0", Scale="1,1,1") {
    rotationMode=BMM_Additive
    rotationSpace=BCS_BoneSpace
    translationMode=BMM_Additive
    translationSpace=BCS_BoneSpace
    alphaBoolBlend {blendOption=Linear}
    boneToModify {boneName=dyn_flap_bk_r_03}
    rotation {pitch=5}
    translation {z=3}
   }
   n50 = "Transform (Modify) Bone - Bone: dyn_flap_bk_l_01"(ComponentPose=@n49, Alpha=1, Translation="0,0,0", Rotation="5,0,0", Scale="1,1,1") {
    rotationMode=BMM_Additive
    rotationSpace=BCS_BoneSpace
    translationMode=BMM_Additive
    translationSpace=BCS_BoneSpace
    alphaBoolBlend {blendOption=Linear}
    boneToModify {boneName=dyn_flap_bk_l_01}
    rotation {pitch=5}
   }
   n51 = "Transform (Modify) Bone - Bone: dyn_flap_bk_r_01"(ComponentPose=@n50, Alpha=1, Translation="0,0,0", Rotation="5,0,0", Scale="1,1,1") {
    rotationMode=BMM_Additive
    rotationSpace=BCS_BoneSpace
    translationMode=BMM_Additive
    translationSpace=BCS_BoneSpace
    alphaBoolBlend {blendOption=Linear}
    boneToModify {boneName=dyn_flap_bk_r_01}
    rotation {pitch=5}
   }
   n52 = "Transform (Modify) Bone - Bone: dyn_flap_bk_l_01"(ComponentPose=@n51, Alpha=1, Translation="0,0,1", Rotation="0,0,0", Scale="1,1,1") {
    rotationMode=BMM_Additive
    rotationSpace=BCS_BoneSpace
    translationMode=BMM_Additive
    translationSpace=BCS_BoneSpace
    alphaBoolBlend {blendOption=Linear}
    boneToModify {boneName=dyn_flap_bk_l_01}
    translation {z=1}
   }
   n53 = "Transform (Modify) Bone - Bone: dyn_sash_bot_01"(ComponentPose=@n52, Alpha=1, Translation="0,0,0", Rotation="3,0,0", Scale="1,1,1") {
    rotationMode=BMM_Additive
    rotationSpace=BCS_BoneSpace
    translationMode=BMM_Additive
    translationSpace=BCS_BoneSpace
    alphaBoolBlend {blendOption=Linear}
    boneToModify {boneName=dyn_sash_bot_01}
    rotation {pitch=3}
   }
   n54 = "Transform (Modify) Bone - Bone: dyn_sash_bot_02"(ComponentPose=@n53, Alpha=1, Translation="0,0,0", Rotation="0,0,5", Scale="1,1,1") {
    rotationMode=BMM_Additive
    rotationSpace=BCS_BoneSpace
    translationMode=BMM_Additive
    translationSpace=BCS_BoneSpace
    alphaBoolBlend {blendOption=Linear}
    boneToModify {boneName=dyn_sash_bot_02}
    rotation {roll=5}
   }
   n55 = "Constraint - dyn_sash_top_02 (P)"(ComponentPose=@n54, Alpha=1, ConstraintWeights_0=1) {
    alphaBoolBlend {blendOption=Linear}
    boneToModify {boneName=dyn_sash_top_02}
    constraintSetup [{targetBone {boneName=dyn_sash_bot_01}}]
    constraintWeights [1]
   }
   n56 = Anim Dynamics - Bone: dyn_sash_top_02!(ComponentPose=@n55, Alpha=1, GravityOverride:Vector="0,0,-980", ExternalForce="0,0,0") -> (Pose) {
    angularDampingOverride=0.99
    angularSpringConstant=200
    bGravityOverrideInSimSpace=true
    bOverrideAngularDamping=true
    bUseGravityOverride=true
    gravityScale=0.3
    simulationSpace=BoneRelative
    alphaBoolBlend {blendOption=Linear}
    boundBone {boneName=dyn_sash_top_02}
    componentAppliedLinearAccClamp=<2500 2500 1500>
    componentLinearAccScale=<0.5 0.5 0.5>
    componentLinearVelScale=<0.55 0.55 0.55>
    gravityOverride {z=500}
    physicsBodyDefinitions [{
      sphereCollisionRadius=0
      boundBone {boneName=dyn_sash_top_02}
      boxExtents=<18 18 18>
      constraintSetup {angularLimitsMax=<5 0 10>, angularLimitsMin=<-5 -30 -10>, angularTarget=<0 0 0>}
      localJointOffset=<1 0 0>
     }]
    relativeSpaceBone {boneName=root}
   }
   Save cached pose 'Base_Pose'(Pose=Component To Local(ComponentPose=@n56)) {cacheName=Base_Pose}
   n57 = "Transform (Modify) Bone - Bone: dyn_flap_bk_c"(ComponentPose=Local To Component(LocalPose=Use cached pose 'Base_Pose'()), Alpha=1, Translation="-1,0,11", Rotation="2,0,0", Scale="1,1,1") {
    rotationMode=BMM_Additive
    rotationSpace=BCS_BoneSpace
    translationMode=BMM_Additive
    translationSpace=BCS_BoneSpace
    alphaBoolBlend {blendOption=Linear}
    boneToModify {boneName=dyn_flap_bk_c}
    rotation {pitch=2}
    translation {x=-1, z=11}
   }
   n58 = Component To Local(ComponentPose=@n57)
   n59 = Use cached pose 'Base_Pose'()
   n60 = Get sliding_bool() -> (sliding_bool:bool)
   n61 = "Transform (Modify) Bone - Bone: dyn_skirt_fix"(
    ComponentPose=Local To Component(LocalPose=Blend Poses by bool(BlendPose_0=@n58, BlendPose_1=@n59, bActiveValue=@n60) {blendType=Linear})
    Alpha=1
    Translation="0,2,0"
    Rotation="0,0,0"
    Scale="1,1,1"
   ) {
    rotationMode=BMM_Additive
    rotationSpace=BCS_BoneSpace
    translationMode=BMM_Additive
    translationSpace=BCS_BoneSpace
    alphaBoolBlend {blendOption=Linear}
    boneToModify {boneName=dyn_skirt_fix}
    translation {y=2}
   }
   Save cached pose 'Modified Pose'(Pose=Component To Local(ComponentPose=@n61)) {cacheName=Modified Pose}
   n62 = Control Rig(Source=Copy Pose From Mesh() {bCopyCurves=true, bCopyCustomAttributes=true, bUseAttachedParent=true, bUseMeshPose=true}) {
    alphaBoolBlend {blendOption=Linear}
    controlRigAssetReference {blueprintRigClass=&/M_MED_LazarusLens_Wing_Deform_CtrlRig.M_MED_LazarusLens_Wing_Deform_CtrlRig_C}
   }
   n63 = "Constraint - dyn_flap_fr_r_base (P)"(ComponentPose=Local To Component(LocalPose=@n62), Alpha=1, ConstraintWeights_0=1) {
    alphaBoolBlend {blendOption=Linear}
    boneToModify {boneName=dyn_flap_fr_r_base}
    constraintSetup [{targetBone {boneName=thigh_r}}]
    constraintWeights [1]
   }
   n64 = "Constraint - dyn_flap_fr_l_base (P)"(ComponentPose=@n63, Alpha=1, ConstraintWeights_0=1) {
    alphaBoolBlend {blendOption=Linear}
    boneToModify {boneName=dyn_flap_fr_l_base}
    constraintSetup [{targetBone {boneName=thigh_l}}]
    constraintWeights [1]
   }
   n65 = "Constraint - dyn_flap_fr_r_01 (P)"(ComponentPose=@n64, Alpha=1, ConstraintWeights_0=1) {
    alphaBoolBlend {blendOption=Linear}
    boneToModify {boneName=dyn_flap_fr_r_01}
    constraintSetup [{targetBone {boneName=thigh_r}}]
    constraintWeights [1]
   }
   n66 = "Constraint - dyn_flap_fr_l_01 (P)"(ComponentPose=@n65, Alpha=1, ConstraintWeights_0=1) {
    alphaBoolBlend {blendOption=Linear}
    boneToModify {boneName=dyn_flap_fr_l_01}
    constraintSetup [{targetBone {boneName=thigh_l}}]
    constraintWeights [1]
   }
   n67 = Anim Dynamics - Bone: dyn_pants_l_01!(ComponentPose=@n66, Alpha=1, GravityOverride:Vector="0,0,-980", ExternalForce="0,0,0") -> (Pose) {
    angularDampingOverride=0.99
    angularSpringConstant=100
    bAngularSpring=true
    bGravityOverrideInSimSpace=true
    bOverrideAngularDamping=true
    bUseGravityOverride=true
    gravityScale=0.3
    simulationSpace=BoneRelative
    alphaBoolBlend {blendOption=Linear}
    boundBone {boneName=dyn_pants_l_01}
    componentAppliedLinearAccClamp=<2500 2500 1500>
    componentLinearAccScale=<0.5 0.5 0.5>
    componentLinearVelScale=<0.55 0.55 0.55>
    gravityOverride {z=500}
    physicsBodyDefinitions [{
      sphereCollisionRadius=0
      boundBone {boneName=dyn_pants_l_01}
      boxExtents=<10 10 10>
      constraintSetup {angularLimitsMax=<5 5 5>, angularLimitsMin=<-5 -5 -5>, angularTarget=<0 0 0>}
      localJointOffset=<1 0 0>
     }]
    relativeSpaceBone {boneName=root}
   }
   n68 = Anim Dynamics - Bone: dyn_pants_l_02!(ComponentPose=@n67, Alpha=1, GravityOverride:Vector="0,0,-980", ExternalForce="0,0,0") -> (Pose) {
    angularDampingOverride=0.99
    angularSpringConstant=100
    bAngularSpring=true
    bGravityOverrideInSimSpace=true
    bOverrideAngularDamping=true
    bUseGravityOverride=true
    gravityScale=0.3
    simulationSpace=BoneRelative
    alphaBoolBlend {blendOption=Linear}
    boundBone {boneName=dyn_pants_l_02}
    componentAppliedLinearAccClamp=<2500 2500 1500>
    componentLinearAccScale=<0.5 0.5 0.5>
    componentLinearVelScale=<0.55 0.55 0.55>
    gravityOverride {z=500}
    physicsBodyDefinitions [{
      sphereCollisionRadius=0
      boundBone {boneName=dyn_pants_l_02}
      boxExtents=<10 10 10>
      constraintSetup {angularLimitsMax=<5 10 10>, angularLimitsMin=<-5 -10 -10>, angularTarget=<0 0 0>}
      localJointOffset=<1 0 0>
     }]
    relativeSpaceBone {boneName=root}
   }
   n69 = Anim Dynamics - Bone: dyn_pants_r_01!(ComponentPose=@n68, Alpha=1, GravityOverride:Vector="0,0,-980", ExternalForce="0,0,0") -> (Pose) {
    angularDampingOverride=0.99
    angularSpringConstant=100
    bAngularSpring=true
    bGravityOverrideInSimSpace=true
    bOverrideAngularDamping=true
    bUseGravityOverride=true
    gravityScale=0.3
    simulationSpace=BoneRelative
    alphaBoolBlend {blendOption=Linear}
    boundBone {boneName=dyn_pants_r_01}
    componentAppliedLinearAccClamp=<2500 2500 1500>
    componentLinearAccScale=<0.5 0.5 0.5>
    componentLinearVelScale=<0.55 0.55 0.55>
    gravityOverride {z=500}
    physicsBodyDefinitions [{
      sphereCollisionRadius=0
      boundBone {boneName=dyn_pants_r_01}
      boxExtents=<10 10 10>
      constraintSetup {angularLimitsMax=<5 5 5>, angularLimitsMin=<-5 -5 -5>, angularTarget=<0 0 0>}
      localJointOffset=<1 0 0>
     }]
    relativeSpaceBone {boneName=root}
   }
   n70 = Anim Dynamics - Bone: dyn_pants_r_02!(ComponentPose=@n69, Alpha=1, GravityOverride:Vector="0,0,-980", ExternalForce="0,0,0") -> (Pose) {
    angularDampingOverride=0.99
    angularSpringConstant=100
    bAngularSpring=true
    bGravityOverrideInSimSpace=true
    bOverrideAngularDamping=true
    bUseGravityOverride=true
    gravityScale=0.3
    simulationSpace=BoneRelative
    alphaBoolBlend {blendOption=Linear}
    boundBone {boneName=dyn_pants_r_02}
    componentAppliedLinearAccClamp=<2500 2500 1500>
    componentLinearAccScale=<0.5 0.5 0.5>
    componentLinearVelScale=<0.55 0.55 0.55>
    gravityOverride {z=500}
    physicsBodyDefinitions [{
      sphereCollisionRadius=0
      boundBone {boneName=dyn_pants_r_02}
      boxExtents=<10 10 10>
      constraintSetup {angularLimitsMax=<5 10 10>, angularLimitsMin=<-5 -10 -10>, angularTarget=<0 0 0>}
      localJointOffset=<1 0 0>
     }]
    relativeSpaceBone {boneName=root}
   }
   n71 = "Constraint - dyn_flap_fr_l_02_outer (P)"(ComponentPose=@n70, Alpha=1, ConstraintWeights_0=1) {
    alphaBoolBlend {blendOption=Linear}
    boneToModify {boneName=dyn_flap_fr_l_02_outer}
    constraintSetup [{targetBone {boneName=dyn_pants_l_01}}]
    constraintWeights [1]
   }
   n72 = "Constraint - dyn_flap_fr_l_02_innter (P)"(ComponentPose=@n71, Alpha=1, ConstraintWeights_0=1) {
    alphaBoolBlend {blendOption=Linear}
    boneToModify {boneName=dyn_flap_fr_l_02_innter}
    constraintSetup [{targetBone {boneName=dyn_pants_l_01}}]
    constraintWeights [1]
   }
   n73 = "Constraint - dyn_flap_fr_l_03_outer (P)"(ComponentPose=@n72, Alpha=1, ConstraintWeights_0=1) {
    alphaBoolBlend {blendOption=Linear}
    boneToModify {boneName=dyn_flap_fr_l_03_outer}
    constraintSetup [{targetBone {boneName=dyn_pants_l_02}}]
    constraintWeights [1]
   }
   n74 = "Constraint - dyn_flap_fr_l_03_innter (P)"(ComponentPose=@n73, Alpha=1, ConstraintWeights_0=1) {
    alphaBoolBlend {blendOption=Linear}
    boneToModify {boneName=dyn_flap_fr_l_03_innter}
    constraintSetup [{targetBone {boneName=dyn_pants_l_02}}]
    constraintWeights [1]
   }
   n75 = "Constraint - dyn_flap_fr_r_02_outer (P)"(ComponentPose=@n74, Alpha=1, ConstraintWeights_0=1) {
    alphaBoolBlend {blendOption=Linear}
    boneToModify {boneName=dyn_flap_fr_r_02_outer}
    constraintSetup [{targetBone {boneName=dyn_pants_r_01}}]
    constraintWeights [1]
   }
   n76 = "Constraint - dyn_flap_fr_r_02_innter (P)"(ComponentPose=@n75, Alpha=1, ConstraintWeights_0=1) {
    alphaBoolBlend {blendOption=Linear}
    boneToModify {boneName=dyn_flap_fr_r_02_innter}
    constraintSetup [{targetBone {boneName=dyn_pants_r_01}}]
    constraintWeights [1]
   }
   n77 = "Constraint - dyn_flap_fr_r_03_outer (P)"(ComponentPose=@n76, Alpha=1, ConstraintWeights_0=1) {
    alphaBoolBlend {blendOption=Linear}
    boneToModify {boneName=dyn_flap_fr_r_03_outer}
    constraintSetup [{targetBone {boneName=dyn_pants_r_02}}]
    constraintWeights [1]
   }
   n78 = "Constraint - dyn_flap_fr_r_03_innter (P)"(ComponentPose=@n77, Alpha=1, ConstraintWeights_0=1) {
    alphaBoolBlend {blendOption=Linear}
    boneToModify {boneName=dyn_flap_fr_r_03_innter}
    constraintSetup [{targetBone {boneName=dyn_pants_r_02}}]
    constraintWeights [1]
   }
   n79 = Anim Dynamics - Bone: dyn_flap_fr_l_base!(ComponentPose=@n78, Alpha=1, GravityOverride:Vector="0,0,-980", ExternalForce="0,0,0") -> (Pose) {
    angularDampingOverride=0.99
    angularSpringConstant=200
    bGravityOverrideInSimSpace=true
    bOverrideAngularDamping=true
    bUseGravityOverride=true
    gravityScale=0.3
    simulationSpace=BoneRelative
    alphaBoolBlend {blendOption=Linear}
    boundBone {boneName=dyn_flap_fr_l_base}
    componentAppliedLinearAccClamp=<2500 2500 2500>
    componentLinearAccScale=<0.5 0.4 0.55>
    componentLinearVelScale=<0.55 0.55 0.55>
    gravityOverride {z=500}
    physicsBodyDefinitions [{
      sphereCollisionRadius=0
      boundBone {boneName=dyn_flap_fr_l_base}
      boxExtents=<15 15 15>
      constraintSetup {angularLimitsMax=<5 0 10>, angularLimitsMin=<-5 -45 -10>, angularTarget=<0 0 0>}
      localJointOffset=<0.8 0 0>
     }]
    relativeSpaceBone {boneName=root}
   }
   n80 = Anim Dynamics - Bone: dyn_flap_fr_l_01!(ComponentPose=@n79, Alpha=1, GravityOverride:Vector="0,0,-980", ExternalForce="0,0,0") -> (Pose) {
    angularDampingOverride=0.99
    angularSpringConstant=200
    bGravityOverrideInSimSpace=true
    bOverrideAngularDamping=true
    bUseGravityOverride=true
    gravityScale=0.3
    simulationSpace=BoneRelative
    alphaBoolBlend {blendOption=Linear}
    boundBone {boneName=dyn_flap_fr_l_01}
    componentAppliedLinearAccClamp=<2500 2500 2500>
    componentLinearAccScale=<0.5 0.4 0.55>
    componentLinearVelScale=<0.55 0.55 0.55>
    gravityOverride {z=500}
    physicsBodyDefinitions [{
      sphereCollisionRadius=0
      boundBone {boneName=dyn_flap_fr_l_01}
      boxExtents=<15 15 15>
      constraintSetup {angularLimitsMax=<5 0 10>, angularLimitsMin=<-5 -45 -10>, angularTarget=<0 0 0>}
      localJointOffset=<0.8 0 0>
     }]
    relativeSpaceBone {boneName=root}
   }
   n81 = Anim Dynamics - Bone: dyn_flap_fr_l_02_outer!(ComponentPose=@n80, Alpha=1, GravityOverride:Vector="0,0,-980", ExternalForce="0,0,0") -> (Pose) {
    angularDampingOverride=0.99
    angularSpringConstant=200
    bGravityOverrideInSimSpace=true
    bOverrideAngularDamping=true
    bUseGravityOverride=true
    gravityScale=0.3
    simulationSpace=BoneRelative
    alphaBoolBlend {blendOption=Linear}
    boundBone {boneName=dyn_flap_fr_l_02_outer}
    componentAppliedLinearAccClamp=<2500 2500 2500>
    componentLinearAccScale=<0.5 0.4 0.55>
    componentLinearVelScale=<0.55 0.55 0.55>
    gravityOverride {z=500}
    physicsBodyDefinitions [{
      sphereCollisionRadius=0
      boundBone {boneName=dyn_flap_fr_l_02_outer}
      boxExtents=<15 15 15>
      constraintSetup {angularLimitsMax=<5 0 10>, angularLimitsMin=<-5 -45 -10>, angularTarget=<0 0 0>}
      localJointOffset=<0.8 0 0>
     }]
    relativeSpaceBone {boneName=root}
   }
   n82 = Anim Dynamics - Bone: dyn_flap_fr_l_02_innter!(ComponentPose=@n81, Alpha=1, GravityOverride:Vector="0,0,-980", ExternalForce="0,0,0") -> (Pose) {
    angularDampingOverride=0.99
    angularSpringConstant=200
    bGravityOverrideInSimSpace=true
    bOverrideAngularDamping=true
    bUseGravityOverride=true
    gravityScale=0.3
    simulationSpace=BoneRelative
    alphaBoolBlend {blendOption=Linear}
    boundBone {boneName=dyn_flap_fr_l_02_innter}
    componentAppliedLinearAccClamp=<2500 2500 2500>
    componentLinearAccScale=<0.5 0.4 0.55>
    componentLinearVelScale=<0.55 0.55 0.55>
    gravityOverride {z=500}
    physicsBodyDefinitions [{
      sphereCollisionRadius=0
      boundBone {boneName=dyn_flap_fr_l_02_innter}
      boxExtents=<15 15 15>
      constraintSetup {angularLimitsMax=<10 0 10>, angularLimitsMin=<-5 -45 -10>, angularTarget=<0 0 0>}
      localJointOffset=<0.8 0 0>
     }]
    relativeSpaceBone {boneName=root}
   }
   n83 = Anim Dynamics - Bone: dyn_flap_fr_l_03_outer!(ComponentPose=@n82, Alpha=1, GravityOverride:Vector="0,0,-980", ExternalForce="0,0,0") -> (Pose) {
    angularDampingOverride=0.99
    angularSpringConstant=200
    bGravityOverrideInSimSpace=true
    bOverrideAngularDamping=true
    bUseGravityOverride=true
    gravityScale=0.3
    simulationSpace=BoneRelative
    alphaBoolBlend {blendOption=Linear}
    boundBone {boneName=dyn_flap_fr_l_03_outer}
    componentAppliedLinearAccClamp=<2500 2500 2500>
    componentLinearAccScale=<0.5 0.4 0.55>
    componentLinearVelScale=<0.55 0.55 0.55>
    gravityOverride {z=500}
    physicsBodyDefinitions [{
      sphereCollisionRadius=0
      boundBone {boneName=dyn_flap_fr_l_03_outer}
      boxExtents=<15 15 15>
      constraintSetup {angularLimitsMax=<5 0 10>, angularLimitsMin=<-5 -45 -10>, angularTarget=<0 0 0>}
      localJointOffset=<0.8 0 0>
     }]
    relativeSpaceBone {boneName=root}
   }
   n84 = Anim Dynamics - Bone: dyn_flap_fr_l_03_innter!(ComponentPose=@n83, Alpha=1, GravityOverride:Vector="0,0,-980", ExternalForce="0,0,0") -> (Pose) {
    angularDampingOverride=0.99
    angularSpringConstant=200
    bGravityOverrideInSimSpace=true
    bOverrideAngularDamping=true
    bUseGravityOverride=true
    gravityScale=0.3
    simulationSpace=BoneRelative
    alphaBoolBlend {blendOption=Linear}
    boundBone {boneName=dyn_flap_fr_l_03_innter}
    componentAppliedLinearAccClamp=<2500 2500 2500>
    componentLinearAccScale=<0.5 0.4 0.55>
    componentLinearVelScale=<0.55 0.55 0.55>
    gravityOverride {z=500}
    physicsBodyDefinitions [{
      sphereCollisionRadius=0
      boundBone {boneName=dyn_flap_fr_l_03_innter}
      boxExtents=<15 15 15>
      constraintSetup {angularLimitsMax=<10 0 10>, angularLimitsMin=<-5 -45 -10>, angularTarget=<0 0 0>}
      localJointOffset=<0.8 0 0>
     }]
    relativeSpaceBone {boneName=root}
   }
   n85 = "Transform (Modify) Bone - Bone: dyn_flap_fr_l_01"(ComponentPose=@n84, Alpha=1, Translation="0,0,0.5", Rotation="0,0,0", Scale="1,1,1") {
    rotationMode=BMM_Additive
    rotationSpace=BCS_BoneSpace
    translationMode=BMM_Additive
    translationSpace=BCS_BoneSpace
    alphaBoolBlend {blendOption=Linear}
    boneToModify {boneName=dyn_flap_fr_l_01}
    translation {z=0.5}
   }
   n86 = Anim Dynamics - Bone: dyn_flap_fr_r_base!(ComponentPose=@n85, Alpha=1, GravityOverride:Vector="0,0,-980", ExternalForce="0,0,0") -> (Pose) {
    angularDampingOverride=0.99
    angularSpringConstant=200
    bGravityOverrideInSimSpace=true
    bOverrideAngularDamping=true
    bUseGravityOverride=true
    gravityScale=0.3
    simulationSpace=BoneRelative
    alphaBoolBlend {blendOption=Linear}
    boundBone {boneName=dyn_flap_fr_r_base}
    componentAppliedLinearAccClamp=<2500 2500 2500>
    componentLinearAccScale=<0.5 0.4 0.55>
    componentLinearVelScale=<0.55 0.55 0.55>
    gravityOverride {z=500}
    physicsBodyDefinitions [{
      sphereCollisionRadius=0
      boundBone {boneName=dyn_flap_fr_r_base}
      boxExtents=<15 15 15>
      constraintSetup {angularLimitsMax=<2 0 2>, angularLimitsMin=<-2 -45 -2>, angularTarget=<0 0 0>}
      localJointOffset=<0.8 0 0>
     }]
    planarLimits [{drivingBone {boneName=root}, planeTransform {location=<0 0 0>, rotation=<rot 0 0 0>, scale=<3 3 3>}}]
    relativeSpaceBone {boneName=root}
   }
   n87 = Anim Dynamics - Bone: dyn_flap_fr_r_01!(ComponentPose=@n86, Alpha=1, GravityOverride:Vector="0,0,-980", ExternalForce="0,0,0") -> (Pose) {
    angularDampingOverride=0.99
    angularSpringConstant=200
    bGravityOverrideInSimSpace=true
    bOverrideAngularDamping=true
    bUseGravityOverride=true
    gravityScale=0.3
    simulationSpace=BoneRelative
    alphaBoolBlend {blendOption=Linear}
    boundBone {boneName=dyn_flap_fr_r_01}
    componentAppliedLinearAccClamp=<2500 2500 2500>
    componentLinearAccScale=<0.5 0.4 0.55>
    componentLinearVelScale=<0.55 0.55 0.55>
    gravityOverride {z=500}
    physicsBodyDefinitions [{
      sphereCollisionRadius=0
      boundBone {boneName=dyn_flap_fr_r_01}
      boxExtents=<15 15 15>
      constraintSetup {angularLimitsMax=<5 0 10>, angularLimitsMin=<-5 -45 -10>, angularTarget=<0 0 0>}
      localJointOffset=<0.8 0 0>
     }]
    planarLimits [{drivingBone {boneName=root}, planeTransform {location=<0 0 0>, rotation=<rot 0 0 0>, scale=<3 3 3>}}]
    relativeSpaceBone {boneName=root}
   }
   n88 = Anim Dynamics - Bone: dyn_flap_fr_r_02_outer!(ComponentPose=@n87, Alpha=1, GravityOverride:Vector="0,0,-980", ExternalForce="0,0,0") -> (Pose) {
    angularDampingOverride=0.99
    angularSpringConstant=200
    bGravityOverrideInSimSpace=true
    bOverrideAngularDamping=true
    bUseGravityOverride=true
    gravityScale=0.3
    simulationSpace=BoneRelative
    alphaBoolBlend {blendOption=Linear}
    boundBone {boneName=dyn_flap_fr_r_02_outer}
    componentAppliedLinearAccClamp=<2500 2500 2500>
    componentLinearAccScale=<0.5 0.4 0.55>
    componentLinearVelScale=<0.55 0.55 0.55>
    gravityOverride {z=500}
    physicsBodyDefinitions [{
      sphereCollisionRadius=0
      boundBone {boneName=dyn_flap_fr_r_02_outer}
      boxExtents=<15 15 15>
      constraintSetup {angularLimitsMax=<5 0 10>, angularLimitsMin=<-5 -45 -10>, angularTarget=<0 0 0>}
      localJointOffset=<0.8 0 0>
     }]
    planarLimits [{drivingBone {boneName=root}, planeTransform {location=<0 0 0>, rotation=<rot 0 0 0>, scale=<3 3 3>}}]
    relativeSpaceBone {boneName=root}
   }
   n89 = Anim Dynamics - Bone: dyn_flap_fr_r_02_innter!(ComponentPose=@n88, Alpha=1, GravityOverride:Vector="0,0,-980", ExternalForce="0,0,0") -> (Pose) {
    angularDampingOverride=0.99
    angularSpringConstant=200
    bGravityOverrideInSimSpace=true
    bOverrideAngularDamping=true
    bUseGravityOverride=true
    gravityScale=0.3
    simulationSpace=BoneRelative
    alphaBoolBlend {blendOption=Linear}
    boundBone {boneName=dyn_flap_fr_r_02_innter}
    componentAppliedLinearAccClamp=<2500 2500 2500>
    componentLinearAccScale=<0.5 0.4 0.55>
    componentLinearVelScale=<0.55 0.55 0.55>
    gravityOverride {z=500}
    physicsBodyDefinitions [{
      sphereCollisionRadius=0
      boundBone {boneName=dyn_flap_fr_r_02_innter}
      boxExtents=<15 15 15>
      constraintSetup {angularLimitsMax=<10 0 10>, angularLimitsMin=<-5 -45 -10>, angularTarget=<0 0 0>}
      localJointOffset=<0.8 0 0>
     }]
    planarLimits [{drivingBone {boneName=root}, planeTransform {location=<0 0 0>, rotation=<rot 0 0 0>}}]
    relativeSpaceBone {boneName=root}
   }
   n90 = Anim Dynamics - Bone: dyn_flap_fr_r_03_outer!(ComponentPose=@n89, Alpha=1, GravityOverride:Vector="0,0,-980", ExternalForce="0,0,0") -> (Pose) {
    angularDampingOverride=0.99
    angularSpringConstant=200
    bGravityOverrideInSimSpace=true
    bOverrideAngularDamping=true
    bUseGravityOverride=true
    gravityScale=0.3
    simulationSpace=BoneRelative
    alphaBoolBlend {blendOption=Linear}
    boundBone {boneName=dyn_flap_fr_r_03_outer}
    componentAppliedLinearAccClamp=<2500 2500 2500>
    componentLinearAccScale=<0.5 0.4 0.55>
    componentLinearVelScale=<0.55 0.55 0.55>
    gravityOverride {z=500}
    physicsBodyDefinitions [{
      sphereCollisionRadius=0
      boundBone {boneName=dyn_flap_fr_r_03_outer}
      boxExtents=<15 15 15>
      constraintSetup {angularLimitsMax=<5 0 10>, angularLimitsMin=<-5 -45 -10>, angularTarget=<0 0 0>}
      localJointOffset=<0.8 0 0>
     }]
    planarLimits [{drivingBone {boneName=root}, planeTransform {location=<0 0 0>, rotation=<rot 0 0 0>, scale=<3 3 3>}}]
    relativeSpaceBone {boneName=root}
   }
   n91 = Anim Dynamics - Bone: dyn_flap_fr_r_03_innter!(ComponentPose=@n90, Alpha=1, GravityOverride:Vector="0,0,-980", ExternalForce="0,0,0") -> (Pose) {
    angularDampingOverride=0.99
    angularSpringConstant=200
    bGravityOverrideInSimSpace=true
    bOverrideAngularDamping=true
    bUseGravityOverride=true
    gravityScale=0.3
    simulationSpace=BoneRelative
    alphaBoolBlend {blendOption=Linear}
    boundBone {boneName=dyn_flap_fr_r_03_innter}
    componentAppliedLinearAccClamp=<2500 2500 2500>
    componentLinearAccScale=<0.5 0.4 0.55>
    componentLinearVelScale=<0.55 0.55 0.55>
    gravityOverride {z=500}
    physicsBodyDefinitions [{
      sphereCollisionRadius=0
      boundBone {boneName=dyn_flap_fr_r_03_innter}
      boxExtents=<15 15 15>
      constraintSetup {angularLimitsMax=<10 0 10>, angularLimitsMin=<-5 -45 -10>, angularTarget=<0 0 0>}
      localJointOffset=<0.8 0 0>
     }]
    planarLimits [{drivingBone {boneName=root}, planeTransform {location=<0 0 0>, rotation=<rot 0 0 0>, scale=<3 3 3>}}]
    relativeSpaceBone {boneName=root}
   }
   n92 = "Transform (Modify) Bone - Bone: dyn_flap_fr_r_base"(ComponentPose=@n91, Alpha=1, Translation="0.5,0,1.5", Rotation="0,0,-5", Scale="1,1,1") {
    rotationMode=BMM_Additive
    rotationSpace=BCS_BoneSpace
    scaleMode=BMM_Additive
    scaleSpace=BCS_BoneSpace
    translationMode=BMM_Additive
    translationSpace=BCS_BoneSpace
    alphaBoolBlend {blendOption=Linear}
    boneToModify {boneName=dyn_flap_fr_r_base}
    rotation {roll=-5}
    translation {x=0.5, z=1.5}
   }
   n93 = "Transform (Modify) Bone - Bone: dyn_flap_fr_l_01"(ComponentPose=@n92, Alpha=1, Translation="0.5,0,1.5", Rotation="0,0,0", Scale="1,1,1") {
    rotationMode=BMM_Additive
    rotationSpace=BCS_BoneSpace
    scaleMode=BMM_Additive
    scaleSpace=BCS_BoneSpace
    translationMode=BMM_Additive
    translationSpace=BCS_BoneSpace
    alphaBoolBlend {blendOption=Linear}
    boneToModify {boneName=dyn_flap_fr_l_01}
    translation {x=0.5, z=1.5}
   }
   n94 = "Transform (Modify) Bone - Bone: dyn_flap_fr_l_02_outer"(ComponentPose=@n93, Alpha=1, Translation="0,0,1", Rotation="0,0,0", Scale="1,1,1") {
    rotationMode=BMM_Additive
    rotationSpace=BCS_BoneSpace
    scaleMode=BMM_Additive
    scaleSpace=BCS_BoneSpace
    translationMode=BMM_Additive
    translationSpace=BCS_BoneSpace
    alphaBoolBlend {blendOption=Linear}
    boneToModify {boneName=dyn_flap_fr_l_02_outer}
    translation {z=1}
   }
   n95 = "Transform (Modify) Bone - Bone: dyn_flap_fr_l_02_innter"(ComponentPose=@n94, Alpha=1, Translation="0,0,1", Rotation="0,0,0", Scale="1,1,1") {
    rotationMode=BMM_Additive
    rotationSpace=BCS_BoneSpace
    scaleMode=BMM_Additive
    scaleSpace=BCS_BoneSpace
    translationMode=BMM_Additive
    translationSpace=BCS_BoneSpace
    alphaBoolBlend {blendOption=Linear}
    boneToModify {boneName=dyn_flap_fr_l_02_innter}
    translation {z=1}
   }
   n96 = "Transform (Modify) Bone - Bone: dyn_flap_fr_r_02_outer"(ComponentPose=@n95, Alpha=1, Translation="0,0,1", Rotation="0,0,0", Scale="1,1,1") {
    rotationMode=BMM_Additive
    rotationSpace=BCS_BoneSpace
    scaleMode=BMM_Additive
    scaleSpace=BCS_BoneSpace
    translationMode=BMM_Additive
    translationSpace=BCS_BoneSpace
    alphaBoolBlend {blendOption=Linear}
    boneToModify {boneName=dyn_flap_fr_r_02_outer}
    translation {z=1}
   }
   n97 = "Transform (Modify) Bone - Bone: dyn_flap_fr_r_02_innter"(ComponentPose=@n96, Alpha=1, Translation="0,0,1", Rotation="0,0,0", Scale="1,1,1") {
    rotationMode=BMM_Additive
    rotationSpace=BCS_BoneSpace
    scaleMode=BMM_Additive
    scaleSpace=BCS_BoneSpace
    translationMode=BMM_Additive
    translationSpace=BCS_BoneSpace
    alphaBoolBlend {blendOption=Linear}
    boneToModify {boneName=dyn_flap_fr_r_02_innter}
    translation {z=1}
   }
   n98 = "Constraint - dyn_flap_fr_mid_01 (PPPPP)"!(
    ComponentPose=@n97
    Alpha=1
    ConstraintWeights_0=0.2159
    ConstraintWeights_1:float=0.0113
    ConstraintWeights_2:float=0.7657
    ConstraintWeights_3:float=0.0011
    ConstraintWeights_4:float=0.006
   ) -> (Pose) {
    alphaBoolBlend {blendOption=Linear}
    boneToModify {boneName=dyn_flap_fr_mid_01}
    constraintSetup [
     {targetBone {boneName=spine_01}}
     {targetBone {boneName=spine_02}}
     {targetBone {boneName=pelvis}}
     {targetBone {boneName=thigh_twist_01_l}}
     {targetBone {boneName=thigh_twist_01_r}}
    ]
    constraintWeights [1,1,1,1,1]
   }
   n99 = "Constraint - dyn_flap_fr_mid_02 (PPPPPPPPP)"!(
    ComponentPose=@n98
    Alpha=1
    ConstraintWeights_0=0.0274
    ConstraintWeights_1:float=0
    ConstraintWeights_2:float=0.3569
    ConstraintWeights_3:float=0.0619
    ConstraintWeights_4:float=0.0599
    ConstraintWeights_5:float=0.0116
    ConstraintWeights_6:float=0.4357
    ConstraintWeights_7:float=0.0015
    ConstraintWeights_8:float=0.045
   ) -> (Pose) {
    alphaBoolBlend {blendOption=Linear}
    boneToModify {boneName=dyn_flap_fr_mid_02}
    constraintSetup [
     {targetBone {boneName=spine_01}}
     {targetBone {boneName=spine_02}}
     {targetBone {boneName=pelvis}}
     {targetBone {boneName=thigh_twist_01_l}}
     {targetBone {boneName=thigh_twist_01_r}}
     {targetBone {boneName=dyn_flap_fr_r_01}}
     {targetBone {boneName=dyn_flap_fr_l_base}}
     {targetBone {boneName=deform_glute_r}}
     {targetBone {boneName=deform_groin_r}}
    ]
    constraintWeights [1,1,1,1,1,1,1,1,1]
   }
   n100 = "Constraint - dyn_flap_fr_mid_02 (PPPPPPPPP)"!(
    ComponentPose=@n99
    bAlphaBoolEnabled:bool=Get LOD2() -> (LOD2:bool)
    ConstraintWeights_0=0.0274
    ConstraintWeights_1=0
    ConstraintWeights_2=0.3569
    ConstraintWeights_3=0.0619
    ConstraintWeights_4=0.0599
    ConstraintWeights_5=0.0116
    ConstraintWeights_6=0.4357
    ConstraintWeights_7=0.0015
    ConstraintWeights_8=0.045
   ) -> (Pose) {
    alphaInputType=Bool
    alphaBoolBlend {blendOption=Linear}
    boneToModify {boneName=dyn_flap_fr_mid_02}
    constraintSetup [
     {targetBone {boneName=spine_01}}
     {targetBone {boneName=spine_02}}
     {targetBone {boneName=pelvis}}
     {targetBone {boneName=thigh_l}}
     {targetBone {boneName=thigh_r}}
     {targetBone {boneName=dyn_flap_fr_r_01}}
     {targetBone {boneName=dyn_flap_fr_l_base}}
     {targetBone {boneName=deform_glute_r}}
     {targetBone {boneName=deform_groin_r}}
    ]
    constraintWeights [1,1,1,1,1,1,1,1,1]
   }
   n101 = "Constraint - dyn_flap_fr_mid_03 (PPPPPPPP)"!(
    ComponentPose=@n100
    Alpha=1
    ConstraintWeights_0=0.0019
    ConstraintWeights_1:float=0.0257
    ConstraintWeights_2:float=0.0094
    ConstraintWeights_3:float=0.008
    ConstraintWeights_4:float=0.1547
    ConstraintWeights_5:float=0.1107
    ConstraintWeights_6:float=0.3524
    ConstraintWeights_7:float=0.3372
   ) -> (Pose) {
    alphaBoolBlend {blendOption=Linear}
    boneToModify {boneName=dyn_flap_fr_mid_03}
    constraintSetup [
     {targetBone {boneName=spine_01}}
     {targetBone {boneName=pelvis}}
     {targetBone {boneName=thigh_twist_01_l}}
     {targetBone {boneName=thigh_twist_01_r}}
     {targetBone {boneName=dyn_flap_fr_r_01}}
     {targetBone {boneName=dyn_flap_fr_l_01}}
     {targetBone {boneName=dyn_flap_fr_r_base}}
     {targetBone {boneName=dyn_flap_fr_l_base}}
    ]
    constraintWeights [1,1,1,1,1,1,1,1]
   }
   n102 = "Constraint - dyn_flap_fr_mid_03 (PPPPPPPP)"!(
    ComponentPose=@n101
    bAlphaBoolEnabled:bool=Get LOD2()
    ConstraintWeights_0=0.0019
    ConstraintWeights_1=0.0257
    ConstraintWeights_2=0.0094
    ConstraintWeights_3=0.008
    ConstraintWeights_4=0.1547
    ConstraintWeights_5=0.1107
    ConstraintWeights_6=0.3524
    ConstraintWeights_7=0.3372
   ) -> (Pose) {
    alphaInputType=Bool
    alphaBoolBlend {blendOption=Linear}
    boneToModify {boneName=dyn_flap_fr_mid_03}
    constraintSetup [
     {targetBone {boneName=spine_01}}
     {targetBone {boneName=pelvis}}
     {targetBone {boneName=thigh_l}}
     {targetBone {boneName=thigh_r}}
     {targetBone {boneName=dyn_flap_fr_r_01}}
     {targetBone {boneName=dyn_flap_fr_l_01}}
     {targetBone {boneName=dyn_flap_fr_r_base}}
     {targetBone {boneName=dyn_flap_fr_l_base}}
    ]
    constraintWeights [1,1,1,1,1,1,1,1]
   }
   n103 = "Constraint - dyn_flap_fr_mid_04 (PPPPPPPPPPPPPPPPP)"!(
    ComponentPose=@n102
    Alpha=1
    ConstraintWeights_0=0.0003
    ConstraintWeights_1:float=0
    ConstraintWeights_2:float=0.0001
    ConstraintWeights_3:float=0.0038
    ConstraintWeights_4:float=0.0018
    ConstraintWeights_5:float=0.0001
    ConstraintWeights_6:float=0.0296
    ConstraintWeights_7:float=0
    ConstraintWeights_8:float=0
    ConstraintWeights_9:float=0.2458
    ConstraintWeights_10:float=0.1796
    ConstraintWeights_11:float=0.0006
    ConstraintWeights_12:float=0
    ConstraintWeights_13:float=0.0007
    ConstraintWeights_14:float=0
    ConstraintWeights_15:float=0.3508
    ConstraintWeights_16:float=0.1869
   ) -> (Pose) {
    alphaBoolBlend {blendOption=Linear}
    boneToModify {boneName=dyn_flap_fr_mid_04}
    constraintSetup [
     {targetBone {boneName=spine_01}}
     {targetBone {boneName=spine_02}}
     {targetBone {boneName=thigh_l}}
     {targetBone {boneName=pelvis}}
     {targetBone {boneName=thigh_twist_01_l}}
     {targetBone {boneName=thigh_r}}
     {targetBone {boneName=thigh_twist_01_r}}
     {targetBone {boneName=calf_r}}
     {targetBone {boneName=calf_l}}
     {targetBone {boneName=dyn_flap_fr_r_01}}
     {targetBone {boneName=dyn_flap_fr_l_01}}
     {targetBone {boneName=dyn_flap_fr_l_02_outer}}
     {targetBone {boneName=dyn_flap_fr_l_03_outer}}
     {targetBone {boneName=dyn_flap_fr_l_02_innter}}
     {targetBone {boneName=dyn_flap_fr_l_03_innter}}
     {targetBone {boneName=dyn_flap_fr_r_base}}
     {targetBone {boneName=dyn_flap_fr_l_base}}
    ]
    constraintWeights [1,1,1,1,1,1,1,1,1,1,1,1,1,1,1,1,1]
   }
   n104 = "Constraint - dyn_flap_fr_mid_04 (PPPPPPPPPPPPPPPPP)"!(
    ComponentPose=@n103
    bAlphaBoolEnabled:bool=Get LOD2()
    ConstraintWeights_0=0.0003
    ConstraintWeights_1=0
    ConstraintWeights_2=0.0019
    ConstraintWeights_3=0.0038
    ConstraintWeights_4=0
    ConstraintWeights_5=0.0297
    ConstraintWeights_6=0.0296
    ConstraintWeights_7=0
    ConstraintWeights_8=0
    ConstraintWeights_9=0.2458
    ConstraintWeights_10=0.1796
    ConstraintWeights_11=0.0006
    ConstraintWeights_12=0
    ConstraintWeights_13=0.0007
    ConstraintWeights_14=0
    ConstraintWeights_15=0.3508
    ConstraintWeights_16=0.1869
   ) -> (Pose) {
    alphaInputType=Bool
    alphaBoolBlend {blendOption=Linear}
    boneToModify {boneName=dyn_flap_fr_mid_04}
    constraintSetup [
     {targetBone {boneName=spine_01}}
     {targetBone {boneName=spine_02}}
     {targetBone {boneName=thigh_l}}
     {targetBone {boneName=pelvis}}
     {targetBone {boneName=thigh_twist_01_l}}
     {targetBone {boneName=thigh_r}}
     {targetBone {boneName=thigh_twist_01_r}}
     {targetBone {boneName=calf_r}}
     {targetBone {boneName=calf_l}}
     {targetBone {boneName=dyn_flap_fr_r_01}}
     {targetBone {boneName=dyn_flap_fr_l_01}}
     {targetBone {boneName=dyn_flap_fr_l_02_outer}}
     {targetBone {boneName=dyn_flap_fr_l_03_outer}}
     {targetBone {boneName=dyn_flap_fr_l_02_innter}}
     {targetBone {boneName=dyn_flap_fr_l_03_innter}}
     {targetBone {boneName=dyn_flap_fr_r_base}}
     {targetBone {boneName=dyn_flap_fr_l_base}}
    ]
    constraintWeights [1,1,1,1,1,1,1,1,1,1,1,1,1,1,1,1,1]
   }
   n105 = "Constraint - dyn_flap_fr_mid_01 (U)"(ComponentPose=@n104, Alpha=1, ConstraintWeights_0=1) {
    alphaBoolBlend {blendOption=Linear}
    boneToModify {boneName=dyn_flap_fr_mid_01}
    constraintSetup [{transformType=LookAt, targetBone {boneName=dyn_flap_fr_mid_02}}]
    constraintWeights [1]
   }
   n106 = "Constraint - dyn_flap_fr_mid_02 (U)"(ComponentPose=@n105, Alpha=1, ConstraintWeights_0=1) {
    alphaBoolBlend {blendOption=Linear}
    boneToModify {boneName=dyn_flap_fr_mid_02}
    constraintSetup [{transformType=LookAt, targetBone {boneName=dyn_flap_fr_mid_03}}]
    constraintWeights [1]
   }
   n107 = "Constraint - dyn_flap_fr_mid_03 (U)"(ComponentPose=@n106, Alpha=1, ConstraintWeights_0=1) {
    alphaBoolBlend {blendOption=Linear}
    boneToModify {boneName=dyn_flap_fr_mid_03}
    constraintSetup [{transformType=LookAt, targetBone {boneName=dyn_flap_fr_mid_04}}]
    constraintWeights [1]
   }
   n108 = Control Rig!(Source=Component To Local(ComponentPose=@n107), Alpha, crouching:double=Get crouching() -> (crouching:double), crouch_left:double=0) -> (Pose) {
    alphaBoolBlend {blendOption=Linear}
    controlRigAssetReference {blueprintRigClass=&/M_MED_LazarusLens_Wing_Body_CtrlRig.M_MED_LazarusLens_Wing_Body_CtrlRig_C}
   }
   n109 = "Transform (Modify) Bone - Bone: dyn_flap_bk_c"!(ComponentPose=Local To Component(LocalPose=@n108), bAlphaBoolEnabled:bool=True, Translation="0,0,20", Rotation="100,0,0", Scale="1,1,1") -> (Pose) {
    alphaInputType=Bool
    rotationMode=BMM_Additive
    rotationSpace=BCS_BoneSpace
    translationMode=BMM_Additive
    translationSpace=BCS_BoneSpace
    alphaBoolBlend {blendOption=Linear}
    boneToModify {boneName=dyn_flap_bk_c}
    rotation {pitch=100}
    translation {z=20}
   }
   n110 = "Transform (Modify) Bone - Bone: dyn_flap_fr_mid_02"(ComponentPose=@n109, Alpha=1, Translation="0,0,1", Rotation="0,0,0", Scale="1,1,1") {
    rotationMode=BMM_Additive
    rotationSpace=BCS_BoneSpace
    translationMode=BMM_Additive
    translationSpace=BCS_BoneSpace
    alphaBoolBlend {blendOption=Linear}
    boneToModify {boneName=dyn_flap_fr_mid_02}
    translation {z=1}
   }
   n111 = "Transform (Modify) Bone - Bone: dyn_sash_bot_01"(ComponentPose=@n110, Alpha=1, Translation="0,0,0", Rotation="90,0,0", Scale="1,1,1") {
    rotationMode=BMM_Additive
    rotationSpace=BCS_BoneSpace
    translationMode=BMM_Additive
    translationSpace=BCS_BoneSpace
    alphaBoolBlend {blendOption=Linear}
    boneToModify {boneName=dyn_sash_bot_01}
    rotation {pitch=90}
   }
   n112 = "Constraint - dyn_pelvis_space (T)"(ComponentPose=@n111, Alpha=1, ConstraintWeights_0=1) {
    alphaBoolBlend {blendOption=Linear}
    boneToModify {boneName=dyn_pelvis_space}
    constraintSetup [{transformType=Translation, targetBone {boneName=pelvis}}]
    constraintWeights [1]
   }
   n113 = "Transform (Modify) Bone - Bone: dyn_flap_fr_mid_mstr"(ComponentPose=@n112, Alpha=1, Translation="0,0,0", Rotation="5,0,0", Scale="1,1,1") {
    rotationMode=BMM_Additive
    rotationSpace=BCS_BoneSpace
    translationMode=BMM_Additive
    translationSpace=BCS_BoneSpace
    alphaBoolBlend {blendOption=Linear}
    boneToModify {boneName=dyn_flap_fr_mid_mstr}
    rotation {pitch=5}
   }
   n114 = "Transform (Modify) Bone - Bone: dyn_sash_bot_mstr"!(ComponentPose=@n113, bAlphaBoolEnabled:bool=True, Translation="0,0,0", Rotation="5,0,0", Scale="1,1,1") -> (Pose) {
    alphaInputType=Bool
    rotationMode=BMM_Additive
    rotationSpace=BCS_BoneSpace
    translationMode=BMM_Additive
    translationSpace=BCS_BoneSpace
    alphaBoolBlend {blendOption=Linear}
    boneToModify {boneName=dyn_sash_bot_mstr}
    rotation {pitch=5}
   }
   n115 = "Transform (Modify) Bone - Bone: dyn_sash_bot_01"!(ComponentPose=@n114, bAlphaBoolEnabled:bool=True, Translation="0,0,0", Rotation="-20,0,0", Scale="1,1,1") -> (Pose) {
    alphaInputType=Bool
    rotationMode=BMM_Additive
    rotationSpace=BCS_BoneSpace
    translationMode=BMM_Additive
    translationSpace=BCS_BoneSpace
    alphaBoolBlend {blendOption=Linear}
    boneToModify {boneName=dyn_sash_bot_01}
    rotation {pitch=-20}
   }
   Save cached pose 'PreRBAN'(Pose=Component To Local(ComponentPose=@n115)) {cacheName=PreRBAN}
   Get bIsCrouching()
   Get skydive_dynamics() -> (skydive_dynamics:Vector)
  ]}
 EventGraph {nodes [
   n1 = Event BlueprintUpdateAnimation()
   > n2 = Sequence!(execute) -> (then_0, then_1, then_2:Exec, then_3:Exec, then_4:Exec, then_5:Exec, then_6:Exec)
   > n3 = Gravity Override State Function DEPRECATED(CharacterPartAnimInstance=Self-Reference(), GameTimeInSeconds=GetGameTimeInSeconds(), GravityOverrideParams=Get Gravity Override Params())
   > Set Gravity Override(Gravity Override=@n3)
   n4 = Get bAnimDynamicsEnabled() -> (bAnimDynamicsEnabled:bool)
   n5 = Get bIsCrouching()
   Set crouching(execute=@n2.then_1, crouching:double=SelectFloat(A=1, B=0, bPickA=AND Boolean(A=@n5, B=@n4))) -> (Output_Get:double)
   > n6 = Set RBAN_alpha(RBAN_alpha:bool=Get bRigidBodyEnabled()) -> (Output_Get:bool)
   > Set RBAN_off(RBAN_off:bool=NOT Boolean(A=@n6)) -> (Output_Get:bool)
   n7 = Get bIsMoving2D()
   n8 = Get bIsCrouching()
   > Set crouch move alpha(crouch move alpha:double=SelectFloat(A=1, B=0, bPickA=AND Boolean(A=@n8, B=@n7))) -> (Output_Get:double)
   > Set LL_Backpack_Equipped(LL_Backpack_Equipped:bool=BackpackItemDefinitionHasAnyMetaTag!(
     Tags:Array of Gameplay Tag Structures=Make Array!("[0]"=(TagName="Cosmetics.MetaTags.LazarusLens"), "[1]":Gameplay Tag Structure=(TagName="Cosmetics.MetaTags.LazarusLens.Wing")) -> (Array)
     bRequiresEquipped:bool=false
    ) -> (ReturnValue:bool)) -> (Output_Get:bool)
   n9 = Branch(execute=@n2.then_2, Condition=Get bIsCrouching())
   > Set Sim Space Settings(Sim Space Settings:Sim Space Settings Structure=Get Sim Space Settings_crouch() -> (Sim Space Settings_crouch:Sim Space Settings Structure)) -> (Output_Get:Sim Space Settings Structure)
   Set Sim Space Settings(execute=@n9.else, Sim Space Settings=Get Sim Space Settings_stand() -> (Sim Space Settings_stand:Sim Space Settings Structure))
   n10 = Branch(execute=@n2.then_3, Condition=Get bIsInFrontEnd())
   Set WPO Globals Function(
    execute=[@n10.else, @n48.then]
    Material=Get Mat Instance Dynamic()
    Overlay Material=Get Overlay Mat Instance Dynamic()
    Skeletal Mesh=Get SkelMesh()
    "Local Noise Transform  Location"=Get Local Noise Transform Location() -> (Local Noise Transform Location:Vector)
    Local Noise Offset=Get Local Noise Offset()
   )
   > Set WPO Regional Function!(
    execute
    Material:Material Instance Dynamic Object Reference=Get Mat Instance Dynamic()
    Overlay Material:Material Instance Dynamic Object Reference=Get Overlay Mat Instance Dynamic()
    Skeletal Mesh:Skeletal Mesh Component Object Reference=Get SkelMesh()
    Velocity Transform:Name=foot_r
    Velocity Interp Speed:double=Get Velocity Interp Speed()
    Collision Transform:Name=calf_r
    Collision Transform Axis:EAxis Enum
    Invert Collision Transform Axis:bool=false
    Delta Time X:double=@n1.DeltaTimeX
    Previous Velocity Transform Location:"Vector (by ref)"=Get R1 Velocity Transform Location()
    Previous Velocity Transform Velocity:"Vector (by ref)"=Get R1 Velocity Transform Velocity()
    Region:int=1
   ) -> (then)
   > Set WPO Regional Function(
    Material=Get Mat Instance Dynamic()
    Overlay Material=Get Overlay Mat Instance Dynamic()
    Skeletal Mesh=Get SkelMesh()
    Velocity Transform=foot_l
    Velocity Interp Speed=Get Velocity Interp Speed()
    Collision Transform=calf_l
    Invert Collision Transform Axis=true
    Delta Time X=@n1.DeltaTimeX
    Previous Velocity Transform Location=Get R2 Velocity Transform Location()
    Previous Velocity Transform Velocity=Get R2 Velocity Transform Velocity()
    Region=2
   )
   > n11 = Branch(Condition=Get bIsSkydiveFloating())
   > n12 = Set WPO Pho Wind Intensity(WPO Pho Wind Intensity=0.6)
   SetScalarParameterValue(
    execute=[@n12.then, @n13.then]
    self=[Get Mat Instance Dynamic(), Get Overlay Mat Instance Dynamic()]
    ParameterName=ClothSim_WindStrength
    Value=Get WPO Pho Wind Intensity()
   )
   n13 = Set WPO Pho Wind Intensity(execute=@n11.else, WPO Pho Wind Intensity=0)
   n14 = OR Boolean(A=Get bIsSliding(), B=Get bIsKneeSliding())
   Set sliding_bool(execute=@n2.then_4, sliding_bool:bool=@n14) -> (Output_Get:bool)
   n15 = BreakTransform(InTransform=GetDeltaTransformFromRefPose(self=GetOwnerSkeletalMeshComponent(), BoneName=thigh_twist_01_l, BaseName=pelvis))
   > Set thighs_up_crouch(thighs_up_crouch:double=MapRangeClamped(Value=BreakRotator(InRot=@n15.Rotation).Yaw, InRangeA=110, InRangeB=130)) -> (Output_Get:double)
   n16 = Get bIsParachuteOpened()
   n17 = SelectFloat(A=0.5, B=0, bPickA=AND Boolean(A=Get LL_Backpack_Equipped() -> (LL_Backpack_Equipped:bool), B=NOT Boolean(A=@n16)))
   n18 = SelectFloat(A=0, B=1, bPickA=Get RBAN_alpha() -> (RBAN_alpha:bool))
   > n19 = Set wing_backpack_var(wing_backpack_var:double=SelectFloat(A=@n17, B=@n18, bPickA="float > float"(A=@n17, B=@n18))) -> (Output_Get:double)
   > Set wing_backpack_var_half(wing_backpack_var_half:double=float * float(A=@n19, B=0.5)) -> (Output_Get:double)
   n20 = Get bIsSkyDiving()
   n21 = Get bIsSkydiveFloating()
   n22 = Get bIsSkydiveDiveMode()
   > Set pants_dampen(pants_dampen:double=SelectFloat(A=0.65, B=1, bPickA=OR Boolean!(A=@n20, B=@n21, C:bool=@n22) -> (ReturnValue))) -> (Output_Get:double)
   n23 = AND Boolean!(
    A=OR Boolean!(A=Get bIsJumping(), B=Get bIsFalling()) -> (ReturnValue)
    B=Get bAnimDynamicsEnabled()
    C:bool=NOT Boolean(A=Get bIsSkyDiving())
   ) -> (ReturnValue)
   > Set skirt_jumping(skirt_jumping:double=FInterpTo(Current=Get skirt_jumping() -> (skirt_jumping:double), Target=SelectFloat(A=0, B=1, bPickA=@n23), DeltaTime=0.5, InterpSpeed=0.32)) -> (Output_Get:double)
   n24 = OR Boolean(A=Get bIsSkydiveDiveMode(), B=Get bIsSkydiveFloating())
   > Set skydive_dynamics(skydive_dynamics:Vector=MakeVector(Y=0, Z=SelectFloat(A=1000, B=0, bPickA=@n24))) -> (Output_Get:Vector)
   n25 = GetAnimInstance(self=GetOwnerSkeletalMeshComponent())
   n26 = Is Valid(exec=@n2.then_5, InputObject=@n25)
   n27 = PlayerMovement(
    execute=[@n26.Is Not Valid, @n38.then]
    AnimInstance=Self-Reference()
    XDotProductCoefficient=1
    YDotProductCoefficient=1
    ZDotProductCoefficient=1
   )
   n28 = float + float(
    A=float * float(A="Absolute (Float)"(A=@n27), B=0.5) {is=title K2Node_CommutativeAssociativeBinaryOperator}
    B=MapRangeClamped(Value=@n27.YMovement, InRangeA=0, InRangeB=-1)
   ) {is=title K2Node_CommutativeAssociativeBinaryOperator}
   n29 = Get Is_Custom_Emote_Playing()
   > Set BwdMovement(BwdMovement:float=SelectFloat(A=MapRangeClamped(Value=@n28, InRangeA=0, InRangeB=1.1, OutRangeA=0.3, OutRangeB=0), B=0, bPickA="float > float"(A=@n29))) -> (Output_Get:float)
   n30 = MapRangeClamped(Value=@n27, InRangeA=-0.2, InRangeB=1)
   n31 = float * float(A="Absolute (Float)"(A=@n27.YMovement), B=0.2)
   > Set RightMovement(RightMovement:float=MapRangeClamped(Value=float + float(A=@n31, B=@n30), InRangeA=0, InRangeB=1.2, OutRangeB=0.5)) -> (Output_Get:float)
   n32 = MapRangeClamped(Value=@n27, InRangeA=0.2, InRangeB=-1, OutRangeB=0.3)
   n33 = float * float(A="Absolute (Float)"(A=@n27.YMovement), B=0.2)
   > Set LeftMovement(LeftMovement:float=MapRangeClamped(Value=float + float(A=@n33, B=@n32), InRangeA=0, InRangeB=1.2, OutRangeB=0.5)) -> (Output_Get:float)
   n34 = NOT Boolean(A=Get RBAN_off())
   n35 = Get bIsParachuteOpened()
   > Set Skydive_Parachute(Skydive_Parachute:double=SelectFloat(A=0.75, B=0, bPickA=AND Boolean!(A=@n34, B=@n35) -> (ReturnValue))) -> (Output_Get:double)
   > Set LOD2(LOD2:bool="integer >= integer"(A=Get CurrentLOD(), B=2)) -> (Output_Get:bool)
   n36 = Montage_IsPlaying(self=@n25, Montage=/BRCosmetics/Animation/Game/MainPlayer/Emotes/Donut1/Donut_One_CMM_M_Frontend)
   n37 = Montage_IsPlaying(self=@n25, Montage=/BRCosmetics/Animation/Game/MainPlayer/Emotes/Donut1/Donut_One_CMM_M)
   n38 = Set Is_Custom_Emote_Playing(execute=@n26.Is Valid, Is_Custom_Emote_Playing:double=SelectFloat(A=0, B=0.3, bPickA=OR Boolean(A=@n37, B=@n36))) -> (Output_Get:double)
   n39 = Gravity Override State Function DEPRECATED(
    execute=@n2.then_6
    CharacterPartAnimInstance=Self-Reference()
    GameTimeInSeconds=GetGameTimeInSeconds()
    GravityOverrideParams=Get Gravity Override Params()
   )
   > Set Gravity Override(Gravity Override=@n39)
   Event OnCharacterCustomizationLoaded()
   > n40 = Set SkelMesh(SkelMesh=GetPartSkeletalMeshComponent())
   > n41 = Get Dynamic Material Instance(SkeletalMeshComponent=@n40)
   > Set Mat Instance Dynamic(Mat Instance Dynamic=@n41)
   > n42 = GetOrCreateDynamicOverlayMaterial(MeshComponent=@n40, MaterialName=None)
   > Set Overlay Mat Instance Dynamic(Overlay Mat Instance Dynamic=@n42)
   n43 = Event OnEndSkydiving()
   ResetDynamics(execute=[@n43.then, @n44.then], InTeleportType=ResetPhysics)
   n44 = Event OnParachuteOpened()
   Event BlueprintInitializeAnimation()
   > Set SkelMesh(SkelMesh=GetOwnerSkeletalMeshComponent())
   n45 = TryGetPawnOwner()
   > Is Valid(InputObject=@n45)
   > n46 = Cast To FortPlayerPawnAthena(Object=@n45)
   > Set PlayerPawn(PlayerPawn:Fort Player Pawn Athena Object Reference=@n46) -> (Output_Get:Fort Player Pawn Athena Object Reference)
   n47 = Get bIsSkyDiving()
   n48 = Branch(Condition=@n47)
   Set WPO Pho Wind Intensity(execute=@n48.else, WPO Pho Wind Intensity=0)
   > SetScalarParameterValue(
    self=[Get Mat Instance Dynamic(), Get Overlay Mat Instance Dynamic()]
    ParameterName=ClothSim_WindStrength
    Value=Get WPO Pho Wind Intensity()
   )
   AND Boolean(A=@n14, B=Get bRigidBodyEnabled())
   OR Boolean(A=@n47, B=Get bIsFalling())
   Get LL_Backpack_Equipped()
  ]}
]
variables [
 BwdMovement {type=number}
 crouch move alpha {type=number}
 crouching {type=number}
 Gravity Override {type=Vector}
 Gravity Override Params {type=GravityOverrideParamsStruct}
 Gravity Override_Frontend {type=Vector}
 Is_Custom_Emote_Playing {type=number}
 LeftMovement {type=number}
 LL_Backpack_Equipped {type=boolean}
 Local Noise Offset {type=Vector}
 Local Noise Transform Location {type=Vector}
 LOD2 {type=boolean}
 Mat Instance Dynamic {type=MaterialInstanceDynamic}
 Overlay Mat Instance Dynamic {type=MaterialInstanceDynamic}
 pants_dampen {type=number}
 PlayerPawn {type=FortPlayerPawnAthena}
 R1 Velocity Transform Location {type=Vector}
 R1 Velocity Transform Velocity {type=Vector}
 R2 Velocity Transform Location {type=Vector}
 R2 Velocity Transform Velocity {type=Vector}
 RBAN_alpha {type=boolean}
 RBAN_off {type=boolean}
 RightMovement {type=number}
 Sim Space Settings {
  type=SimSpaceSettings
  default {
   dampingAlpha=1
   maxAngularAcceleration=10000
   maxAngularVelocity=10000
   maxLinearAcceleration=10000
   maxLinearVelocity=10000
   velocityScaleZ=1
   worldAlpha=0
   externalLinearDragV=<0 0 0>
  }
 }
 Sim Space Settings_crouch {
  type=SimSpaceSettings
  default {
   dampingAlpha=1
   maxAngularAcceleration=500
   maxAngularVelocity=3
   maxLinearAcceleration=350
   maxLinearVelocity=250
   velocityScaleZ=0
   worldAlpha=0.2
   externalLinearDragV=<0.2 0.2 0>
  }
 }
 Sim Space Settings_stand {
  type=SimSpaceSettings
  default {
   dampingAlpha=1
   maxAngularAcceleration=1000
   maxAngularVelocity=3
   maxLinearAcceleration=700
   maxLinearVelocity=500
   velocityScaleZ=0
   worldAlpha=0.7
   externalLinearDragV=<0.2 0.2 0>
  }
 }
 SkelMesh {type=SkeletalMeshComponent}
 skirt_jumping {type=number}
 skydive_dynamics {type=Vector}
 Skydive_Parachute {type=number}
 sliding_bool {type=boolean}
 thighs_up_crouch {type=number}
 Velocity Interp Speed {type=number}
 wing_backpack_var {type=number, default=1}
 wing_backpack_var_half {type=number, default=1}
 WPO Pho Wind Intensity {type=number}
]
