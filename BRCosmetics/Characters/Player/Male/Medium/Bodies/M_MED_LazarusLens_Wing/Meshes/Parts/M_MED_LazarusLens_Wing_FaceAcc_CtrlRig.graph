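# Control Rig for the LazarusLens Wing face accessory that procedurally corrects dynamic hair bones as the head and shoulders move.
# Spherical pose readers on head and right clavicle weight offsets on back hair, one passing through AlphaInterp with rBAN (default 0) as output maximum.
# A second sequence branch blends front strands across spine, neck, clavicle and head via parent constraints, adding fixed or inverse-pose-weighted offsets.
# A left-clavicle upperarm chain is disconnected from execution.
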ControlRigBlueprint /BRCosmetics/Characters/Player/Male/Medium/Bodies/M_MED_LazarusLens_Wing/Meshes/Parts/M_MED_LazarusLens_Wing_FaceAcc_CtrlRig
graphs [
 RigVMFunctionLibrary {links [], nodes []}
 RigVMModel {nodes [
   RigUnit_BeginExecution = &RigUnit_BeginExecution()
   SequenceAggregate = &RigVMFunction_Sequence(ExecuteContext=@RigUnit_BeginExecution.ExecutePin)
   n1 = &RigUnit_SphericalPoseReader(
    ActiveRegionScaleFactors=(PositiveWidth=2.5,NegativeWidth=3,PositiveHeight=1,NegativeHeight=3)
    ActiveRegionSize=0.1
    Debug=(bDrawDebug=false,DebugScale=25,DebugSegments=20,DebugThickness=0.25)
    DriverAxis=(X=1,Y=0,Z=0)
    DriverItem=(Type=Bone,Name="head")
    FalloffRegionScaleFactors=(PositiveWidth=1,NegativeWidth=0.5,PositiveHeight=0.5,NegativeHeight=0.5)
    FalloffSize=0.5
    FlipHeightScaling=true
    OptionalParentItem=(Type=Bone,Name="spine_05")
    RotationOffset=(X=-35,Y=0,Z=0)
    ExecutePin=@SequenceAggregate.A
   )
   > &RigUnit_ModifyTransforms(
    ItemToModify="((Item=(Type=Bone,Name=\"dyn_Hair_BackR_03\"),Transform=(Rotation=(X=0,Y=0,Z=0,W=1),Translation=(X=-1,Y=-3,Z=2),Scale3D=(X=1,Y=1,Z=1))),(Item=(Type=Bone,Name=\"dyn_Hair_BackR_01\"),Transform=(Rotation=(X=0,Y=0,Z=0,W=1),Translation=(X=2,Y=-2,Z=0),Scale3D=(X=1,Y ...(truncated)"
    Weight=@n1.OutputParam
   )
   > n2 = &RigUnit_SphericalPoseReader(
    ActiveRegionScaleFactors=(PositiveWidth=1,NegativeWidth=1,PositiveHeight=1,NegativeHeight=1)
    ActiveRegionSize=0.02
    Debug=(bDrawDebug=true,bDraw2D=false,bDrawLocalAxes=true,DebugScale=25,DebugSegments=12,DebugThickness=0.15)
    DriverAxis=(X=-1,Y=0,Z=0)
    DriverItem=(Type=Bone,Name="clavicle_r")
    FalloffRegionScaleFactors=(PositiveWidth=1,NegativeWidth=1,PositiveHeight=1,NegativeHeight=1)
    FalloffSize=0.15
    FlipWidthScaling=false
    OptionalParentItem=(Type=Bone,Name="spine_05")
    RotationOffset=(X=0,Y=125,Z=-5)
   )
   > &RigUnit_ModifyTransforms(
    ItemToModify=((Item=(Type=Bone,Name="dyn_Hair_BackR_03"),Transform=(Rotation=(X=0,Y=0,Z=0,W=1),Translation=(X=0,Y=-2,Z=2),Scale3D=(X=1,Y=1,Z=1))))
    Weight=@n2.OutputParam
   )
   > n3 = &RigUnit_SphericalPoseReader(
    ActiveRegionScaleFactors=(PositiveWidth=1,NegativeWidth=1,PositiveHeight=1,NegativeHeight=0.5)
    ActiveRegionSize=0.09
    Debug=(bDrawDebug=false,DebugScale=25,DebugSegments=20,DebugThickness=0.25)
    DriverAxis=(X=0,Y=1,Z=0)
    DriverItem=(Type=Bone,Name="head")
    FalloffRegionScaleFactors=(PositiveWidth=1,NegativeWidth=1,PositiveHeight=1.1,NegativeHeight=0.7)
    FalloffSize=0.3
    OptionalParentItem=(Type=Bone,Name="spine_05")
    RotationOffset=(X=125,Y=0,Z=0)
   )
   AlphaInterp = &RigVMFunction_AlphaInterp(
    InterpSpeedDecreasing=10
    InterpSpeedIncreasing=10
    Scale=1
    bClampResult=true
    bInterpResult=False
    bMapRange=true
    Value=@n3.OutputParam
    OutRange.Max=RigVMVariableNode() {variable=rBAN}.Value
   )
   > &RigUnit_ModifyTransforms(
    ItemToModify="((Item=(Type=Bone,Name=\"dyn_Hair_BackR_02\"),Transform=(Rotation=(X=0,Y=0,Z=0,W=1),Translation=(X=0.5,Y=1,Z=-1),Scale3D=(X=1,Y=1,Z=1))),(Item=(Type=Bone,Name=\"dyn_Hair_Back_02\"),Transform=(Rotation=(X=0,Y=0,Z=0,W=1),Translation=(X=2,Y=1.5,Z=-1),Scale3D=(X=1 ...(truncated)"
    Weight=@AlphaInterp.Result
    ItemToModify.0.Transform.Rotation=&RigVMFunction_MathQuaternionFromEuler(Euler=(X=0,Y=-5,Z=0)).Result
    ItemToModify.1.Transform.Rotation=&RigVMFunction_MathQuaternionFromEuler(Euler=(X=0,Y=5,Z=0)).Result
    ItemToModify.2.Transform.Rotation=&RigVMFunction_MathQuaternionFromEuler(Euler=(X=5,Y=0,Z=0)).Result
    ItemToModify.3.Transform.Rotation=&RigVMFunction_MathQuaternionFromEuler(Euler=(X=0,Y=-5,Z=0)).Result
   )
   > &RigUnit_ModifyTransforms(
    ItemToModify=((Item=(Type=Bone,Name="dyn_Hair_BackR_03"),Transform=(Rotation=(X=0,Y=0,Z=0,W=1),Translation=(X=0,Y=0,Z=0),Scale3D=(X=1,Y=1,Z=1))))
    Weight=@AlphaInterp.Result
   )
   > &RigUnit_ModifyTransforms(
    ItemToModify="((Item=(Type=Bone,Name=\"dyn_Hair_BackR_05\"),Transform=(Rotation=(X=0,Y=0,Z=0,W=1),Translation=(X=-4,Y=0,Z=0),Scale3D=(X=1,Y=1,Z=1))),(Item=(Type=Bone,Name=\"dyn_Hair_Back_05\"),Transform=(Rotation=(X=0,Y=0,Z=0,W=1),Translation=(X=-3,Y=0,Z=0),Scale3D=(X=1,Y=1 ...(truncated)"
    Weight=@AlphaInterp.Result
    ItemToModify.0.Transform.Rotation=&RigVMFunction_MathQuaternionFromEuler(Euler=(X=0,Y=0,Z=7)).Result
    ItemToModify.3.Transform.Rotation=&RigVMFunction_MathQuaternionFromEuler(Euler=(X=0,Y=0,Z=-10)).Result
   )
   > n4 = &RigUnit_SphericalPoseReader(
    ActiveRegionScaleFactors=(PositiveWidth=2.5,NegativeWidth=3,PositiveHeight=1,NegativeHeight=3)
    ActiveRegionSize=0.1
    Debug=(bDrawDebug=false,DebugScale=25,DebugSegments=20,DebugThickness=0.25)
    DriverAxis=(X=1,Y=0,Z=0)
    DriverItem=(Type=Bone,Name="head")
    FalloffRegionScaleFactors=(PositiveWidth=1,NegativeWidth=0.5,PositiveHeight=0.5,NegativeHeight=0.6)
    FalloffSize=0.5
    FlipHeightScaling=true
    OptionalParentItem=(Type=Bone,Name="spine_05")
    RotationOffset=(X=0,Y=-90,Z=-90)
   )
   > &RigUnit_ModifyTransforms(
    ItemToModify="((Item=(Type=Bone,Name=\"dyn_Hair_BackL_02\"),Transform=(Rotation=(X=0,Y=0,Z=0,W=1),Translation=(X=1.5,Y=0,Z=0),Scale3D=(X=1,Y=1,Z=1))),(Item=(Type=Bone,Name=\"dyn_Hair_BackL_03\"),Transform=(Rotation=(X=0,Y=0,Z=0,W=1),Translation=(X=0,Y=0,Z=2),Scale3D=(X=1,Y= ...(truncated)"
    WeightMinimum=0
    Weight=@n4.OutputParam
   )
   > &RigUnit_ModifyTransforms(
    ItemToModify=((Item=(Type=Bone,Name="dyn_Hair_BackR_02"),Transform=(Rotation=(X=0,Y=0,Z=0,W=1),Translation=(X=0.5,Y=0,Z=0),Scale3D=(X=1,Y=1,Z=1))))
    WeightMinimum=0
    Weight=@n4.OutputParam
   )
   &RigUnit_ParentConstraint(
    AdvancedSettings=(InterpolationType=Average,RotationOrderForFilter=XZY)
    Child=(Type=Bone,Name="dyn_hair_r_Fr_00")
    Parents=((Item=(Type=Bone,Name="spine_05"),Weight=0.5),(Item=(Type=Bone,Name="head"),Weight=0.5))
    Weight=1
    ExecutePin=@SequenceAggregate.B
   )
   > &RigUnit_ParentConstraint(
    AdvancedSettings=(InterpolationType=Average,RotationOrderForFilter=XZY)
    Child=(Type=Bone,Name="dyn_hair_l_Front_00")
    Parents=((Item=(Type=Bone,Name="spine_05"),Weight=0.6),(Item=(Type=Bone,Name="head"),Weight=0.4))
    Weight=1
   )
   > &RigUnit_ParentConstraint(
    AdvancedSettings=(InterpolationType=Average,RotationOrderForFilter=XZY)
    Child=(Type=Bone,Name="dyn_hair_l_Front_01")
    Parents=((Item=(Type=Bone,Name="spine_05"),Weight=0.22),(Item=(Type=Bone,Name="neck_01"),Weight=0.3),(Item=(Type=Bone,Name="neck_02"),Weight=0.24),(Item=(Type=Bone,Name="clavicle_l"),Weight=0.14),(Item=(Type=Bone,Name="head"),Weight=0.1))
    Weight=1
   )
   > n5 = &RigUnit_SphericalPoseReader(
    ActiveRegionScaleFactors=(PositiveWidth=2.5,NegativeWidth=3,PositiveHeight=1,NegativeHeight=3)
    ActiveRegionSize=0.1
    Debug=(bDrawDebug=false,DebugScale=25,DebugSegments=20,DebugThickness=0.25)
    DriverAxis=(X=1,Y=0,Z=0)
    DriverItem=(Type=Bone,Name="head")
    FalloffRegionScaleFactors=(PositiveWidth=1,NegativeWidth=0.5,PositiveHeight=0.5,NegativeHeight=0.5)
    FalloffSize=0.5
    FlipHeightScaling=true
    OptionalParentItem=(Type=Bone,Name="spine_05")
    RotationOffset=(X=-35,Y=0,Z=0)
   )
   > &RigUnit_ModifyTransforms(
    ItemToModify=((Item=(Type=Bone,Name="dyn_hair_l_Front_01"),Transform=(Rotation=(X=0,Y=0,Z=0,W=1),Translation=(X=0,Y=0,Z=1.5),Scale3D=(X=1,Y=1,Z=1))))
    Weight=1
    WeightMinimum=0
    ItemToModify.0.Transform.Rotation=&RigVMFunction_MathQuaternionFromEuler(Euler=(X=-15,Y=0,Z=0)).Result
   )
   > &RigUnit_ParentConstraint(
    AdvancedSettings=(InterpolationType=Average,RotationOrderForFilter=XZY)
    Child=(Type=Bone,Name="dyn_hair_l_Front_02")
    Parents=((Item=(Type=Bone,Name="spine_05"),Weight=0.53),(Item=(Type=Bone,Name="neck_01"),Weight=0.02),(Item=(Type=Bone,Name="neck_02"),Weight=0.05),(Item=(Type=Bone,Name="clavicle_l"),Weight=0.32),(Item=(Type=Bone,Name="head"),Weight=0.08))
    Weight=1
   )
   > &RigUnit_ModifyTransforms(
    ItemToModify=((Item=(Type=Bone,Name="dyn_hair_l_Front_02"),Transform=(Rotation=(X=0,Y=0,Z=0,W=1),Translation=(X=0,Y=0,Z=1.5),Scale3D=(X=1,Y=1,Z=1))))
    WeightMinimum=0
    ItemToModify.0.Transform.Rotation=&RigVMFunction_MathQuaternionFromEuler(Euler=(X=-15,Y=0,Z=0)).Result
    Weight=&RigVMFunction_MathDoubleSub(A=1, B=@n5.OutputParam).Result
   )
   > &RigUnit_ParentConstraint(
    AdvancedSettings=(InterpolationType=Average,RotationOrderForFilter=XZY)
    Child=(Type=Bone,Name="dyn_hair_r_Fr_01")
    Parents=((Item=(Type=Bone,Name="spine_05"),Weight=0.3),(Item=(Type=Bone,Name="neck_01"),Weight=0.3),(Item=(Type=Bone,Name="neck_02"),Weight=0.18),(Item=(Type=Bone,Name="clavicle_r"),Weight=0.18),(Item=(Type=Bone,Name="head"),Weight=0.04))
    Weight=1
   )
   > &RigUnit_ParentConstraint(
    AdvancedSettings=(InterpolationType=Average,RotationOrderForFilter=XZY)
    Child=(Type=Bone,Name="dyn_hair_r_F_02")
    Parents=((Item=(Type=Bone,Name="spine_05"),Weight=0.63),(Item=(Type=Bone,Name="neck_01"),Weight=0.03),(Item=(Type=Bone,Name="neck_02"),Weight=0.07),(Item=(Type=Bone,Name="clavicle_r"),Weight=0.22),(Item=(Type=Bone,Name="spine_04"),Weight=0.05))
    Weight=1
   )
   > n6 = &RigUnit_SphericalPoseReader(
    ActiveRegionScaleFactors=(PositiveWidth=2.5,NegativeWidth=3,PositiveHeight=1,NegativeHeight=3)
    ActiveRegionSize=0.1
    Debug=(bDrawDebug=false,DebugScale=25,DebugSegments=20,DebugThickness=0.25)
    DriverAxis=(X=1,Y=0,Z=0)
    DriverItem=(Type=Bone,Name="head")
    FalloffRegionScaleFactors=(PositiveWidth=1,NegativeWidth=0.5,PositiveHeight=0.5,NegativeHeight=0.5)
    FalloffSize=0.5
    FlipHeightScaling=true
    OptionalParentItem=(Type=Bone,Name="spine_05")
    RotationOffset=(X=-35,Y=0,Z=0)
   )
   n7 = &RigVMFunction_MathDoubleSub(A=1, B=@n6.OutputParam)
   > &RigUnit_ModifyTransforms(
    ItemToModify=((Item=(Type=Bone,Name="dyn_hair_r_F_02"),Transform=(Rotation=(X=0,Y=0,Z=0,W=1),Translation=(X=0,Y=0,Z=0.5),Scale3D=(X=1,Y=1,Z=1))))
    WeightMinimum=0
    Weight=@n7.Result
   )
   > &RigUnit_ParentConstraint(
    AdvancedSettings=(InterpolationType=Average,RotationOrderForFilter=XZY)
    Child=(Type=Bone,Name="dyn_hair_r_Fr_03")
    Parents=((Item=(Type=Bone,Name="spine_05"),Weight=0.63),(Item=(Type=Bone,Name="neck_01"),Weight=0.03),(Item=(Type=Bone,Name="neck_02"),Weight=0.07),(Item=(Type=Bone,Name="clavicle_r"),Weight=0.22),(Item=(Type=Bone,Name="spine_04"),Weight=0.05))
    Weight=1
   )
   > &RigUnit_ModifyTransforms(
    ItemToModify=((Item=(Type=Bone,Name="dyn_hair_r_Fr_03"),Transform=(Rotation=(X=0,Y=0,Z=0,W=1),Translation=(X=0,Y=0,Z=1),Scale3D=(X=1,Y=1,Z=1))))
    WeightMinimum=0
    Weight=@n7.Result
   )
   n8 = &RigUnit_SphericalPoseReader(
    ActiveRegionScaleFactors=(PositiveWidth=1,NegativeWidth=1,PositiveHeight=1,NegativeHeight=1)
    ActiveRegionSize=0.2
    Debug=(bDrawDebug=true,bDraw2D=false,bDrawLocalAxes=true,DebugScale=25,DebugSegments=12,DebugThickness=0.15)
    DriverAxis=(X=1,Y=0,Z=0)
    DriverItem=(Type=Bone,Name="clavicle_l")
    ExecutePin=()
    FalloffRegionScaleFactors=(PositiveWidth=1,NegativeWidth=1,PositiveHeight=1,NegativeHeight=1)
    FalloffSize=0.35
    FlipWidthScaling=false
    OptionalParentItem=(Type=Bone,Name="spine_05")
    RotationOffset=(X=120,Y=0,Z=-60)
   )
   > &RigUnit_ModifyTransforms(
    ItemToModify=((Item=(Type=Bone,Name="upperarm_l"),Transform=(Rotation=(X=0,Y=0,Z=0,W=1),Translation=(X=0,Y=0,Z=0),Scale3D=(X=1,Y=1,Z=1))))
    WeightMinimum=0
    Weight=@n8.OutputParam
   )
   &RigVMFunction_MathQuaternionFromEuler(Euler=(X=0,Y=0,Z=-10))
  ]}
]
hierarchy [
 root {loc=<0 0 0>, rot=<0 0 0 1>}
 pelvis {parent=root, loc=<0 -0.7071 93.1191>, rot=<0.0127 0.707 -0.0127 -0.707>}
 spine_01 {parent=pelvis, loc=<6.7494 -0.6388 0>, rot=<0 0 0.034 -0.9994>}
 spine_02 {parent=spine_01, loc=<7.2518 -0.035 0>, rot=<0 0 0.0238 -0.9997>}
 spine_03 {parent=spine_02, loc=<6.0104 0 0>, rot=<0 0 0.0149 -0.9999>}
 spine_04 {parent=spine_03, loc=<8.1024 0 0>, rot=<0 0 -0.0132 -0.9999>}
 spine_05 {parent=spine_04, loc=<10.0146 0 0>, rot=<0 0 -0.0015 -1>}
 neck_01 {parent=spine_05, loc=<11.3138 0.1529 0>, rot=<0 0 -0.2003 -0.9797>}
 neck_02 {parent=neck_01, loc=<4.5627 0 0>, rot=<0 0 0.0504 -0.9987>}
 head {parent=neck_02, loc=<5.3914 0 0>, rot=<0 0 0.115 -0.9934>}
 dyn_Hair_BackR_01 {parent=head, loc=<8.1854 -0.439 4.9819>, rot=<0 0 0.0928 -0.9957>}
 dyn_Hair_BackR_02 {parent=dyn_Hair_BackR_01, loc=<-1.363 -5.3153 0>, rot=<0 0 0.5593 -0.8289>}
 dyn_Hair_BackR_03 {parent=dyn_Hair_BackR_02, loc=<0.1785 -6.9081 0.4801>, rot=<0.5973 -0.4046 0.4308 -0.5421>}
 dyn_Hair_BackR_04 {parent=dyn_Hair_BackR_03, loc=<9.3698 0 0>, rot=<-0.0022 -0.0674 -0.0049 -0.9977>}
 dyn_Hair_BackR_05 {parent=dyn_Hair_BackR_04, loc=<7.1626 0 0>, rot=<-0.0058 -0.1471 -0.0268 -0.9887>}
 dyn_Hair_Back_01 {parent=head, loc=<7.9815 -4.0299 0.5844>, rot=<0 0 0.0928 -0.9957>}
 dyn_Hair_Back_02 {parent=dyn_Hair_Back_01, loc=<-1.9753 -3.7984 0>, rot=<0 0 0.5593 -0.8289>}
 dyn_Hair_Back_03 {parent=dyn_Hair_Back_02, loc=<0 -6.8979 0>, rot=<0.5614 -0.4324 0.4356 -0.5551>}
 dyn_Hair_Back_04 {parent=dyn_Hair_Back_03, loc=<9.3289 0 0>, rot=<-0.0113 -0.1192 -0.0344 -0.9922>}
 dyn_Hair_BackL_01 {parent=head, loc=<7.9722 -1.0353 -4.0889>, rot=<0 0 0.0928 -0.9957>}
 dyn_Hair_BackL_02 {parent=dyn_Hair_BackL_01, loc=<-1.3095 -5.4477 0>, rot=<0 0 0.5593 -0.8289>}
 dyn_Hair_BackL_03 {parent=dyn_Hair_BackL_02, loc=<0.5638 -6.905 0>, rot=<0.564 -0.4214 0.4157 -0.5758>}
 dyn_Hair_BackL_04 {parent=dyn_Hair_BackL_03, loc=<9.4223 0 0>, rot=<-0.02 -0.064 -0.0437 -0.9968>}
 dyn_hair_l_Front_MSTR {parent=head, loc=<0.2137 -2.9751 -6.9207>, rot=<-0.0621 -0.9924 -0.0013 -0.1066>}
 dyn_hair_l_Front_00 {parent=dyn_hair_l_Front_MSTR, loc=<0.5 0 0>, rot=<0.6937 0.036 -0.1276 -0.708>}
 dyn_hair_l_Front_01 {parent=dyn_hair_l_Front_00, loc=<6.4093 0 0>, rot=<0.0223 0.1351 0.0408 -0.9897>}
 dyn_hair_l_Front_02 {parent=dyn_hair_l_Front_01, loc=<4.6342 0 0>, rot=<-0.063 -0.0235 -0.0895 -0.9937>}
 dyn_hair_r_Fr_MSTR {parent=head, loc=<0.2983 -2.6068 6.7103>, rot=<-0.0806 -0.9889 0.0111 0.124>}
 dyn_hair_r_Fr_00 {parent=dyn_hair_r_Fr_MSTR, loc=<0.4731 0 0>, rot=<0.7134 0.0843 -0.0109 -0.6956>}
 dyn_hair_r_Fr_01 {parent=dyn_hair_r_Fr_00, loc=<3.644 0 0>, rot=<0.0399 0.1057 0.0913 -0.9894>}
 dyn_hair_r_F_02 {parent=dyn_hair_r_Fr_01, loc=<4.9408 0 0>, rot=<0.0306 0.0521 0.0476 -0.997>}
 dyn_hair_r_Fr_03 {parent=dyn_hair_r_F_02, loc=<4.4193 0 0>, rot=<0.0187 -0.034 0.0281 -0.9989>}
 dyn_hair_r_Fr_04 {parent=dyn_hair_r_Fr_03, loc=<5.1383 0 0>, rot=<-0.729 -0.0901 0.0799 -0.6739>}
 clavicle_l {parent=spine_05, loc=<4.3575 -0.1466 -2.3223>, rot=<0.0465 -0.7289 -0.009 -0.6829>}
 upperarm_l {parent=clavicle_l, loc=<15.5111 0 0>, rot=<0.0226 -0.3234 -0.0088 -0.9459>}
 upperarm_twist_01_l {parent=upperarm_l, loc=<0.0041 0 0>, rot=<0 0 0 1>}
 clavicle_r {parent=spine_05, loc=<4.3575 -0.1466 2.3223>, rot=<0.7289 0.0465 -0.6829 0.009>}
 upperarm_r {parent=clavicle_r, loc=<-15.5111 0 0>, rot=<0.0226 -0.3234 -0.0088 -0.9459>}
 upperarm_twist_01_r {parent=upperarm_r, loc=<-0.0041 0 0>, rot=<0 0 0 -1>}
]
variables [rBAN {type=double, default=0}]
