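# Physics asset for the LazarusLens Wing male character's FaceAcc mesh, defining capsule bodies for dynamic hair and the upper body it can collide with.
# Three five-link back chains and two front chains (left and right) hang from the head through root bodies, with per-joint angular limits shaping their bend.
# Spine, neck, head, clavicle and upper-arm capsules link with free rotation.
# Extra constraints also tie spine_05 to a link in each front chain.
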PhysicsAsset /BRCosmetics/Characters/Player/Male/Medium/Bodies/M_MED_LazarusLens_Wing/Meshes/Parts/M_MED_LazarusLens_Wing_FaceAcc_Physics
bodies {
 clavicle_l [{
   shapeType=Capsule
   length=8.4916
   radius=6.0166
   shapeName=clavicle_r_capsule
   center=<13.46 -1.3988 3.3355>
   rotation=<rot 70.1361 46.4862 -132.5387>
  }]
 clavicle_r [{
   shapeType=Capsule
   length=9.3765
   radius=6.1872
   shapeName=clavicle_r_capsule
   center=<-12.7024 1.0712 -3.3924>
   rotation=<rot 73.0182 58.6454 -120.9946>
  }]
 dyn_Hair_BackL_01 [{shapeType=Capsule, length=0.0017, radius=3.0581, shapeName=dyn_Hair_BackL_01_capsule}]
 dyn_Hair_BackL_02 [{
   shapeType=Capsule
   length=4.8415
   radius=6.1327
   shapeName=dyn_Hair_BackL_02_capsule
   center=<0 -2.4043 0.0611>
   rotation=<rot 0 -0.0001 91.4546>
  }]
 dyn_Hair_BackL_03 [{shapeType=Capsule, length=5.0237, radius=6.2383, shapeName=dyn_Hair_BackR_04_capsule, center=<4.2872 0 3>, rotation=<rot 0 0 80.9436>}]
 dyn_Hair_BackL_04 [{shapeType=Capsule, length=6.3842, radius=6.6339, shapeName=dyn_Hair_BackL_04_capsule, center=<7.2432 0 3>, rotation=<rot 82.7733 0 0>}]
 dyn_Hair_BackL_05 [{shapeType=Capsule, length=0, radius=6.8744, shapeName=dyn_Hair_BackL_04_capsule, center=<4.2975 0 3>, rotation=<rot 90 0 0>}]
 dyn_Hair_BackR_01 [{shapeType=Capsule, length=0.0017, radius=3.0581, shapeName=dyn_Hair_BackR_01_capsule}]
 dyn_Hair_BackR_02 [{
   shapeType=Capsule
   length=4.8415
   radius=6.1327
   shapeName=dyn_Hair_BackR_02_capsule
   center=<0 -2.4043 0.0611>
   rotation=<rot 0 -0.0001 91.4546>
  }]
 dyn_Hair_BackR_03 [{shapeType=Capsule, length=5.5962, radius=5.6658, shapeName=dyn_Hair_BackR_04_capsule, center=<4.2872 0 3>, rotation=<rot -0.0001 0 75.6705>}]
 dyn_Hair_BackR_04 [{shapeType=Capsule, length=6.593, radius=6.4251, shapeName=dyn_Hair_BackL_04_capsule, center=<7.2173 0 2.796>, rotation=<rot 82.7733 0 0>}]
 dyn_Hair_BackR_05 [{shapeType=Capsule, length=0, radius=6.8744, shapeName=dyn_Hair_BackL_04_capsule, center=<4.2975 0 3>, rotation=<rot 90 0 0>}]
 dyn_Hair_Back_01 [{shapeType=Capsule, length=0.0017, radius=3.0581, shapeName=dyn_Hair_Back_01_capsule}]
 dyn_Hair_Back_02 [{
   shapeType=Capsule
   length=4.8415
   radius=6.1327
   shapeName=dyn_Hair_Back_02_capsule
   center=<0 -2.4043 0.0611>
   rotation=<rot 0 -0.0001 91.4546>
  }]
 dyn_Hair_Back_03 [{shapeType=Capsule, length=5.0237, radius=6.2383, shapeName=dyn_Hair_BackR_04_capsule, center=<4.2872 0 3>, rotation=<rot 0 0 80.9436>}]
 dyn_Hair_Back_04 [{shapeType=Capsule, length=6.3842, radius=6.6339, shapeName=dyn_Hair_BackL_04_capsule, center=<7.2432 0 3>, rotation=<rot 82.7733 0 0>}]
 dyn_Hair_Back_05 [{shapeType=Capsule, length=0, radius=6.8744, shapeName=dyn_Hair_BackL_04_capsule, center=<4.2975 0 3>, rotation=<rot 90 0 0>}]
 dyn_hair_l_Front_00 [{shapeType=Capsule, length=2.75, radius=2.25, shapeName=dyn_hair_l_Front_01_capsule, center=<3.6882 0 0>, rotation=<rot 90 0 0>}]
 dyn_hair_l_Front_01 [{shapeType=Capsule, length=2.75, radius=2.25, shapeName=dyn_hair_l_Front_01_capsule, center=<3.6882 0 0>, rotation=<rot 90 0 0>}]
 dyn_hair_l_Front_02 [{shapeType=Capsule, length=4.076, radius=1.8238, shapeName=dyn_hair_l_Front_02_capsule, center=<6.4546 0 0>, rotation=<rot 90 0 0>}]
 dyn_hair_l_Front_MSTR [{shapeType=Capsule, length=0.505, radius=0.505, shapeName=dyn_hair_l_Front_MSTR_capsule, rotation=<rot 0 -90 0>}]
 dyn_hair_r_F_02 [{shapeType=Capsule, length=2.375, radius=2.375, shapeName=dyn_hair_r_Front_03_capsule, center=<3.5 0 0>, rotation=<rot 90 0 0>}]
 dyn_hair_r_Fr_00 [{shapeType=Capsule, length=1.875, radius=3.125, shapeName=dyn_hair_r_Front_01_capsule, center=<3 0 0>, rotation=<rot 90 0 0>}]
 dyn_hair_r_Fr_01 [{shapeType=Capsule, length=1.875, radius=3.125, shapeName=dyn_hair_r_Front_01_capsule, center=<3 0 0>, rotation=<rot 90 0 0>}]
 dyn_hair_r_Fr_03 [{shapeType=Capsule, length=2.375, radius=2.375, shapeName=dyn_hair_r_Front_03_capsule, center=<3.5 0 0>, rotation=<rot 90 0 0>}]
 dyn_hair_r_Fr_04 [{shapeType=Capsule, length=1.5281, radius=1.875, shapeName=dyn_hair_r_Front_04_capsule, center=<3.5 0 0>, rotation=<rot 90 0 0>}]
 dyn_hair_r_Fr_MSTR [{shapeType=Capsule, length=0.505, radius=0.505, shapeName=dyn_hair_r_Fr_MSTR_capsule, rotation=<rot 0 -90 0>}]
 head [
  {shapeType=Capsule, length=4.29, radius=5.96, shapeName=head_capsule, center=<2.0202 0.1733 -0.2592>, rotation=<rot 90 -0.02 -0.0202>}
  {shapeType=Capsule, length=7.8888, radius=2.3612, shapeName=head_capsule, center=<-1.8806 6.8693 -0.2592>, rotation=<rot 90 180 -158.1663>}
 ]
 neck_01 [{shapeType=Capsule, length=18.4006, radius=4.088, shapeName=neck_01_capsule, center=<-0.4323 -1.1063 0>, rotation=<rot 0 -90 0>}]
 neck_02 [{shapeType=Capsule, length=0.3859, radius=4.9321, shapeName=neck_01_capsule, center=<-0.4323 -1.1063 0>, rotation=<rot 90 0 0>}]
 spine_04 [{shapeType=Capsule, length=43.2801, radius=6.4777, shapeName=spine_04_capsule, center=<1.2998 0.4464 0>}]
 spine_05 [{shapeType=Capsule, length=19.8868, radius=8.9285, shapeName=spine_05_capsule, center=<2.0565 -1.6824 0>}]
 upperarm_l [{shapeType=Capsule, length=42, radius=6.25, shapeName=upperarm_l_capsule, center=<25.2818 0 -2.8292>, rotation=<rot 83.6148 -0.0003 -0.0004>}]
 upperarm_r [{shapeType=Capsule, length=42, radius=6.25, shapeName=upperarm_l_capsule, center=<-24.8822 0 2.7845>, rotation=<rot 83.6148 -0.0003 -0.0004>}]
}
constraints [
 {
  bone1Name=spine_05
  bone2Name=spine_04
  swing1LimitDegrees=45
  swing1Motion=Free
  swing2LimitDegrees=45
  swing2Motion=Free
  twistLimitDegrees=45
  twistMotion=Free
 }
 {
  bone1Name=dyn_Hair_BackL_01
  bone2Name=head
  swing1LimitDegrees=45
  swing1Motion=Free
  swing2LimitDegrees=45
  swing2Motion=Free
  twistLimitDegrees=45
  twistMotion=Free
 }
 {bone1Name=dyn_Hair_BackL_02, bone2Name=dyn_Hair_BackL_01, swing1LimitDegrees=60, swing2LimitDegrees=20, twistLimitDegrees=40}
 {bone1Name=dyn_Hair_BackL_03, bone2Name=dyn_Hair_BackL_02, swing1LimitDegrees=25, swing2LimitDegrees=45, twistLimitDegrees=10}
 {bone1Name=dyn_Hair_BackL_04, bone2Name=dyn_Hair_BackL_03, swing1LimitDegrees=35, swing2LimitDegrees=60, twistLimitDegrees=30}
 {
  bone1Name=dyn_Hair_BackR_01
  bone2Name=head
  swing1LimitDegrees=45
  swing1Motion=Free
  swing2LimitDegrees=45
  swing2Motion=Free
  twistLimitDegrees=45
  twistMotion=Free
 }
 {bone1Name=dyn_Hair_BackR_02, bone2Name=dyn_Hair_BackR_01, swing1LimitDegrees=20, swing2LimitDegrees=10, twistLimitDegrees=20}
 {bone1Name=dyn_Hair_BackR_03, bone2Name=dyn_Hair_BackR_02, swing1LimitDegrees=25, swing2LimitDegrees=45, twistLimitDegrees=30}
 {bone1Name=dyn_Hair_BackR_04, bone2Name=dyn_Hair_BackR_03, swing1LimitDegrees=35, swing2LimitDegrees=45, twistLimitDegrees=30}
 {
  bone1Name=dyn_Hair_Back_01
  bone2Name=head
  swing1LimitDegrees=45
  swing1Motion=Free
  swing2LimitDegrees=45
  swing2Motion=Free
  twistLimitDegrees=45
  twistMotion=Free
 }
 {bone1Name=dyn_Hair_Back_02, bone2Name=dyn_Hair_Back_01, swing1LimitDegrees=20, swing2LimitDegrees=10, twistLimitDegrees=20}
 {bone1Name=dyn_Hair_Back_03, bone2Name=dyn_Hair_Back_02, swing1LimitDegrees=10, swing2LimitDegrees=45, twistLimitDegrees=10}
 {bone1Name=dyn_Hair_Back_04, bone2Name=dyn_Hair_Back_03, swing1LimitDegrees=35, swing2LimitDegrees=60, twistLimitDegrees=30}
 {bone1Name=dyn_hair_l_Front_00, bone2Name=dyn_hair_l_Front_MSTR, swing1LimitDegrees=20, swing2LimitDegrees=60, twistLimitDegrees=20}
 {bone1Name=dyn_hair_l_Front_01, bone2Name=dyn_hair_l_Front_00, swing1LimitDegrees=32, swing2LimitDegrees=60, twistLimitDegrees=20}
 {bone1Name=dyn_hair_l_Front_02, bone2Name=dyn_hair_l_Front_01, swing1LimitDegrees=35, swing2LimitDegrees=60, twistLimitDegrees=20}
 {bone1Name=dyn_hair_r_Fr_00, bone2Name=dyn_hair_r_Fr_MSTR, swing1LimitDegrees=15, swing2LimitDegrees=50, twistLimitDegrees=10}
 {bone1Name=dyn_hair_r_Fr_01, bone2Name=dyn_hair_r_Fr_00, swing1LimitDegrees=18, swing2LimitDegrees=45, twistLimitDegrees=10}
 {bone1Name=dyn_hair_r_F_02, bone2Name=dyn_hair_r_Fr_01, swing1LimitDegrees=20, swing2LimitDegrees=40, twistLimitDegrees=10}
 {bone1Name=dyn_hair_r_Fr_03, bone2Name=dyn_hair_r_F_02, swing1LimitDegrees=25, swing2LimitDegrees=45, twistLimitDegrees=10}
 {bone1Name=dyn_hair_r_Fr_04, bone2Name=dyn_hair_r_Fr_03, swing1LimitDegrees=35, swing2LimitDegrees=15, twistLimitDegrees=10}
 {
  bone1Name=dyn_hair_r_Fr_MSTR
  bone2Name=head
  swing1LimitDegrees=45
  swing1Motion=Free
  swing2LimitDegrees=45
  swing2Motion=Free
  twistLimitDegrees=45
  twistMotion=Free
 }
 {
  bone1Name=dyn_hair_l_Front_MSTR
  bone2Name=head
  swing1LimitDegrees=45
  swing1Motion=Free
  swing2LimitDegrees=45
  swing2Motion=Free
  twistLimitDegrees=45
  twistMotion=Free
 }
 {
  bone1Name=clavicle_l
  bone2Name=spine_05
  swing1LimitDegrees=45
  swing1Motion=Free
  swing2LimitDegrees=45
  swing2Motion=Free
  twistLimitDegrees=45
  twistMotion=Free
 }
 {
  bone1Name=clavicle_r
  bone2Name=spine_05
  swing1LimitDegrees=45
  swing1Motion=Free
  swing2LimitDegrees=45
  swing2Motion=Free
  twistLimitDegrees=45
  twistMotion=Free
 }
 {
  bone1Name=upperarm_l
  bone2Name=clavicle_l
  swing1LimitDegrees=45
  swing1Motion=Free
  swing2LimitDegrees=45
  swing2Motion=Free
  twistLimitDegrees=45
  twistMotion=Free
 }
 {
  bone1Name=upperarm_r
  bone2Name=clavicle_r
  swing1LimitDegrees=45
  swing1Motion=Free
  swing2LimitDegrees=45
  swing2Motion=Free
  twistLimitDegrees=45
  twistMotion=Free
 }
 {
  bone1Name=head
  bone2Name=neck_02
  swing1LimitDegrees=45
  swing1Motion=Free
  swing2LimitDegrees=45
  swing2Motion=Free
  twistLimitDegrees=45
  twistMotion=Free
 }
 {
  bone1Name=neck_01
  bone2Name=spine_05
  swing1LimitDegrees=45
  swing1Motion=Free
  swing2LimitDegrees=45
  swing2Motion=Free
  twistLimitDegrees=45
  twistMotion=Free
 }
 {bone1Name=dyn_Hair_BackR_05, bone2Name=dyn_Hair_BackR_04, swing1LimitDegrees=45, swing2LimitDegrees=45, twistLimitDegrees=45}
 {bone1Name=dyn_Hair_Back_05, bone2Name=dyn_Hair_Back_04, swing1LimitDegrees=45, swing2LimitDegrees=45, twistLimitDegrees=45}
 {bone1Name=dyn_Hair_BackL_05, bone2Name=dyn_Hair_BackL_04, swing1LimitDegrees=45, swing2LimitDegrees=45, twistLimitDegrees=45}
 {
  bone1Name=neck_02
  bone2Name=neck_01
  swing1LimitDegrees=45
  swing1Motion=Free
  swing2LimitDegrees=45
  swing2Motion=Free
  twistLimitDegrees=45
  twistMotion=Free
 }
 {bone1Name=spine_05, bone2Name=dyn_hair_r_Fr_01, swing1LimitDegrees=45, swing2LimitDegrees=45, twistLimitDegrees=45}
 {bone1Name=spine_05, bone2Name=dyn_hair_l_Front_01, swing1LimitDegrees=45, swing2LimitDegrees=45, twistLimitDegrees=45}
]
