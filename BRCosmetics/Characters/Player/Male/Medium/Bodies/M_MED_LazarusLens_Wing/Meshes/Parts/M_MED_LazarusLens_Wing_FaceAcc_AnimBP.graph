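# Post-physics AnimBP for the LazarusLens Wing male face accessory, animating back hair and topknot.
# The copied body pose passes through hair constraints, a Control Rig and head-up offsets, then RigidBody simulation, with a separate front-end physics asset.
# AnimDynamics topknot springs and crouch/head-up nudges follow, before weighted multi-bone constraints re-pin lower back hair to the upper body.
# The EventGraph updates gravity override, head-tilt and rigid-body alphas, plus backbling and jump variables unread by the AnimGraph.
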AnimBlueprint /BRCosmetics/Characters/Player/Male/Medium/Bodies/M_MED_LazarusLens_Wing/Meshes/Parts/M_MED_LazarusLens_Wing_FaceAcc_AnimBP
target_skeleton=&/M_MED_LazarusLens_Wing_FaceAcc_Skeleton
graphs [
 AnimGraph {nodes [
   "Constraint - dyn_Hair_BackR_03 (PPPPPPPP)"!(
    ComponentPose
    Alpha=1
    ConstraintWeights_0=0.0005
    ConstraintWeights_1:float=0.2495
    ConstraintWeights_2:float=0.0023
    ConstraintWeights_3:float=0.1033
    ConstraintWeights_4:float=0.0519
    ConstraintWeights_5:float=0.3253
    ConstraintWeights_6:float=0.2663
    ConstraintWeights_7:float=0.0008
   ) -> (Pose) {
    alphaBoolBlend {blendOption=Linear}
    boneToModify {boneName=dyn_Hair_BackR_03}
    constraintSetup [
     {targetBone {boneName=spine_03}}
     {targetBone {boneName=spine_05}}
     {targetBone {boneName=clavicle_l}}
     {targetBone {boneName=clavicle_r}}
     {targetBone {boneName=neck_01}}
     {targetBone {boneName=head}}
     {targetBone {boneName=neck_02}}
     {targetBone {boneName=upperarm_twist_01_r}}
    ]
    constraintWeights [1,1,1,1,1,1,1,1]
   }
   n1 = "Constraint - dyn_Hair_BackL_03 (PPPPPPPPPPPPPPPPPPPPPPPPPPPPP)"!(
    ComponentPose=Local To Component(LocalPose=Use cached pose 'Base_Pose'())
    Alpha=1
    ConstraintWeights_0=0
    ConstraintWeights_1:float=0
    ConstraintWeights_2:float=0.0002
    ConstraintWeights_3:float=0.0726
    ConstraintWeights_4:float=0.0222
    ConstraintWeights_5:float=0.5728
    ConstraintWeights_6:float=0.1549
    ConstraintWeights_7:float=0
    ConstraintWeights_8:float=0
    ConstraintWeights_9:float=0
    ConstraintWeights_10:float=0
    ConstraintWeights_11:float=0
    ConstraintWeights_12:float=0.0133
    ConstraintWeights_13:float=0
    ConstraintWeights_14:float=0.0134
    ConstraintWeights_15:float=0.0027
    ConstraintWeights_16:float=0
    ConstraintWeights_17:float=0.0146
    ConstraintWeights_18:float=0
    ConstraintWeights_19:float=0.0013
    ConstraintWeights_20:float=0
    ConstraintWeights_21:float=0.0012
    ConstraintWeights_22:float=0.0007
    ConstraintWeights_23:float=0.0001
    ConstraintWeights_24:float=0.0371
    ConstraintWeights_25:float=0.0833
    ConstraintWeights_26:float=0.0095
    ConstraintWeights_27:float=0
    ConstraintWeights_28:float=0
   ) -> (Pose) {
    alphaBoolBlend {blendOption=Linear}
    boneToModify {boneName=dyn_Hair_BackL_03}
    constraintSetup [
     {targetBone {boneName=pelvis}}
     {targetBone {boneName=spine_01}}
     {targetBone {boneName=spine_02}}
     {targetBone {boneName=spine_03}}
     {targetBone {boneName=spine_04}}
     {targetBone {boneName=spine_05}}
     {targetBone {boneName=clavicle_l}}
     {targetBone {boneName=upperarm_l}}
     {targetBone {boneName=lowerarm_l}}
     {targetBone {boneName=lowerarm_twist_02_l}}
     {targetBone {boneName=deform_elbow_out_l}}
     {targetBone {boneName=deform_elbow_in_l}}
     {targetBone {boneName=upperarm_twist_01_l}}
     {targetBone {boneName=deform_upperarm_fr_l}}
     {targetBone {boneName=deform_upperarm_up_l}}
     {targetBone {boneName=deform_upperarm_bk_l}}
     {targetBone {boneName=upperarm_twist_02_l}}
     {targetBone {boneName=clavicle_r}}
     {targetBone {boneName=upperarm_r}}
     {targetBone {boneName=upperarm_twist_01_r}}
     {targetBone {boneName=deform_upperarm_fr_r}}
     {targetBone {boneName=deform_upperarm_up_r}}
     {targetBone {boneName=deform_upperarm_bk_r}}
     {targetBone {boneName=upperarm_twist_02_r}}
     {targetBone {boneName=neck_01}}
     {targetBone {boneName=neck_02}}
     {targetBone {boneName=head}}
     {targetBone {boneName=deform_pec_r}}
     {targetBone {boneName=deform_pec_l}}
    ]
    constraintWeights [1,1,1,1,1,1,1,1,1,1,1,1,1,1,1,1,1,1,1,1,1,1,1,1,1,1,1,1,1]
   }
   n2 = "Constraint - dyn_Hair_Back_03 (PPPPPPPPPPPPPPPPPPPPP)"!(
    ComponentPose=@n1
    Alpha=1
    ConstraintWeights_0=0
    ConstraintWeights_1:float=0
    ConstraintWeights_2:float=0.0007
    ConstraintWeights_3:float=0.0926
    ConstraintWeights_4:float=0.0303
    ConstraintWeights_5:float=0.5803
    ConstraintWeights_6:float=0.0653
    ConstraintWeights_7:float=0.0058
    ConstraintWeights_8:float=0
    ConstraintWeights_9:float=0.0053
    ConstraintWeights_10:float=0.0016
    ConstraintWeights_11:float=0
    ConstraintWeights_12:float=0.0629
    ConstraintWeights_13:float=0.0064
    ConstraintWeights_14:float=0
    ConstraintWeights_15:float=0.0056
    ConstraintWeights_16:float=0.0037
    ConstraintWeights_17:float=0.0003
    ConstraintWeights_18:float=0.04
    ConstraintWeights_19:float=0.0886
    ConstraintWeights_20:float=0.0108
   ) -> (Pose) {
    alphaBoolBlend {blendOption=Linear}
    boneToModify {boneName=dyn_Hair_Back_03}
    constraintSetup [
     {targetBone {boneName=pelvis}}
     {targetBone {boneName=spine_01}}
     {targetBone {boneName=spine_02}}
     {targetBone {boneName=spine_03}}
     {targetBone {boneName=spine_04}}
     {targetBone {boneName=spine_05}}
     {targetBone {boneName=clavicle_l}}
     {targetBone {boneName=upperarm_twist_01_l}}
     {targetBone {boneName=deform_upperarm_fr_l}}
     {targetBone {boneName=deform_upperarm_up_l}}
     {targetBone {boneName=deform_upperarm_bk_l}}
     {targetBone {boneName=upperarm_twist_02_l}}
     {targetBone {boneName=clavicle_r}}
     {targetBone {boneName=upperarm_twist_01_r}}
     {targetBone {boneName=deform_upperarm_fr_r}}
     {targetBone {boneName=deform_upperarm_up_r}}
     {targetBone {boneName=deform_upperarm_bk_r}}
     {targetBone {boneName=upperarm_twist_02_r}}
     {targetBone {boneName=neck_01}}
     {targetBone {boneName=neck_02}}
     {targetBone {boneName=head}}
    ]
    constraintWeights [1,1,1,1,1,1,1,1,1,1,1,1,1,1,1,1,1,1,1,1,1]
   }
   n3 = "Constraint - dyn_Hair_BackR_03 (PPPPPPPPPPPPPPPPPPPPPPPPPPPPP)"!(
    ComponentPose=@n2
    Alpha=1
    ConstraintWeights_0=0
    ConstraintWeights_1:float=0
    ConstraintWeights_2:float=0.0002
    ConstraintWeights_3:float=0.0537
    ConstraintWeights_4:float=0.0154
    ConstraintWeights_5:float=0.5507
    ConstraintWeights_6:float=0.0054
    ConstraintWeights_7:float=0
    ConstraintWeights_8:float=0
    ConstraintWeights_9:float=0
    ConstraintWeights_10:float=0
    ConstraintWeights_11:float=0
    ConstraintWeights_12:float=0.0004
    ConstraintWeights_13:float=0
    ConstraintWeights_14:float=0.0004
    ConstraintWeights_15:float=0.0001
    ConstraintWeights_16:float=0
    ConstraintWeights_17:float=0.2209
    ConstraintWeights_18:float=0
    ConstraintWeights_19:float=0.0184
    ConstraintWeights_20:float=0.0001
    ConstraintWeights_21:float=0.02
    ConstraintWeights_22:float=0.0051
    ConstraintWeights_23:float=0.0005
    ConstraintWeights_24:float=0.0312
    ConstraintWeights_25:float=0.0707
    ConstraintWeights_26:float=0.0068
    ConstraintWeights_27:float=0
    ConstraintWeights_28:float=0
   ) -> (Pose) {
    alphaBoolBlend {blendOption=Linear}
    boneToModify {boneName=dyn_Hair_BackR_03}
    constraintSetup [
     {targetBone {boneName=pelvis}}
     {targetBone {boneName=spine_01}}
     {targetBone {boneName=spine_02}}
     {targetBone {boneName=spine_03}}
     {targetBone {boneName=spine_04}}
     {targetBone {boneName=spine_05}}
     {targetBone {boneName=clavicle_l}}
     {targetBone {boneName=upperarm_l}}
     {targetBone {boneName=lowerarm_l}}
     {targetBone {boneName=lowerarm_twist_02_l}}
     {targetBone {boneName=deform_elbow_out_l}}
     {targetBone {boneName=deform_elbow_in_l}}
     {targetBone {boneName=upperarm_twist_01_l}}
     {targetBone {boneName=deform_upperarm_fr_l}}
     {targetBone {boneName=deform_upperarm_up_l}}
     {targetBone {boneName=deform_upperarm_bk_l}}
     {targetBone {boneName=upperarm_twist_02_l}}
     {targetBone {boneName=clavicle_r}}
     {targetBone {boneName=upperarm_r}}
     {targetBone {boneName=upperarm_twist_01_r}}
     {targetBone {boneName=deform_upperarm_fr_r}}
     {targetBone {boneName=deform_upperarm_up_r}}
     {targetBone {boneName=deform_upperarm_bk_r}}
     {targetBone {boneName=upperarm_twist_02_r}}
     {targetBone {boneName=neck_01}}
     {targetBone {boneName=neck_02}}
     {targetBone {boneName=head}}
     {targetBone {boneName=deform_pec_r}}
     {targetBone {boneName=deform_pec_l}}
    ]
    constraintWeights [1,1,1,1,1,1,1,1,1,1,1,1,1,1,1,1,1,1,1,1,1,1,1,1,1,1,1,1,1]
   }
   n4 = "Constraint - dyn_Hair_BackL_04 (PPPPPPPPPPPPPPPPPPPPPPPPP)"!(
    ComponentPose=@n3
    Alpha=1
    ConstraintWeights_0=0
    ConstraintWeights_1:float=0
    ConstraintWeights_2:float=0.0002
    ConstraintWeights_3:float=0.0726
    ConstraintWeights_4:float=0.0222
    ConstraintWeights_5:float=0.5728
    ConstraintWeights_6:float=0.1549
    ConstraintWeights_7:float=0
    ConstraintWeights_8:float=0.0133
    ConstraintWeights_9:float=0
    ConstraintWeights_10:float=0.0134
    ConstraintWeights_11:float=0.0027
    ConstraintWeights_12:float=0
    ConstraintWeights_13:float=0.0146
    ConstraintWeights_14:float=0
    ConstraintWeights_15:float=0.0013
    ConstraintWeights_16:float=0
    ConstraintWeights_17:float=0.0012
    ConstraintWeights_18:float=0.0007
    ConstraintWeights_19:float=0.0001
    ConstraintWeights_20:float=0.0371
    ConstraintWeights_21:float=0.0833
    ConstraintWeights_22:float=0.0095
    ConstraintWeights_23:float=0
    ConstraintWeights_24:float=0
   ) -> (Pose) {
    alphaBoolBlend {blendOption=Linear}
    boneToModify {boneName=dyn_Hair_BackL_04}
    constraintSetup [
     {targetBone {boneName=pelvis}}
     {targetBone {boneName=spine_01}}
     {targetBone {boneName=spine_02}}
     {targetBone {boneName=spine_03}}
     {targetBone {boneName=spine_04}}
     {targetBone {boneName=spine_05}}
     {targetBone {boneName=clavicle_l}}
     {targetBone {boneName=upperarm_l}}
     {targetBone {boneName=upperarm_twist_01_l}}
     {targetBone {boneName=deform_upperarm_fr_l}}
     {targetBone {boneName=deform_upperarm_up_l}}
     {targetBone {boneName=deform_upperarm_bk_l}}
     {targetBone {boneName=upperarm_twist_02_l}}
     {targetBone {boneName=clavicle_r}}
     {targetBone {boneName=upperarm_r}}
     {targetBone {boneName=upperarm_twist_01_r}}
     {targetBone {boneName=deform_upperarm_fr_r}}
     {targetBone {boneName=deform_upperarm_up_r}}
     {targetBone {boneName=deform_upperarm_bk_r}}
     {targetBone {boneName=upperarm_twist_02_r}}
     {targetBone {boneName=neck_01}}
     {targetBone {boneName=neck_02}}
     {targetBone {boneName=head}}
     {targetBone {boneName=deform_pec_r}}
     {targetBone {boneName=deform_pec_l}}
    ]
    constraintWeights [1,1,1,1,1,1,1,1,1,1,1,1,1,1,1,1,1,1,1,1,1,1,1,1,1]
   }
   n5 = "Constraint - dyn_Hair_Back_04 (PPPPPPPPPPPPPPPPPPPPPPPPPPPPP)"!(
    ComponentPose=@n4
    Alpha=1
    ConstraintWeights_0=0
    ConstraintWeights_1:float=0
    ConstraintWeights_2:float=0.0005
    ConstraintWeights_3:float=0.0749
    ConstraintWeights_4:float=0.0242
    ConstraintWeights_5:float=0.5894
    ConstraintWeights_6:float=0.0578
    ConstraintWeights_7:float=0
    ConstraintWeights_8:float=0
    ConstraintWeights_9:float=0
    ConstraintWeights_10:float=0
    ConstraintWeights_11:float=0
    ConstraintWeights_12:float=0.0049
    ConstraintWeights_13:float=0
    ConstraintWeights_14:float=0.0046
    ConstraintWeights_15:float=0.0013
    ConstraintWeights_16:float=0
    ConstraintWeights_17:float=0.056
    ConstraintWeights_18:float=0
    ConstraintWeights_19:float=0.0054
    ConstraintWeights_20:float=0
    ConstraintWeights_21:float=0.0049
    ConstraintWeights_22:float=0.003
    ConstraintWeights_23:float=0.0002
    ConstraintWeights_24:float=0.0484
    ConstraintWeights_25:float=0.1081
    ConstraintWeights_26:float=0.0164
    ConstraintWeights_27:float=0
    ConstraintWeights_28:float=0
   ) -> (Pose) {
    alphaBoolBlend {blendOption=Linear}
    boneToModify {boneName=dyn_Hair_Back_04}
    constraintSetup [
     {targetBone {boneName=pelvis}}
     {targetBone {boneName=spine_01}}
     {targetBone {boneName=spine_02}}
     {targetBone {boneName=spine_03}}
     {targetBone {boneName=spine_04}}
     {targetBone {boneName=spine_05}}
     {targetBone {boneName=clavicle_l}}
     {targetBone {boneName=upperarm_l}}
     {targetBone {boneName=lowerarm_l}}
     {targetBone {boneName=lowerarm_twist_02_l}}
     {targetBone {boneName=deform_elbow_out_l}}
     {targetBone {boneName=deform_elbow_in_l}}
     {targetBone {boneName=upperarm_twist_01_l}}
     {targetBone {boneName=deform_upperarm_fr_l}}
     {targetBone {boneName=deform_upperarm_up_l}}
     {targetBone {boneName=deform_upperarm_bk_l}}
     {targetBone {boneName=upperarm_twist_02_l}}
     {targetBone {boneName=clavicle_r}}
     {targetBone {boneName=upperarm_r}}
     {targetBone {boneName=upperarm_twist_01_r}}
     {targetBone {boneName=deform_upperarm_fr_r}}
     {targetBone {boneName=deform_upperarm_up_r}}
     {targetBone {boneName=deform_upperarm_bk_r}}
     {targetBone {boneName=upperarm_twist_02_r}}
     {targetBone {boneName=neck_01}}
     {targetBone {boneName=neck_02}}
     {targetBone {boneName=head}}
     {targetBone {boneName=deform_pec_r}}
     {targetBone {boneName=deform_pec_l}}
    ]
    constraintWeights [1,1,1,1,1,1,1,1,1,1,1,1,1,1,1,1,1,1,1,1,1,1,1,1,1,1,1,1,1]
   }
   n6 = "Constraint - dyn_Hair_BackR_04 (PPPPPPPPPPPPPPPPPPPPPPPPPPPPP)"!(
    ComponentPose=@n5
    Alpha=1
    ConstraintWeights_0=0
    ConstraintWeights_1:float=0
    ConstraintWeights_2:float=0.0002
    ConstraintWeights_3:float=0.0537
    ConstraintWeights_4:float=0.0154
    ConstraintWeights_5:float=0.5507
    ConstraintWeights_6:float=0.0054
    ConstraintWeights_7:float=0
    ConstraintWeights_8:float=0
    ConstraintWeights_9:float=0
    ConstraintWeights_10:float=0
    ConstraintWeights_11:float=0
    ConstraintWeights_12:float=0.0004
    ConstraintWeights_13:float=0
    ConstraintWeights_14:float=0.0004
    ConstraintWeights_15:float=0.0001
    ConstraintWeights_16:float=0
    ConstraintWeights_17:float=0.2209
    ConstraintWeights_18:float=0
    ConstraintWeights_19:float=0.0184
    ConstraintWeights_20:float=0.0001
    ConstraintWeights_21:float=0.02
    ConstraintWeights_22:float=0.0051
    ConstraintWeights_23:float=0.0005
    ConstraintWeights_24:float=0.0312
    ConstraintWeights_25:float=0.0707
    ConstraintWeights_26:float=0.0068
    ConstraintWeights_27:float=0
    ConstraintWeights_28:float=0
   ) -> (Pose) {
    alphaBoolBlend {blendOption=Linear}
    boneToModify {boneName=dyn_Hair_BackR_04}
    constraintSetup [
     {targetBone {boneName=pelvis}}
     {targetBone {boneName=spine_01}}
     {targetBone {boneName=spine_02}}
     {targetBone {boneName=spine_03}}
     {targetBone {boneName=spine_04}}
     {targetBone {boneName=spine_05}}
     {targetBone {boneName=clavicle_l}}
     {targetBone {boneName=upperarm_l}}
     {targetBone {boneName=lowerarm_l}}
     {targetBone {boneName=lowerarm_twist_02_l}}
     {targetBone {boneName=deform_elbow_out_l}}
     {targetBone {boneName=deform_elbow_in_l}}
     {targetBone {boneName=upperarm_twist_01_l}}
     {targetBone {boneName=deform_upperarm_fr_l}}
     {targetBone {boneName=deform_upperarm_up_l}}
     {targetBone {boneName=deform_upperarm_bk_l}}
     {targetBone {boneName=upperarm_twist_02_l}}
     {targetBone {boneName=clavicle_r}}
     {targetBone {boneName=upperarm_r}}
     {targetBone {boneName=upperarm_twist_01_r}}
     {targetBone {boneName=deform_upperarm_fr_r}}
     {targetBone {boneName=deform_upperarm_up_r}}
     {targetBone {boneName=deform_upperarm_bk_r}}
     {targetBone {boneName=upperarm_twist_02_r}}
     {targetBone {boneName=neck_01}}
     {targetBone {boneName=neck_02}}
     {targetBone {boneName=head}}
     {targetBone {boneName=deform_pec_r}}
     {targetBone {boneName=deform_pec_l}}
    ]
    constraintWeights [1,1,1,1,1,1,1,1,1,1,1,1,1,1,1,1,1,1,1,1,1,1,1,1,1,1,1,1,1]
   }
   Output Pose(Result=Component To Local(ComponentPose=@n6)) {layerGroup=None}
   n7 = RigidBody!(
    ComponentPose=Local To Component(LocalPose=Use cached pose 'PreRBAN'())
    Alpha
    OverrideWorldGravity:Vector=Get gravity_override() -> (gravity_override:Vector)
    ExternalForce
   ) -> (Pose) {
    bClampLinearTranslationLimitToRefPose=true
    bOverrideWorldGravity=true
    overridePhysicsAsset=&/M_MED_LazarusLens_Wing_FrontEnd_FaceAcc_Physics
    simulationSpace=BaseBoneSpace
    alphaBoolBlend {blendOption=Linear}
    baseBoneRef {boneName=spine_05}
    componentAppliedLinearAccClamp=<4000 4000 4000>
    componentLinearAccScale=<0.5 0.75 0.75>
    componentLinearVelScale=<0.5 0.5 0.4>
    overrideWorldGravity {z=-800}
   }
   n8 = Component To Local(ComponentPose=@n7)
   n9 = RigidBody(ComponentPose=Local To Component(LocalPose=Use cached pose 'PreRBAN'()), OverrideWorldGravity=Get gravity_override()) {
    bClampLinearTranslationLimitToRefPose=true
    bOverrideWorldGravity=true
    overridePhysicsAsset=&/M_MED_LazarusLens_Wing_FaceAcc_Physics
    simulationSpace=BaseBoneSpace
    alphaBoolBlend {blendOption=Linear}
    baseBoneRef {boneName=spine_05}
    componentAppliedLinearAccClamp=<4000 4000 4000>
    componentLinearAccScale=<0.5 0.75 0.75>
    componentLinearVelScale=<0.5 0.5 0.4>
    overrideWorldGravity {z=-800}
   }
   n10 = Component To Local(ComponentPose=@n9)
   n11 = Get bIsInFrontEnd()
   n12 = Anim Dynamics - Bone: dyn_topknot_01(
    ComponentPose=Local To Component(LocalPose=Blend Poses by bool(BlendPose_0=@n8, BlendPose_1=@n10, bActiveValue=@n11) {blendType=Linear})
    Alpha=1
    ExternalForce="0,0,0"
   ) {
    angularDampingOverride=0.9
    angularSpringConstant=600
    bAngularSpring=true
    bGravityOverrideInSimSpace=true
    bOverrideAngularDamping=true
    bOverrideLinearDamping=true
    bUseGravityOverride=true
    gravityScale=0.35
    linearDampingOverride=0.999
    linearSpringConstant=150
    simulationSpace=RootRelative
    alphaBoolBlend {blendOption=Linear}
    boundBone {boneName=dyn_topknot_01}
    componentAppliedLinearAccClamp=<1500 1500 1500>
    componentLinearAccScale=<0.1 0.1 0.25>
    componentLinearVelScale=<0.3 0.3 0.4>
    gravityOverride {z=-980}
    physicsBodyDefinitions [{
      sphereCollisionRadius=0
      boundBone {boneName=dyn_topknot_01}
      boxExtents=<6 9 6>
      constraintSetup {angularLimitsMax=<5 40 35>, angularLimitsMin=<-5 -75 -20>, angularTarget=<1 0 0>}
      localJointOffset=<0.75 0 -0.2>
     }]
   }
   n13 = Anim Dynamics - Bone: dyn_topknot_02(ComponentPose=@n12, Alpha=1, ExternalForce="0,0,0") {
    angularDampingOverride=0.9
    angularSpringConstant=300
    bAngularSpring=true
    bGravityOverrideInSimSpace=true
    bOverrideAngularDamping=true
    bOverrideLinearDamping=true
    bUseGravityOverride=true
    gravityScale=0.35
    linearDampingOverride=0.999
    linearSpringConstant=150
    simulationSpace=RootRelative
    alphaBoolBlend {blendOption=Linear}
    boundBone {boneName=dyn_topknot_02}
    componentAppliedLinearAccClamp=<1500 1500 1500>
    componentLinearAccScale=<0.1 0.1 0.25>
    componentLinearVelScale=<0.3 0.3 0.4>
    gravityOverride {z=-980}
    physicsBodyDefinitions [{
      sphereCollisionRadius=0
      boundBone {boneName=dyn_topknot_02}
      boxExtents=<6 9 6>
      constraintSetup {angularLimitsMax=<5 40 35>, angularLimitsMin=<-5 -75 -20>, angularTarget=<1 0 0>}
      localJointOffset=<0.65 0 -0.2>
     }]
   }
   n14 = "Transform (Modify) Bone - Bone: dyn_Hair_BackR_01"(ComponentPose=@n13, Alpha=Get Head_Up() -> (Head_Up:double), Translation="2,2.5,0", Rotation="0,0,0", Scale="1,1,1") {
    rotationMode=BMM_Additive
    rotationSpace=BCS_BoneSpace
    translationMode=BMM_Additive
    translationSpace=BCS_BoneSpace
    alphaBoolBlend {blendOption=Linear}
    boneToModify {boneName=dyn_Hair_BackR_01}
    translation=<2 2.5>
   }
   n15 = "Transform (Modify) Bone - Bone: dyn_Hair_BackR_02"(ComponentPose=@n14, Alpha=1, Translation="0,0,0", Rotation="0,0,0", Scale="1,1,1") {
    rotationMode=BMM_Additive
    rotationSpace=BCS_BoneSpace
    translationMode=BMM_Additive
    translationSpace=BCS_BoneSpace
    alphaBoolBlend {blendOption=Linear}
    boneToModify {boneName=dyn_Hair_BackR_02}
   }
   n16 = Get bIsCrouching()
   n17 = "Transform (Modify) Bone - Bone: dyn_Hair_BackL_02"!(ComponentPose=@n15, bAlphaBoolEnabled:bool=@n16, Translation="2,2,0", Rotation="0,0,0", Scale="1,1,1") -> (Pose) {
    alphaInputType=Bool
    rotationMode=BMM_Additive
    rotationSpace=BCS_BoneSpace
    translationMode=BMM_Additive
    translationSpace=BCS_BoneSpace
    alphaBoolBlend {blendOption=Linear}
    boneToModify {boneName=dyn_Hair_BackL_02}
    translation=<2 2>
   }
   n18 = "Transform (Modify) Bone - Bone: dyn_Hair_BackR_01"!(ComponentPose=@n17, bAlphaBoolEnabled:bool=@n16, Translation="2,2,0", Rotation="0,0,0", Scale="1,1,1") -> (Pose) {
    alphaInputType=Bool
    rotationMode=BMM_Additive
    rotationSpace=BCS_BoneSpace
    translationMode=BMM_Additive
    translationSpace=BCS_BoneSpace
    alphaBoolBlend {blendOption=Linear}
    boneToModify {boneName=dyn_Hair_BackR_01}
    translation=<2 2>
   }
   n19 = "Transform (Modify) Bone - Bone: dyn_Hair_Back_04"!(ComponentPose=@n18, bAlphaBoolEnabled:bool=@n16, Translation="0,0,0", Rotation="10,0,0", Scale="1,1,1") -> (Pose) {
    alphaInputType=Bool
    rotationMode=BMM_Additive
    rotationSpace=BCS_BoneSpace
    translationMode=BMM_Additive
    translationSpace=BCS_BoneSpace
    alphaBoolBlend {blendOption=Linear}
    boneToModify {boneName=dyn_Hair_Back_04}
    rotation {pitch=10}
   }
   Save cached pose 'Base_Pose'(Pose=Component To Local(ComponentPose=@n19)) {cacheName=Base_Pose}
   n20 = Layered blend per bone(
    BasePose=Copy Pose From Mesh() {bCopyCurves=true, bCopyCustomAttributes=true, bUseAttachedParent=true, bUseMeshPose=true}
    BlendPoses_0=Copy Pose From Mesh(SourceMeshComponent=Get Skel_Mesh()) {bUseAttachedParent=true}
   ) {layerSetup [{branchFilters [{boneName=upperarm_l}, {boneName=upperarm_r}, {boneName=clavicle_r}, {boneName=clavicle_l}]}]}
   n21 = "Constraint - dyn_Hair_BackL_01 (PP)"!(ComponentPose=Local To Component(LocalPose=@n20), Alpha=1, ConstraintWeights_0=0.3, ConstraintWeights_1:float=0.7) -> (Pose) {
    alphaBoolBlend {blendOption=Linear}
    boneToModify {boneName=dyn_Hair_BackL_01}
    constraintSetup [{targetBone {boneName=head}}, {targetBone {boneName=dyn_Hair_Back_01}}]
    constraintWeights [1,1]
   }
   n22 = "Constraint - dyn_Hair_BackL_02 (PP)"!(ComponentPose=@n21, Alpha=1, ConstraintWeights_0=0.3, ConstraintWeights_1:float=0.7) -> (Pose) {
    alphaBoolBlend {blendOption=Linear}
    boneToModify {boneName=dyn_Hair_BackL_02}
    constraintSetup [{targetBone {boneName=head}}, {targetBone {boneName=dyn_Hair_Back_02}}]
    constraintWeights [1,1]
   }
   n23 = "Constraint - dyn_Hair_BackL_03 (PP)"!(ComponentPose=@n22, Alpha=1, ConstraintWeights_0=0.5, ConstraintWeights_1:float=0.5) -> (Pose) {
    alphaBoolBlend {blendOption=Linear}
    boneToModify {boneName=dyn_Hair_BackL_03}
    constraintSetup [{targetBone {boneName=spine_05}}, {targetBone {boneName=dyn_Hair_Back_03}}]
    constraintWeights [1,1]
   }
   n24 = "Constraint - dyn_Hair_BackL_04 (PP)"!(ComponentPose=@n23, Alpha=1, ConstraintWeights_0=0.8, ConstraintWeights_1:float=0.2) -> (Pose) {
    alphaBoolBlend {blendOption=Linear}
    boneToModify {boneName=dyn_Hair_BackL_04}
    constraintSetup [{targetBone {boneName=spine_05}}, {targetBone {boneName=dyn_Hair_Back_04}}]
    constraintWeights [1,1]
   }
   n25 = "Constraint - dyn_Hair_Back_01 (PP)"!(ComponentPose=@n24, Alpha=1, ConstraintWeights_0=0.3, ConstraintWeights_1:float=0.7) -> (Pose) {
    alphaBoolBlend {blendOption=Linear}
    boneToModify {boneName=dyn_Hair_Back_01}
    constraintSetup [{targetBone {boneName=head}}, {targetBone {boneName=dyn_Hair_Back_01}}]
    constraintWeights [1,1]
   }
   n26 = "Constraint - dyn_Hair_Back_02 (PP)"!(ComponentPose=@n25, Alpha=1, ConstraintWeights_0=0.3, ConstraintWeights_1:float=0.7) -> (Pose) {
    alphaBoolBlend {blendOption=Linear}
    boneToModify {boneName=dyn_Hair_Back_02}
    constraintSetup [{targetBone {boneName=head}}, {targetBone {boneName=dyn_Hair_Back_02}}]
    constraintWeights [1,1]
   }
   n27 = "Constraint - dyn_Hair_Back_03 (PP)"!(ComponentPose=@n26, Alpha=1, ConstraintWeights_0=0.6, ConstraintWeights_1:float=0.4) -> (Pose) {
    alphaBoolBlend {blendOption=Linear}
    boneToModify {boneName=dyn_Hair_Back_03}
    constraintSetup [{targetBone {boneName=spine_05}}, {targetBone {boneName=dyn_Hair_Back_03}}]
    constraintWeights [1,1]
   }
   n28 = "Constraint - dyn_Hair_Back_04 (PP)"!(ComponentPose=@n27, Alpha=1, ConstraintWeights_0=1, ConstraintWeights_1:float=0) -> (Pose) {
    alphaBoolBlend {blendOption=Linear}
    boneToModify {boneName=dyn_Hair_Back_04}
    constraintSetup [{targetBone {boneName=spine_05}}, {targetBone {boneName=dyn_Hair_Back_04}}]
    constraintWeights [1,1]
   }
   n29 = "Constraint - dyn_Hair_BackR_01 (PP)"!(ComponentPose=@n28, Alpha=1, ConstraintWeights_0=0.3, ConstraintWeights_1:float=0.7) -> (Pose) {
    alphaBoolBlend {blendOption=Linear}
    boneToModify {boneName=dyn_Hair_BackR_01}
    constraintSetup [{targetBone {boneName=head}}, {targetBone {boneName=dyn_Hair_Back_01}}]
    constraintWeights [1,1]
   }
   n30 = "Constraint - dyn_Hair_BackR_02 (PP)"!(ComponentPose=@n29, Alpha=1, ConstraintWeights_0=0.3, ConstraintWeights_1:float=0.7) -> (Pose) {
    alphaBoolBlend {blendOption=Linear}
    boneToModify {boneName=dyn_Hair_BackR_02}
    constraintSetup [{targetBone {boneName=head}}, {targetBone {boneName=dyn_Hair_Back_02}}]
    constraintWeights [1,1]
   }
   n31 = "Constraint - dyn_Hair_BackR_03 (PP)"!(ComponentPose=@n30, Alpha=1, ConstraintWeights_0=0.5, ConstraintWeights_1:float=0.5) -> (Pose) {
    alphaBoolBlend {blendOption=Linear}
    boneToModify {boneName=dyn_Hair_BackR_03}
    constraintSetup [{targetBone {boneName=spine_05}}, {targetBone {boneName=dyn_Hair_Back_03}}]
    constraintWeights [1,1]
   }
   n32 = "Constraint - dyn_Hair_BackR_04 (PP)"!(ComponentPose=@n31, Alpha=1, ConstraintWeights_0=0.8, ConstraintWeights_1:float=0.2) -> (Pose) {
    alphaBoolBlend {blendOption=Linear}
    boneToModify {boneName=dyn_Hair_BackR_04}
    constraintSetup [{targetBone {boneName=spine_05}}, {targetBone {boneName=dyn_Hair_Back_04}}]
    constraintWeights [1,1]
   }
   n33 = Control Rig!(Source=Component To Local(ComponentPose=@n32), Alpha, rBAN:double=Get rBAN() -> (rBAN:double)) -> (Pose) {
    alphaBoolBlend {blendOption=Linear}
    controlRigAssetReference {blueprintRigClass=&/M_MED_LazarusLens_Wing_FaceAcc_CtrlRig.M_MED_LazarusLens_Wing_FaceAcc_CtrlRig_C}
   }
   n34 = "Transform (Modify) Bone - Bone: dyn_Hair_BackR_01"!(ComponentPose=Local To Component(LocalPose=@n33), Alpha:float=Get Head_Up(), Translation="2,0,0", Rotation="0,30,0", Scale="1,1,1") -> (Pose) {
    rotationMode=BMM_Additive
    rotationSpace=BCS_BoneSpace
    translationMode=BMM_Additive
    translationSpace=BCS_BoneSpace
    alphaBoolBlend {blendOption=Linear}
    boneToModify {boneName=dyn_Hair_BackR_01}
    rotation {yaw=30}
    translation {x=2}
   }
   n35 = "Transform (Modify) Bone - Bone: dyn_Hair_Back_01"(ComponentPose=@n34, Alpha=Get Head_Up(), Translation="0,0,0", Rotation="0,20,0", Scale="1,1,1") {
    rotationMode=BMM_Additive
    rotationSpace=BCS_BoneSpace
    translationMode=BMM_Additive
    translationSpace=BCS_BoneSpace
    alphaBoolBlend {blendOption=Linear}
    boneToModify {boneName=dyn_Hair_Back_01}
    rotation {yaw=20}
   }
   Save cached pose 'PreRBAN'(Pose=Component To Local(ComponentPose=@n35)) {cacheName=PreRBAN}
   Use cached pose 'Base_Pose'()
  ]}
 EventGraph {nodes [
   Event BlueprintUpdateAnimation()
   > n1 = Sequence!(execute) -> (then_0, then_1, then_2:Exec)
   n2 = Get Skel_Mesh()
   > Set neck_rot_up(neck_rot_up:double=GetAlphaFromReferenceTransformDelta(SkeletalMeshComponent=@n2, BoneName=head, BaseName=spine_05, InputRange=(X=0,Y=-140), TwistAxis="0,0,1")) -> (Output_Get:double)
   > Set neck_rot_dn(neck_rot_dn:double=GetAlphaFromReferenceTransformDelta(SkeletalMeshComponent=@n2, BoneName=neck_01, BaseName=spine_05, DeltaComponentType=EulerX, InputRange=(X=-15,Y=50), TwistAxis="0,0,1")) -> (Output_Get:double)
   n3 = Branch(execute=@n1.then_1, Condition=Get bRigidBodyEnabled())
   > Set rBAN(rBAN:double=1) -> (Output_Get:double)
   Set rBAN(execute=@n3.else, rBAN=0)
   n4 = Gravity Override State Function DEPRECATED(
    execute=@n1.then_2
    CharacterPartAnimInstance=Self-Reference()
    GameTimeInSeconds=GetGameTimeInSeconds()
    GravityOverrideParams=Get Gravity Override Params()
   )
   > Set gravity_override(gravity_override=@n4)
   n5 = GetDeltaTransformFromRefPose(self=GetOwnerSkeletalMeshComponent(), BoneName=head, BaseName=neck_02)
   n6 = MapRangeClamped(Value=BreakRotator(InRot=BreakTransform(InTransform=@n5).Rotation).Yaw, InRangeA=5, InRangeB=-20)
   > Set Head_Up(Head_Up:double=SelectFloat(A=@n6, B=0, bPickA=Get bRigidBodyEnabled())) -> (Output_Get:double)
   > Set head_up_rban_off(head_up_rban_off:double=@n6) -> (Output_Get:double)
   n7 = CheckForBackbling()
   n8 = Get bRigidBodyEnabled()
   > n9 = Set backpack_wing_var(backpack_wing_var:double=SelectFloat(A=0.75, B=0, bPickA=AND Boolean(A=@n8, B=@n7))) -> (Output_Get:double)
   n10 = float + float(A=@n9, B=0.5)
   > Set backpack_wing_var_2(backpack_wing_var_2:double=@n10) -> (Output_Get:double)
   > Set Backpack_Wing_Var_3(Backpack_Wing_Var_3:double=@n10) -> (Output_Get:double)
   > Set LL_Backpack_Equipped(LL_Backpack_Equipped:bool=BackpackItemDefinitionHasAnyMetaTag!(
     Tags:Array of Gameplay Tag Structures=Make Array!("[0]"=(TagName="Cosmetics.MetaTags.LazarusLens"), "[1]":Gameplay Tag Structure=(TagName="Cosmetics.MetaTags.LazarusLens.Wing")) -> (Array)
     bRequiresEquipped:bool=false
    ) -> (ReturnValue:bool)) -> (Output_Get:bool)
   n11 = Event OnBeginFalling()
   n12 = Branch(execute=[@n11.then, @n15.then], Condition=Get isWearingPackBool() -> (isWearingPackBool:bool))
   Set jumping(execute=[@n14.then, @n18.then, @n12.then], jumping:double=0) -> (Output_Get:double)
   Set jumping(execute=@n12.else, jumping=1)
   Event OnCharacterCustomizationLoaded()
   n13 = CheckForBackbling()
   > Set isWearingPackBool(isWearingPackBool:bool=OR Boolean!(A=@n13, B=@n13.bIsBackblingACape, C:bool=@n13.bIsBackblingLong) -> (ReturnValue)) -> (Output_Get:bool)
   > Set is_WearingCape(is_WearingCape:bool=@n13.bIsBackblingLong) -> (Output_Get:bool)
   > Set FrontEndCape(FrontEndCape:bool=AND Boolean(A=@n13.bIsBackblingLong, B=Get bIsInFrontEnd())) -> (Output_Get:bool)
   n14 = Event OnEndFalling()
   n15 = Event OnBeginJumping()
   Event BlueprintInitializeAnimation()
   > Set Skel_Mesh(Skel_Mesh=GetOwnerSkeletalMeshComponent())
   > n16 = Cast To PlayerPawn_Athena(Object=TryGetPawnOwner())
   > n17 = Set Pawn Owner(Pawn Owner:Fort Player Pawn Athena Object Reference=@n16) -> (Output_Get:Fort Player Pawn Athena Object Reference)
   > Set Shoulder_Mesh(Shoulder_Mesh:Skeletal Mesh Component Object Reference=GetSkeletalMeshForPartType(self=@n17)) -> (Output_Get:Skeletal Mesh Component Object Reference)
   n18 = Event OnEndJumping()
   SelectFloat(A=@n6, B=0, bPickA=false)
   n19 = Get bRigidBodyEnabled()
   SelectFloat(A=1, B=0, bPickA=NOT Boolean(A=@n19))
   AND Boolean(A=@n13, B=NOT Boolean(A=@n13.bIsBackblingLong))
   "Not Equal (Boolean)"(A=@n19) -> (ReturnValue=~)
   Get bIsTargeting()
  ]}
]
variables [
 backpack_wing_var {type=number}
 backpack_wing_var_2 {type=number}
 Backpack_Wing_Var_3 {type=number}
 FrontEndCape {type=boolean}
 Gravity Override Params {type=GravityOverrideParamsStruct}
 gravity_override {type=Vector}
 Head_Up {type=number}
 head_up_rban_off {type=number}
 is_WearingCape {type=boolean}
 isWearingPackBool {type=boolean}
 jumping {type=number}
 LL_Backpack_Equipped {type=boolean}
 neck_rot_dn {type=number}
 neck_rot_up {type=number}
 Pawn Owner {type=FortPlayerPawnAthena}
 rBAN {type=number}
 Shoulder_Mesh {type=SkeletalMeshComponent}
 Skel_Mesh {type=SkeletalMeshComponent}
]
